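# Physics asset for the chain on the female medium SilentTempo Peril character body, set up for simulated secondary motion.
# Three dyn_mstr_chain_sim bones carry capsules, from a tiny root to longer, thicker segments, linked in series by two constraints.
# Limits favor one swing axis, keeping the other swing and twist tight.
# Capsules on the pelvis, left thigh and left calf are unconstrained body volumes, apparently colliders for the chain.
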PhysicsAsset /BRCosmetics/Characters/Player/Female/Medium/Bodies/F_MED_SilentTempo_Peril/Meshes/F_MED_SilentTempo_Peril_Chain_Physics
bodies {
 dfrm_calf_l [
  {
   shapeType=Capsule
   length=40.7286
   radius=9.5896
   shapeName=dfrm_calf_l_capsule
   center=<27.544 2.8812 0>
   rotation=<rot 90 179.9999 -179.9999>
  }
  {shapeType=Capsule, length=4.3319, radius=9.4123, shapeName=dfrm_calf_l_capsule, center=<0.9113 3.0061 0>, rotation=<rot 0 -0.0003 -90>}
 ]
 dfrm_thigh_l [{
   shapeType=Capsule
   length=20.3514
   radius=12.25
   shapeName=dfrm_thigh_l_capsule
   center=<22.5408 2.0645 3.2318>
   rotation=<rot 88.7613 -0.0015 2.9551>
  }]
 dyn_mstr_chain_sim [{shapeType=Capsule, length=0, radius=0.5, shapeName=dyn_mstr_chain_sim_capsule}]
 dyn_mstr_chain_sim_01 [{shapeType=Capsule, length=28.2476, radius=2, shapeName=dyn_mstr_chain_sim_01_capsule, center=<9 -1 0>}]
 dyn_mstr_chain_sim_02 [{shapeType=Capsule, length=36.7615, radius=3, shapeName=dyn_mstr_chain_sim_02_capsule, center=<9.2661 -1 0>}]
 pelvis [{shapeType=Capsule, length=12.837, radius=10.3442, shapeName=pelvis_capsule, center=<-8.3681 -3.5647 0>}]
}
constraints [
 {bone1Name=dyn_mstr_chain_sim_02, bone2Name=dyn_mstr_chain_sim_01, swing1LimitDegrees=25, swing2LimitDegrees=5, twistLimitDegrees=5}
 {bone1Name=dyn_mstr_chain_sim_01, bone2Name=dyn_mstr_chain_sim, swing1LimitDegrees=25, swing2LimitDegrees=5, twistLimitDegrees=5}
]
