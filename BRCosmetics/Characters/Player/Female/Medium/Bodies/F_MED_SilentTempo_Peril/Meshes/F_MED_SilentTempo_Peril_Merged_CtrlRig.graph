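# Control Rig for the SilentTempo Peril female body cosmetic, adding corrective and secondary bone motion alongside a shared DeformRig function.
# Body Additions turns pose readers, leg measurements, anim attributes and crouch, DBNO and skydive states into weighted skirt and sleeve offsets, plus aim and parent constraints.
# Chain IK runs spline IK on chains, blending control poses by left-thigh raise.
# It also spins clock hands from game time at 4x speed.
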ControlRigBlueprint /BRCosmetics/Characters/Player/Female/Medium/Bodies/F_MED_SilentTempo_Peril/Meshes/F_MED_SilentTempo_Peril_Merged_CtrlRig
graphs [
 ContainedGraph {nodes [
   n1 = RigVMFunctionEntryNode()
   SequenceAggregate = &RigVMFunction_Sequence(ExecuteContext=@n1.ExecuteContext)
   SequenceAggregate_1 = &RigVMFunction_Sequence(ExecuteContext=@SequenceAggregate.B)
   SequenceAggregate_2 = &RigVMFunction_Sequence(ExecuteContext=@SequenceAggregate_1.B)
   SequenceAggregate_3 = &RigVMFunction_Sequence(ExecuteContext=@SequenceAggregate_2.B)
   SequenceAggregate_4 = &RigVMFunction_Sequence(ExecuteContext=@SequenceAggregate_3.B)
   SequenceAggregate_5 = &RigVMFunction_Sequence(ExecuteContext=@SequenceAggregate_4.B)
   RigVMFunctionReturnNode(
    A=@SequenceAggregate.A
    B=@SequenceAggregate_1.A
    C=@SequenceAggregate_2.A
    D=@SequenceAggregate_3.A
    E=@SequenceAggregate_4.A
    F=@SequenceAggregate_5.A
    G=@SequenceAggregate_5.B
   )
  ]}
 ContainedGraph {nodes [
   n1 = RigVMFunctionEntryNode()
   SequenceAggregate = &RigVMFunction_Sequence(ExecuteContext=@n1.ExecuteContext)
   SequenceAggregate_1 = &RigVMFunction_Sequence(ExecuteContext=@SequenceAggregate.B)
   SequenceAggregate_2 = &RigVMFunction_Sequence(ExecuteContext=@SequenceAggregate_1.B)
   SequenceAggregate_3 = &RigVMFunction_Sequence(ExecuteContext=@SequenceAggregate_2.B)
   SequenceAggregate_4 = &RigVMFunction_Sequence(ExecuteContext=@SequenceAggregate_3.B)
   SequenceAggregate_5 = &RigVMFunction_Sequence(ExecuteContext=@SequenceAggregate_4.B)
   SequenceAggregate_6 = &RigVMFunction_Sequence(ExecuteContext=@SequenceAggregate_5.B)
   RigVMFunctionReturnNode(
    A=@SequenceAggregate.A
    B=@SequenceAggregate_1.A
    C=@SequenceAggregate_2.A
    D=@SequenceAggregate_3.A
    E=@SequenceAggregate_4.A
    F=@SequenceAggregate_5.A
    G=@SequenceAggregate_6.A
    H=@SequenceAggregate_6.B
   )
  ]}
 New Function_ContainedGraph {nodes [
   RigVMFunctionEntryNode()
   > SequenceAggregate = RigVMAggregateNode()
   n1 = &RigUnit_SphericalPoseReader(
    ActiveRegionScaleFactors=(PositiveWidth=3.5,NegativeWidth=1,PositiveHeight=3.5,NegativeHeight=3.5)
    ActiveRegionSize=0.1
    Debug=(bDrawDebug=false,DebugScale=25,DebugSegments=20,DebugThickness=0.25)
    DriverAxis=(X=-1,Y=0,Z=0)
    DriverItem=(Type=Bone,Name="thigh_l")
    FalloffRegionScaleFactors=(PositiveWidth=1,NegativeWidth=0.75,PositiveHeight=1,NegativeHeight=1)
    FalloffSize=0.35
    OptionalParentItem=(Type=Bone)
    RotationOffset=(X=-90,Y=0,Z=-35)
    ExecutePin=@SequenceAggregate.B
   )
   > RigVMVariableNode(Value=@n1.OutputParam) {variable=spr_thigh_l_back_alpha}
   > n2 = &RigUnit_SphericalPoseReader(
    ActiveRegionScaleFactors=(PositiveWidth=3.5,NegativeWidth=1,PositiveHeight=3.5,NegativeHeight=3.5)
    ActiveRegionSize=0.1
    Debug=(bDrawDebug=false,DebugScale=25,DebugSegments=20,DebugThickness=0.25)
    DriverAxis=(X=1,Y=0,Z=0)
    DriverItem=(Type=Bone,Name="thigh_r")
    FalloffRegionScaleFactors=(PositiveWidth=1,NegativeWidth=0.75,PositiveHeight=1,NegativeHeight=1)
    FalloffSize=0.35
    FlipHeightScaling=false
    FlipWidthScaling=true
    OptionalParentItem=(Type=Bone)
    RotationOffset=(X=90,Y=0,Z=-35)
   )
   > RigVMVariableNode(Value=@n2.OutputParam) {variable=spr_thigh_r_back_alpha}
   n3 = &RigUnit_GetRelativeTransformForItem(Child=(Type=Bone,Name="calf_l"), Parent=(Type=Bone,Name="thigh_l"), bChildInitial=False, bParentInitial=False)
   > RigVMVariableNode(Value=&RigVMFunction_MathDoubleRemap(
     SourceMaximum=110
     SourceMinimum=25
     TargetMaximum=1
     TargetMinimum=0
     Value=&RigVMFunction_MathQuaternionToEuler(RotationOrder=ZYX, Value=@n3.RelativeTransform.Rotation).Result.Z
    ).Result) {variable=Calf_l_Bent}
   n4 = &RigUnit_GetRelativeTransformForItem(Child=(Type=Bone,Name="foot_l"), Parent=(Type=Bone,Name="root"), bChildInitial=False, bParentInitial=False)
   > RigVMVariableNode(Value=&RigVMFunction_MathDoubleRemap(
     SourceMaximum=70
     SourceMinimum=25
     TargetMaximum=1
     TargetMinimum=0
     Value=&RigVMFunction_MathVectorDistance(B=(X=0,Y=20,Z=0), A=@n4.RelativeTransform.Translation).Result
    ).Result) {variable=Thigh_l_Backward}
   n5 = &RigUnit_GetRelativeTransformForItem(Child=(Type=Bone,Name="calf_r"), Parent=(Type=Bone,Name="thigh_r"), bChildInitial=False, bParentInitial=False)
   > RigVMVariableNode(Value=&RigVMFunction_MathDoubleRemap(
     SourceMaximum=110
     SourceMinimum=25
     TargetMaximum=1
     TargetMinimum=0
     Value=&RigVMFunction_MathQuaternionToEuler(RotationOrder=ZYX, Value=@n5.RelativeTransform.Rotation).Result.Z
    ).Result) {variable=Calf_r_Bent}
   n6 = &RigUnit_GetRelativeTransformForItem(Child=(Type=Bone,Name="foot_r"), Parent=(Type=Bone,Name="root"), bChildInitial=False, bParentInitial=false)
   > RigVMVariableNode(Value=&RigVMFunction_MathDoubleRemap(
     SourceMaximum=70
     SourceMinimum=25
     TargetMaximum=1
     TargetMinimum=0
     Value=&RigVMFunction_MathVectorDistance(B=(X=0,Y=20,Z=0), A=@n6.RelativeTransform.Translation).Result
    ).Result) {variable=Thigh_r_Backward}
   SequenceAggregate_3_1 = RigVMAggregateNode(ExecuteContext=@SequenceAggregate.C)
   n7 = RigVMVariableNode() {variable=forward_backward_movement_alpha}
   RigVMVariableNode(
    Value=&RigVMFunction_MathDoubleRemap(SourceMaximum=1, SourceMinimum=0, TargetMaximum=1, TargetMinimum=0, Value=@n7.Value).Result
    ExecuteContext=@SequenceAggregate_3_1.A
   ) {variable=forward_movement_alpha}
   Remap = &RigVMFunction_MathDoubleRemap(SourceMaximum=0, SourceMinimum=-1, TargetMaximum=0, TargetMinimum=1, Value=@n7.Value)
   > RigVMVariableNode(Value=@Remap.Result) {variable=backward_movement_alpha}
   /Script/ControlRig.RigUnit_PositionConstraintLocalSpaceOffset(
    Child=(Type=Bone,Name="dyn_mstr_back_skirt_l")
    Filter=(bX=True,bY=True,bZ=True)
    Parents=((Item=(Type=Bone,Name="dfrm_thigh_l"),Weight=0.159),(Item=(Type=Bone,Name="deform_glute_l"),Weight=0.2895),(Item=(Type=Bone,Name="pelvis"),Weight=0.5515))
    Weight=1
    bMaintainOffset=True
    ExecutePin=@SequenceAggregate_3_1.C
   ) {is=struct RigVMUnitNode}
   > &RigUnit_PositionConstraintLocalSpaceOffset(
    Child=(Type=Bone,Name="dyn_mstr_back_skirt_r")
    Filter=(bX=True,bY=True,bZ=True)
    Parents=((Item=(Type=Bone,Name="dfrm_thigh_r"),Weight=0.159),(Item=(Type=Bone,Name="deform_glute_r"),Weight=0.3017),(Item=(Type=Bone,Name="pelvis"),Weight=0.5264))
    Weight=1
    bMaintainOffset=True
   )
   n8 = RigVMVariableNode() {variable=left_right_movement_alpha}
   n9 = &RigVMFunction_MathDoubleRemap(SourceMaximum=0, SourceMinimum=-1, TargetMaximum=0, TargetMinimum=1, Value=@n8.Value)
   RigVMVariableNode(Value=@n9.Result, ExecuteContext=@SequenceAggregate_3_1.B) {variable=left_movement_alpha}
   > DISPATCH_RigDispatch_SetAnimAttribute = RigVMDispatchNode(BoneName=root, Name=body_ctrlrig_left_mvmnt_alpha, Value=@n9.Result) {template="DISPATCH_RigDispatch_SetAnimAttribute(in Name,in BoneName,in Value,out Success)"}
   > RigVMVariableNode(Value=&RigVMFunction_MathDoubleRemap(SourceMaximum=1, SourceMinimum=0, TargetMaximum=1, TargetMinimum=0, Value=@n8.Value).Result) {variable=right_movement_alpha}
   /Script/ControlRig.RigUnit_AimItem(
    DebugSettings=(Scale=10,WorldOffset=(Rotation=(X=0,Y=0,Z=0,W=1),Translation=(X=0,Y=0,Z=0),Scale3D=(X=1,Y=1,Z=1)))
    Item=(Type=Bone,Name="dfrm_thigh_l")
    Primary=(Weight=1,Axis=(X=1,Y=0,Z=0),Target=(X=0,Y=0,Z=0),Kind=Location,Space=(Type=Bone,Name="calf_l"))
    Secondary=(Weight=1,Axis=(X=0,Y=0,Z=-1),Target=(X=0,Y=0,Z=0),Kind=Location,Space=(Type=Bone,Name="dfrm_aim_up_thigh_l"))
    Weight=1
    ExecutePin=@SequenceAggregate.D
   ) {is=struct RigVMUnitNode}
   > &RigUnit_AimItem(
    DebugSettings=(Scale=10,WorldOffset=(Rotation=(X=0,Y=0,Z=0,W=1),Translation=(X=0,Y=0,Z=0),Scale3D=(X=1,Y=1,Z=1)))
    Item=(Type=Bone,Name="dfrm_calf_l")
    Primary=(Weight=1,Axis=(X=1,Y=0,Z=0),Target=(X=0,Y=0,Z=0),Kind=Location,Space=(Type=Bone,Name="foot_l"))
    Secondary=(Weight=1,Axis=(X=0,Y=0,Z=-1),Target=(X=0,Y=0,Z=0),Kind=Location,Space=(Type=Bone,Name="dfrm_aim_up_calf_l"))
    Weight=1
   )
   > &RigUnit_AimItem(
    DebugSettings=(Scale=10,WorldOffset=(Rotation=(X=0,Y=0,Z=0,W=1),Translation=(X=0,Y=0,Z=0),Scale3D=(X=1,Y=1,Z=1)))
    Item=(Type=Bone,Name="dfrm_thigh_r")
    Primary=(Weight=1,Axis=(X=1,Y=0,Z=0),Target=(X=0,Y=0,Z=0),Kind=Location,Space=(Type=Bone,Name="calf_r"))
    Secondary=(Weight=1,Axis=(X=0,Y=0,Z=1),Target=(X=0,Y=0,Z=0),Kind=Location,Space=(Type=Bone,Name="dfrm_aim_up_thigh_r"))
    Weight=1
   )
   > &RigUnit_AimItem(
    DebugSettings=(Scale=10,WorldOffset=(Rotation=(X=0,Y=0,Z=0,W=1),Translation=(X=0,Y=0,Z=0),Scale3D=(X=1,Y=1,Z=1)))
    Item=(Type=Bone,Name="dfrm_calf_r")
    Primary=(Weight=1,Axis=(X=1,Y=0,Z=0),Target=(X=0,Y=0,Z=0),Kind=Location,Space=(Type=Bone,Name="foot_r"))
    Secondary=(Weight=1,Axis=(X=0,Y=0,Z=1),Target=(X=0,Y=0,Z=0),Kind=Location,Space=(Type=Bone,Name="dfrm_aim_up_calf_r"))
    Weight=1
   )
   SequenceAggregate_2 = RigVMAggregateNode(ExecuteContext=@SequenceAggregate.F)
   SequenceAggregate_1_1 = &RigVMFunction_Sequence(ExecuteContext=@SequenceAggregate_2.A)
   n10 = &RigVMFunction_ControlFlowBranch(ExecuteContext=@SequenceAggregate_1_1.A, Condition=RigVMVariableNode() {variable=is_skydive_floating}.Value)
   RigVMVariableNode(Value=1, ExecuteContext=@n10.False) {variable=hand_r_multFactor}
   RigVMVariableNode(Value=5.5, ExecuteContext=@n10.True) {variable=hand_r_multFactor}
   Multiply = &RigVMFunction_MathDoubleMul(
    A=RigVMDispatchNode(BoneName=root, Default=0, Name=deformRig_r_wrist_up_alpha) {template="DISPATCH_RigDispatch_GetAnimAttribute(in Name,in BoneName,in Default,out Value,out Found)"}.Value
    B=RigVMVariableNode() {variable=hand_r_multFactor}.Value
   )
   Clamp = &RigVMFunction_MathDoubleClamp(Value=@Multiply.Result)
   n11 = &RigVMFunction_MathQuaternionFromEuler(Euler=(X=0,Y=0,Z=-35))
   &RigUnit_ModifyTransforms(
    ItemToModify="((Item=(Type=Bone,Name=\"dfrm_sleeve_01_r\"),Transform=(Rotation=(X=0,Y=0,Z=0,W=1),Translation=(X=0,Y=0,Z=0),Scale3D=(X=1,Y=1,Z=1))),(Item=(Type=Bone,Name=\"dfrm_mstr_sleeve_01_r\"),Transform=(Rotation=(X=0,Y=0,Z=0,W=1),Translation=(X=0,Y=0,Z=0),Scale3D=(X=1,Y ...(truncated)"
    Weight=@Clamp.Result
    ItemToModify.0.Transform.Rotation=&RigVMFunction_MathQuaternionFromEuler(Euler=(X=0,Y=0,Z=-50)).Result
    ItemToModify.2.Transform.Rotation=@n11.Result
    ItemToModify.3.Transform.Rotation=@n11.Result
    ItemToModify.4.Transform.Rotation=@n11.Result
    ItemToModify.5.Transform.Rotation=@n11.Result
    ExecutePin=@SequenceAggregate_1_1.B
   )
   SequenceAggregate_1 = &RigVMFunction_Sequence(ExecuteContext=@SequenceAggregate_2.B)
   n12 = RigVMDispatchNode(BoneName=root, Default=0, Name=deformRig_l_thigh_up_alpha) {template="DISPATCH_RigDispatch_GetAnimAttribute(in Name,in BoneName,in Default,out Value,out Found)"}
   n13 = RigVMVariableNode() {variable=is_crouching}
   Add = /Script/RigVM.RigVMFunction_MathVectorAdd(A=(X=44,Y=-27,Z=16), B=&RigUnit_GetTransform(Item=(Type=Bone,Name="root"), bInitial=false).Transform.Translation) {is=struct RigVMUnitNode, template="Add::Execute(in A,in B,out Result)"}
   RigVMFunction_MathVectorDistance = &RigVMFunction_MathVectorDistance(
    A=@Add.Result
    B=&RigUnit_GetRelativeTransformForItem(Child=(Type=Bone,Name="foot_r"), Parent=(Type=Bone,Name="root"), bChildInitial=false, bParentInitial=false).RelativeTransform.Translation
   )
   Subtract = /Script/RigVM.RigVMFunction_MathFloatSub(
    A=1
    B=&RigVMFunction_MathFloatRemap(SourceMaximum=0, SourceMinimum=60, TargetMinimum=0, bClamp=true, Value=@RigVMFunction_MathVectorDistance.Result).Result
   ) {is=struct RigVMUnitNode, template="Subtract::Execute(in A,in B,out Result)"}
   n14 = &RigVMFunction_MathFloatMul(
    B=&RigVMFunction_MathFloatMul(A=&RigVMFunction_MathBoolToFloat(Value=@n13.Value).Result, B=@Subtract.Result) {template="Multiply::Execute(in A,in B,out Result)"}.Result
    A=RigVMVariableNode() {variable=right_movement_alpha}.Value
   ) {template="Multiply::Execute(in A,in B,out Result)"}
   Cast = &RigVMFunction_MathBoolToFloat(Value=RigVMVariableNode() {variable=is_dbno}.Value)
   n15 = &RigVMFunction_MathFloatSub(A=1, B=@Cast.Result) {template="Subtract::Execute(in A,in B,out Result)"}
   &RigUnit_ModifyTransforms(
    ItemToModify="((Item=(Type=Bone,Name=\"dyn_mstr_inner_front_skirt_l\"),Transform=(Rotation=(X=0,Y=0,Z=0,W=1),Translation=(X=-6,Y=-2,Z=-5),Scale3D=(X=1,Y=1,Z=1))),(Item=(Type=Bone,Name=\"dyn_mstr_front_skirt_l\"),Transform=(Rotation=(X=0,Y=0,Z=0,W=1),Translation=(X=-15,Y=3,Z ...(truncated)"
    Weight=&RigVMFunction_MathFloatMul(
     B=&RigVMFunction_MathFloatSub(A=@n12.Value, B=@n14.Result) {template="Subtract::Execute(in A,in B,out Result)"}.Result
     A=@n15.Result
    ) {template="Multiply::Execute(in A,in B,out Result)"}.Result
    ItemToModify.0.Transform.Rotation=&RigVMFunction_MathQuaternionFromEuler(Euler=(X=0,Y=45,Z=-5)).Result
    ItemToModify.1.Transform.Rotation=&RigVMFunction_MathQuaternionFromEuler(Euler=(X=30,Y=85,Z=-25)).Result
    ItemToModify.2.Transform.Rotation=&RigVMFunction_MathQuaternionFromEuler(Euler=(X=-15,Y=90,Z=-5)).Result
    ItemToModify.3.Transform.Rotation=&RigVMFunction_MathQuaternionFromEuler(Euler=(X=-25,Y=100,Z=-25)).Result
    ItemToModify.4.Transform.Rotation=&RigVMFunction_MathQuaternionFromEuler(Euler=(X=0,Y=0,Z=10)).Result
    ItemToModify.5.Transform.Rotation=&RigVMFunction_MathQuaternionFromEuler(Euler=(X=0,Y=0,Z=5)).Result
    ExecutePin=@SequenceAggregate_1.A
   )
   > &RigUnit_ModifyTransforms(
    ItemToModify="((Item=(Type=Bone,Name=\"dyn_mstr_inner_front_skirt_l\"),Transform=(Rotation=(X=0,Y=0,Z=0,W=1),Translation=(X=-3,Y=-1,Z=-2.5),Scale3D=(X=1,Y=1,Z=1))),(Item=(Type=Bone,Name=\"dyn_mstr_front_skirt_l\"),Transform=(Rotation=(X=0,Y=0,Z=0,W=1),Translation=(X=-15,Y=3 ...(truncated)"
    ItemToModify.0.Transform.Rotation=&RigVMFunction_MathQuaternionFromEuler(Euler=(X=0,Y=25,Z=-5)).Result
    ItemToModify.1.Transform.Rotation=&RigVMFunction_MathQuaternionFromEuler(Euler=(X=30,Y=65,Z=-25)).Result
    ItemToModify.2.Transform.Rotation=&RigVMFunction_MathQuaternionFromEuler(Euler=(X=-15,Y=75,Z=-5)).Result
    ItemToModify.3.Transform.Rotation=&RigVMFunction_MathQuaternionFromEuler(Euler=(X=-25,Y=75,Z=-5)).Result
    ItemToModify.4.Transform.Rotation=&RigVMFunction_MathQuaternionFromEuler(Euler=(X=0,Y=0,Z=15)).Result
    ItemToModify.5.Transform.Rotation=&RigVMFunction_MathQuaternionFromEuler(Euler=(X=0,Y=0,Z=10)).Result
    ItemToModify.6.Transform.Rotation=&RigVMFunction_MathQuaternionFromEuler(Euler=(X=0,Y=0,Z=20)).Result
    ItemToModify.7.Transform.Rotation=&RigVMFunction_MathQuaternionFromEuler(Euler=(X=0,Y=0,Z=10)).Result
    Weight=&RigVMFunction_MathFloatMul(B=@n14.Result, A=@n15.Result) {template="Multiply::Execute(in A,in B,out Result)"}.Result
   )
   &RigUnit_ModifyTransforms(
    ItemToModify="((Item=(Type=Bone,Name=\"dyn_side_skirt_01_l\"),Transform=(Rotation=(X=0,Y=0,Z=0,W=1),Translation=(X=0,Y=-2,Z=0),Scale3D=(X=1,Y=1,Z=1))),(Item=(Type=Bone,Name=\"dyn_side_skirt_02_l\"),Transform=(Rotation=(X=0,Y=0,Z=0,W=1),Translation=(X=0,Y=0,Z=0),Scale3D=(X=1 ...(truncated)"
    Weight=&RigVMFunction_MathBoolToFloat(Value=RigVMVariableNode() {variable=is_dbno}.Value).Result
    ItemToModify.0.Transform.Rotation=&RigVMFunction_MathQuaternionFromEuler(Euler=(X=0,Y=15,Z=-20)).Result
    ItemToModify.1.Transform.Rotation=&RigVMFunction_MathQuaternionFromEuler(Euler=(X=0,Y=15,Z=-30)).Result
    ItemToModify.2.Transform.Rotation=&RigVMFunction_MathQuaternionFromEuler(Euler=(X=0,Y=0,Z=-65)).Result
    ItemToModify.3.Transform.Rotation=&RigVMFunction_MathQuaternionFromEuler(Euler=(X=0,Y=-10,Z=5)).Result
    ExecutePin=@SequenceAggregate_1.B
   )
   &RigUnit_ModifyTransforms(
    ItemToModify="((Item=(Type=Bone,Name=\"dyn_mstr_side_skirt_l\"),Transform=(Rotation=(X=0,Y=0,Z=0,W=1),Translation=(X=0,Y=0,Z=-2),Scale3D=(X=1,Y=1,Z=1))),(Item=(Type=Bone,Name=\"dyn_mstr_back_side_skirt_l\"),Transform=(Rotation=(X=0,Y=0,Z=0,W=1),Translation=(X=0,Y=-1,Z=1),Sc ...(truncated)"
    ItemToModify.0.Transform.Rotation=&RigVMFunction_MathQuaternionFromEuler(Euler=(X=0,Y=0,Z=-5)).Result
    ItemToModify.1.Transform.Rotation=&RigVMFunction_MathQuaternionFromEuler(Euler=(X=0,Y=-5,Z=-5)).Result
    Weight=RigVMVariableNode() {variable=spine_twist_right_delta_ref_alpha}.Value
    ExecutePin=@SequenceAggregate_2.D
   )
   n16 = RigVMVariableNode() {variable=is_skydiving}
   RigVMFunction_MathQuaternionToEuler = &RigVMFunction_MathQuaternionToEuler(RotationOrder=ZYX, Value=&RigUnit_GetTransform(Item=(Type=Bone,Name="calf_l")).Transform.Rotation)
   &RigUnit_ModifyTransforms(
    ItemToModify="((Item=(Type=Bone,Name=\"dyn_back_side_skirt_01_l\"),Transform=(Rotation=(X=0,Y=0,Z=0,W=1),Translation=(X=0,Y=0,Z=0),Scale3D=(X=1,Y=1,Z=1))),(Item=(Type=Bone,Name=\"dyn_back_side_skirt_02_l\"),Transform=(Rotation=(X=0,Y=0,Z=0,W=1),Translation=(X=0,Y=0,Z=0),Sca ...(truncated)"
    Weight=&RigVMFunction_MathDoubleMul(
     A=&RigVMFunction_MathBoolToFloat(Value=@n16.Value).Result
     B=&RigVMFunction_MathDoubleRemap(SourceMaximum=-125, SourceMinimum=-165, TargetMaximum=1, TargetMinimum=0, Value=@RigVMFunction_MathQuaternionToEuler.Result.Z).Result
    ).Result
    ItemToModify.2.Transform.Rotation=&RigVMFunction_MathQuaternionFromEuler(Euler=(X=0,Y=0,Z=-30)).Result
    ItemToModify.1.Transform.Rotation=&RigVMFunction_MathQuaternionFromEuler(Euler=(X=0,Y=0,Z=-20)).Result
    ItemToModify.0.Transform.Rotation=&RigVMFunction_MathQuaternionFromEuler(Euler=(X=0,Y=0,Z=-10)).Result
    ExecutePin=@SequenceAggregate_2.E
   )
   RigVMFunction_MathQuaternionFromEuler = &RigVMFunction_MathQuaternionFromEuler(Euler=(X=0,Y=0,Z=-75))
   &RigUnit_ModifyTransforms(
    ItemToModify="((Item=(Type=Bone,Name=\"dyn_mstr_back_skirt_l\"),Transform=(Rotation=(X=0,Y=0,Z=0,W=1),Translation=(X=-5,Y=0,Z=0),Scale3D=(X=1,Y=1,Z=1))),(Item=(Type=Bone,Name=\"dyn_mstr_back_side_skirt_l\"),Transform=(Rotation=(X=0,Y=0,Z=0,W=1),Translation=(X=0,Y=0,Z=0),Sca ...(truncated)"
    ItemToModify.0.Transform.Rotation=@RigVMFunction_MathQuaternionFromEuler.Result
    ItemToModify.2.Transform.Rotation=&RigVMFunction_MathQuaternionFromEuler(Euler=(X=0,Y=0,Z=-45)).Result
    Weight=RigVMVariableNode() {variable=spr_thigh_l_back_alpha}.Value
    ExecutePin=@SequenceAggregate_2.C
   )
   > &RigUnit_ModifyTransforms(
    ItemToModify="((Item=(Type=Bone,Name=\"dyn_mstr_back_skirt_r\"),Transform=(Rotation=(X=0,Y=0,Z=0,W=1),Translation=(X=-5,Y=-3,Z=0),Scale3D=(X=1,Y=1,Z=1))),(Item=(Type=Bone,Name=\"dyn_mstr_back_skirt_l\"),Transform=(Rotation=(X=0,Y=0,Z=0,W=1),Translation=(X=0,Y=0,Z=0),Scale3D ...(truncated)"
    ItemToModify.0.Transform.Rotation=&RigVMFunction_MathQuaternionFromEuler(Euler=(X=0,Y=0,Z=-65)).Result
    ItemToModify.1.Transform.Rotation=&RigVMFunction_MathQuaternionFromEuler(Euler=(X=0,Y=0,Z=-35)).Result
    Weight=RigVMVariableNode() {variable=spr_thigh_r_back_alpha}.Value
   )
   n17 = &RigVMFunction_MathDoubleMul(A=RigVMVariableNode() {variable=Calf_l_Bent}.Value, B=RigVMVariableNode() {variable=Thigh_l_Backward}.Value)
   > &RigUnit_ModifyTransforms(
    ItemToModify=((Item=(Type=Bone,Name="dyn_mstr_side_skirt_l"),Transform=(Rotation=(X=0,Y=0,Z=-0.1736,W=0.9848),Translation=(X=0,Y=0,Z=0),Scale3D=(X=1,Y=1,Z=1))))
    WeightMinimum=0
    Weight=@n17.Result
   )
   > &RigUnit_ModifyTransforms(
    ItemToModify=((Item=(Type=Bone,Name="dyn_mstr_back_side_skirt_l"),Transform=(Rotation=(X=0,Y=0,Z=-0.2588,W=0.9659),Translation=(X=0,Y=0,Z=0),Scale3D=(X=1,Y=1,Z=1))))
    WeightMinimum=0
    Weight=@n17.Result
   )
   > &RigUnit_ModifyTransforms(
    ItemToModify=((Item=(Type=Bone,Name="dyn_mstr_back_skirt_l"),Transform=(Rotation=(X=0,Y=0,Z=-0.3827,W=0.9239),Translation=(X=0,Y=0,Z=0),Scale3D=(X=1,Y=1,Z=1))))
    WeightMinimum=0
    Weight=@n17.Result
   )
   > &RigUnit_ModifyTransforms(
    ItemToModify=((Item=(Type=Bone,Name="dyn_back_skirt_02_l"),Transform=(Rotation=(X=0,Y=0,Z=-0.3827,W=0.9239),Translation=(X=0,Y=0,Z=0),Scale3D=(X=1,Y=1,Z=1))))
    WeightMinimum=0
    Weight=@n17.Result
   )
   > &RigUnit_ModifyTransforms(
    ItemToModify=((Item=(Type=Bone,Name="dyn_mstr_chain_sim"),Transform=(Rotation=(X=0,Y=0,Z=-0.0872,W=0.9962),Translation=(X=0,Y=0,Z=0),Scale3D=(X=1,Y=1,Z=1))))
    WeightMinimum=0
    Weight=@n17.Result
   )
   > &RigUnit_ModifyTransforms(
    ItemToModify=((Item=(Type=Bone,Name="dyn_mstr_back_skirt_r"),Transform=(Rotation=(X=0,Y=0,Z=-0.3007,W=0.9537),Translation=(X=0,Y=0,Z=0),Scale3D=(X=1,Y=1,Z=1))))
    WeightMinimum=0
    Weight=&RigVMFunction_MathDoubleMul(A=RigVMVariableNode() {variable=Calf_r_Bent}.Value, B=RigVMVariableNode() {variable=Thigh_r_Backward}.Value).Result
   )
   > &RigUnit_ModifyTransforms(
    ItemToModify=((Item=(Type=Bone,Name="dyn_mstr_back_skirt_r"),Transform=(Rotation=(X=0,Y=0,Z=-0.0872,W=0.9962),Translation=(X=0,Y=0,Z=0),Scale3D=(X=1,Y=1,Z=1))))
    Weight=1
    WeightMinimum=0
   )
   RigVMFunctionReturnNode(ExecuteContext=@SequenceAggregate.G)
   &RigUnit_ParentConstraint(
    AdvancedSettings=(RotationOrderForFilter=XZY)
    Child=(Type=Bone,Name="dyn_base_shawl_A_r")
    Parents="((Item=(Type=Bone,Name=\"spine_05\"),Weight=0.0174),(Item=(Type=Bone,Name=\"deform_upperarm_up_r\"),Weight=0.0673),(Item=(Type=Bone,Name=\"clavicle_r\"),Weight=0.1728),(Item=(Type=Bone,Name=\"deform_upperarm_bk_r\"),Weight=0.3106),(Item=(Type=Bone,Name=\"upperarm_t ...(truncated)"
    Weight=1
    ExecutePin=@SequenceAggregate.E
   )
   > &RigUnit_ParentConstraint(
    AdvancedSettings=(RotationOrderForFilter=XZY)
    Child=(Type=Bone,Name="dyn_base_shawl_B_r")
    Parents=((Item=(Type=Bone,Name="deform_upperarm_up_r"),Weight=0.027),(Item=(Type=Bone,Name="clavicle_r"),Weight=0.0271),(Item=(Type=Bone,Name="deform_upperarm_bk_r"),Weight=0.1547),(Item=(Type=Bone,Name="upperarm_twist_01_r"),Weight=0.7912))
    Weight=1
   )
   > &RigUnit_ParentConstraint(
    AdvancedSettings=(RotationOrderForFilter=XZY)
    Child=(Type=Bone,Name="dyn_mstr_clock_chain")
    Parents=((Item=(Type=Bone,Name="deform_glute_r"),Weight=0.1299),(Item=(Type=Bone,Name="dfrm_thigh_r"),Weight=0.1532),(Item=(Type=Bone,Name="pelvis"),Weight=0.7169))
    Weight=1
   )
   > &RigUnit_ParentConstraint(
    AdvancedSettings=(InterpolationType=Average,RotationOrderForFilter=XZY)
    Child=(Type=Bone,Name="dyn_mstr_chain_sim")
    Parents=((Item=(Type=Bone,Name="dfrm_thigh_l"),Weight=0.7),(Item=(Type=Bone,Name="pelvis"),Weight=0.2),(Item=(Type=Bone,Name="deform_groin_l"),Weight=0.1))
    Weight=1
   )
   &RigUnit_ParentConstraint(
    AdvancedSettings=(RotationOrderForFilter=XZY)
    Child=(Type=Bone,Name="dyn_mstr_lace_l")
    ExecutePin=()
    Parents=((Item=(Type=Bone,Name="foot_l"),Weight=0.2137),(Item=(Type=Bone,Name="calf_l"),Weight=0.7863))
    Weight=1
   )
   > &RigUnit_ParentConstraint(
    AdvancedSettings=(RotationOrderForFilter=XZY)
    Child=(Type=Bone,Name="dyn_mstr_race_r")
    Parents=((Item=(Type=Bone,Name="foot_r"),Weight=0.2137),(Item=(Type=Bone,Name="calf_twist_01_r"),Weight=0.7863))
    Weight=1
   )
   &RigUnit_ParentConstraint(
    AdvancedSettings=(RotationOrderForFilter=XZY)
    Child=(Type=Bone,Name="dyn_mstr_chain_sim")
    ExecutePin=()
    Parents=((Item=(Type=Bone,Name="deform_glute_l"),Weight=0.0247),(Item=(Type=Bone,Name="deform_groin_l"),Weight=0.0682),(Item=(Type=Bone,Name="pelvis"),Weight=0.4336),(Item=(Type=Bone,Name="dfrm_thigh_l"),Weight=0.4735))
    Weight=1
   )
  ]}
 New Function_ContainedGraph {nodes [
   RigVMFunctionEntryNode()
   > SequenceAggregate = RigVMAggregateNode()
   n1 = &RigUnit_SphericalPoseReader(
    ActiveRegionScaleFactors=(PositiveWidth=3.5,NegativeWidth=0.5,PositiveHeight=3,NegativeHeight=3)
    ActiveRegionSize=0.1
    Debug=(bDrawDebug=false,DebugScale=25,DebugSegments=20,DebugThickness=0.25)
    DriverAxis=(X=-1,Y=0,Z=0)
    DriverItem=(Type=Bone,Name="thigh_l")
    FalloffRegionScaleFactors=(PositiveWidth=1,NegativeWidth=1,PositiveHeight=1,NegativeHeight=1)
    FalloffSize=0.45
    OptionalParentItem=(Type=Bone)
    RotationOffset=(X=90,Y=0,Z=0)
    ExecutePin=@SequenceAggregate.A
   )
   > RigVMVariableNode(Value=@n1.OutputParam) {variable=dfrm_rig_thigh_up_alpha}
   n2 = &RigVMFunction_MathDoubleMul(B=4, A=RigVMVariableNode() {variable=game_time_in_seconds}.Value)
   Modulo = /Script/RigVM.RigVMFunction_MathDoubleMod(B=60, A=@n2.Result) {is=struct RigVMUnitNode, template="Modulo::Execute(in A,in B,out Result)"}
   Multiply = &RigVMFunction_MathDoubleMul(B=-1, A=&RigVMFunction_MathDoubleMul(B=6, A=@Modulo.Result).Result)
   &RigUnit_ModifyTransforms(
    ItemToModify=((Item=(Type=Bone,Name="dfrm_clock_second"),Transform=(Rotation=(X=0,Y=0,Z=0,W=1),Translation=(X=0,Y=0,Z=0),Scale3D=(X=1,Y=1,Z=1))))
    Weight=1
    ItemToModify.0.Transform.Rotation=&RigVMFunction_MathQuaternionFromEuler(Euler=(X=0,Y=0,Z=0), Euler.Y=@Multiply.Result).Result
    ExecutePin=@SequenceAggregate.B
   )
   SequenceAggregate_4_1 = &RigVMFunction_Sequence(ExecuteContext=@SequenceAggregate.E)
   SequenceAggregate_5 = RigVMAggregateNode(ExecuteContext=@SequenceAggregate_4_1.A)
   &RigUnit_ParentConstraint(
    AdvancedSettings=(RotationOrderForFilter=XZY)
    Child=(Type=Bone,Name="dyn_mstr_front_chain")
    Parents=((Item=(Type=Bone,Name="dfrm_thigh_r"),Weight=0.0103),(Item=(Type=Bone,Name="deform_groin_l"),Weight=0.2658),(Item=(Type=Bone,Name="pelvis"),Weight=0.2906),(Item=(Type=Bone,Name="dfrm_thigh_l"),Weight=0.4333))
    Weight=1
    ExecutePin=@SequenceAggregate_5.A
   )
   > &RigUnit_ParentConstraint(
    AdvancedSettings=(RotationOrderForFilter=XZY)
    Child=(Type=Bone,Name="dyn_mstr_back_chain")
    Parents=((Item=(Type=Bone,Name="spine_01"),Weight=0.0209),(Item=(Type=Bone,Name="spine_02"),Weight=0.0353),(Item=(Type=Bone,Name="pelvis"),Weight=0.9437))
    Weight=1
   )
   > &RigUnit_ModifyTransforms(
    ItemToModify="((Item=(Type=Bone,Name=\"dyn_front_chain\"),Transform=(Rotation=(X=0,Y=0,Z=0,W=1),Translation=(X=0,Y=0,Z=0),Scale3D=(X=1,Y=1,Z=1))),(Item=(Type=Bone,Name=\"dyn_back_chain\"),Transform=(Rotation=(X=0,Y=0,Z=0,W=1),Translation=(X=0,Y=0,Z=0),Scale3D=(X=1,Y=1,Z=1)) ...(truncated)"
    ItemToModify.0.Transform.Rotation=&RigVMFunction_MathQuaternionFromEuler(Euler=(X=0,Y=45,Z=-45)).Result
    ItemToModify.1.Transform.Rotation=&RigVMFunction_MathQuaternionFromEuler(Euler=(X=0,Y=0,Z=0)).Result
    ItemToModify.2.Transform.Rotation=&RigVMFunction_MathQuaternionFromEuler(Euler=(X=5,Y=30,Z=0)).Result
    Weight=RigVMVariableNode() {variable=dfrm_rig_thigh_up_alpha}.Value
   )
   &RigUnit_ParentConstraint(
    AdvancedSettings=(RotationOrderForFilter=XZY)
    Child=(Type=Null,Name="dyn_front_chain_07_ctrl_null")
    Parents=((Item=(Type=Bone,Name="deform_groin_l"),Weight=0.0127),(Item=(Type=Bone,Name="pelvis"),Weight=0.0238),(Item=(Type=Bone,Name="dyn_mstr_inner_front_skirt_01_l"),Weight=0.1072),(Item=(Type=Bone,Name="dfrm_thigh_l"),Weight=0.8564))
    Weight=1
    ExecutePin=@SequenceAggregate_5.B
   )
   > &RigUnit_ParentConstraint(
    AdvancedSettings=(RotationOrderForFilter=XZY)
    Child=(Type=Null,Name="dyn_front_chain_04_ctrl_null")
    Parents=((Item=(Type=Control,Name="dyn_front_chain_01_ctrl"),Weight=0.5),(Item=(Type=Control,Name="dyn_front_chain_07_ctrl"),Weight=0.5))
    Weight=1
   )
   > &RigUnit_ParentConstraint(
    AdvancedSettings=(RotationOrderForFilter=XZY)
    Child=(Type=Null,Name="dyn_front_chain_010_ctrl_null")
    Parents=((Item=(Type=Control,Name="dyn_front_chain_07_ctrl"),Weight=0.75),(Item=(Type=Control,Name="dyn_main_chain_ctrl"),Weight=0.25))
    Weight=1
   )
   &RigUnit_ParentConstraint(
    AdvancedSettings=(RotationOrderForFilter=XZY)
    Child=(Type=Null,Name="dyn_back_chain_010_ctrl_null")
    Parents=((Item=(Type=Control,Name="dyn_back_chain_01_ctrl"),Weight=0.5),(Item=(Type=Control,Name="dyn_main_chain_ctrl"),Weight=0.5))
    Weight=1
    ExecutePin=@SequenceAggregate_5.C
   )
   > &RigUnit_ParentConstraint(
    AdvancedSettings=(RotationOrderForFilter=XZY)
    Child=(Type=Null,Name="dyn_back_chain_04_ctrl_null")
    Parents=((Item=(Type=Control,Name="dyn_back_chain_01_ctrl"),Weight=0.75),(Item=(Type=Control,Name="dyn_back_chain_010_ctrl"),Weight=0.25))
    Weight=1
   )
   > &RigUnit_ParentConstraint(
    AdvancedSettings=(RotationOrderForFilter=XZY)
    Child=(Type=Null,Name="dyn_back_chain_07_ctrl_null")
    Parents=((Item=(Type=Control,Name="dyn_back_chain_01_ctrl"),Weight=0.4),(Item=(Type=Control,Name="dyn_back_chain_010_ctrl"),Weight=0.6))
    Weight=1
   )
   > &RigUnit_ParentConstraint(
    AdvancedSettings=(RotationOrderForFilter=XZY)
    Child=(Type=Null,Name="dyn_back_chain_013_ctrl_null")
    Parents=((Item=(Type=Control,Name="dyn_back_chain_010_ctrl"),Weight=0.6),(Item=(Type=Control,Name="dyn_main_chain_ctrl"),Weight=0.4))
    Weight=1
   )
   > &RigUnit_ParentConstraint(
    AdvancedSettings=(RotationOrderForFilter=XZY)
    Child=(Type=Null,Name="dyn_back_chain_016_ctrl_null")
    Parents=((Item=(Type=Control,Name="dyn_back_chain_010_ctrl"),Weight=0.25),(Item=(Type=Control,Name="dyn_main_chain_ctrl"),Weight=0.75))
    Weight=1
   )
   &RigUnit_ParentConstraint(
    AdvancedSettings=(RotationOrderForFilter=XZY)
    Child=(Type=Null,Name="dyn_main_chain_ctrl_null")
    Parents=((Item=(Type=Bone,Name="dyn_mstr_chain_sim_02"),Weight=1))
    Weight=1
    ExecutePin=@SequenceAggregate_5.D
   )
   SequenceAggregate_3 = &RigVMFunction_Sequence(ExecuteContext=@SequenceAggregate_4_1.B)
   RigUnit_ItemArray = &RigUnit_ItemArray(Items="((Type=Bone,Name=\"dyn_front_chain_01\"),(Type=Bone,Name=\"dyn_front_chain_02\"),(Type=Bone,Name=\"dyn_front_chain_03\"),(Type=Bone,Name=\"dyn_front_chain_04\"),(Type=Bone,Name=\"dyn_front_chain_05\"),(Type=Bone,Name=\"dyn_front_chain_06\"),(Type=Bone,Name=\"dyn_front_ ...(truncated)")
   SplineIK = RigVMFunctionReferenceNode(
    Primary Axis=(X=1,Y=0,Z=0)
    Samples Per Segment=2
    Secondary Spline Direction=(X=0,Y=5,Z=0)
    Spline Mode=BSpline
    Stretch=true
    Up Axis=(X=0,Y=1,Z=0)
    debug=false
    Bones=@RigUnit_ItemArray.Items
    ExecuteContext=@SequenceAggregate_3.A
    Controls=RigVMVariableNode() {variable=front_chain_ctrls}.Value
   )
   n3 = &RigUnit_ItemArray(Items="((Type=Bone,Name=\"dyn_back_chain_01\"),(Type=Bone,Name=\"dyn_back_chain_02\"),(Type=Bone,Name=\"dyn_back_chain_03\"),(Type=Bone,Name=\"dyn_back_chain_04\"),(Type=Bone,Name=\"dyn_back_chain_05\"),(Type=Bone,Name=\"dyn_back_chain_06\"),(Type=Bone,Name=\"dyn_back_chain_0 ...(truncated)")
   > SplineIK_1 = RigVMFunctionReferenceNode(
    Primary Axis=(X=1,Y=0,Z=0)
    Samples Per Segment=2
    Secondary Spline Direction=(X=0,Y=5,Z=0)
    Spline Mode=BSpline
    Stretch=true
    Up Axis=(X=0,Y=1,Z=0)
    Bones=@n3.Items
    Controls=RigVMVariableNode() {variable=back_chain_ctrls}.Value
   )
   SequenceAggregate_1 = &RigVMFunction_Sequence(ExecuteContext=@SequenceAggregate_3.B)
   n4 = RigVMVariableNode() {variable=dfrm_rig_thigh_up_alpha}
   Subtract = &RigVMFunction_MathFloatSub(A=1, B=@n4.Value) {template="Subtract::Execute(in A,in B,out Result)"}
   n5 = RigVMVariableNode() {variable=chain_ctrls}
   /Script/ControlRig.RigUnit_HierarchySetPoseItemArray(
    ElementType=All
    ExecutePin=@SequenceAggregate_1.A
    Weight=@Subtract.Result
    ItemsToSet=@n5.Value
    Pose=RigVMVariableNode() {variable=chain_ctrls_bindpose_pose}.Value
   ) {is=struct RigVMUnitNode}
   &RigUnit_HierarchySetPoseItemArray(
    ElementType=All
    ExecutePin=@SequenceAggregate_1.B
    ItemsToSet=@n5.Value
    Pose=RigVMVariableNode() {variable=chain_ctrls_thigh_l_up_pose}.Value
    Weight=@n4.Value
   )
   RigVMFunctionReturnNode(ExecuteContext=@SequenceAggregate.H)
   Divide = &RigVMFunction_MathDoubleDiv(B=60, A=&RigVMFunction_MathDoubleMod(B=3600, A=@n2.Result) {template="Modulo::Execute(in A,in B,out Result)"}.Result)
   n6 = &RigVMFunction_MathDoubleMul(B=6, A=@Divide.Result)
   &RigUnit_ModifyTransforms(
    ItemToModify=((Item=(Type=Bone,Name="dfrm_clock_minute"),Transform=(Rotation=(X=0,Y=0,Z=0,W=1),Translation=(X=0,Y=0,Z=0),Scale3D=(X=1,Y=1,Z=1))))
    Weight=1
    ItemToModify.0.Transform.Rotation=&RigVMFunction_MathQuaternionFromEuler(Euler=(X=0,Y=0,Z=0), Euler.Y=&RigVMFunction_MathDoubleMul(B=-1, A=@n6.Result).Result).Result
    ExecutePin=@SequenceAggregate.C
   )
   n7 = &RigVMFunction_MathDoubleMod(B=43200, A=@n2.Result) {template="Modulo::Execute(in A,in B,out Result)"}
   n8 = &RigVMFunction_MathDoubleMod(B=3600, A=@n2.Result) {template="Modulo::Execute(in A,in B,out Result)"}
   Add = &RigVMFunction_MathDoubleAdd(
    A=&RigVMFunction_MathDoubleMul(B=30, A=&RigVMFunction_MathDoubleDiv(B=3600, A=@n7.Result).Result).Result
    B=&RigVMFunction_MathDoubleMul(B=0.5, A=&RigVMFunction_MathDoubleDiv(B=60, A=@n8.Result).Result).Result
   )
   &RigUnit_ModifyTransforms(
    ItemToModify=((Item=(Type=Bone,Name="dfrm_clock_hour"),Transform=(Rotation=(X=0,Y=0,Z=0,W=1),Translation=(X=0,Y=0,Z=0),Scale3D=(X=1,Y=1,Z=1))))
    Weight=1
    ItemToModify.0.Transform.Rotation=&RigVMFunction_MathQuaternionFromEuler(Euler=(X=0,Y=0,Z=0), Euler.Y=&RigVMFunction_MathDoubleMul(B=-1, A=@Add.Result).Result).Result
    ExecutePin=@SequenceAggregate.D
   )
   SequenceAggregate_2 = &RigVMFunction_Sequence(ExecuteContext=())
   &RigUnit_ParentConstraint(
    AdvancedSettings=(RotationOrderForFilter=XZY)
    Child=(Type=Bone,Name="dyn_mstr_tassel_01")
    Parents="((Item=(Type=Bone,Name=\"dfrm_thigh_r\"),Weight=0.0139),(Item=(Type=Bone,Name=\"dyn_front_skirt_01_l\"),Weight=0.0402),(Item=(Type=Bone,Name=\"dyn_front_skirt_03_l\"),Weight=0.0666),(Item=(Type=Bone,Name=\"dyn_front_skirt_02_l\"),Weight=0.0733),(Item=(Type=Bone,Na ...(truncated)"
    Weight=1
    ExecutePin=@SequenceAggregate_2.A
   )
   > &RigUnit_ParentConstraint(
    AdvancedSettings=(RotationOrderForFilter=XZY)
    Child=(Type=Bone,Name="dyn_mstr_tassel_02")
    Parents="((Item=(Type=Bone,Name=\"dyn_side_skirt_03_l\"),Weight=0.0102),(Item=(Type=Bone,Name=\"dyn_side_skirt_04_l\"),Weight=0.012),(Item=(Type=Bone,Name=\"dyn_side_skirt_02_l\"),Weight=0.0176),(Item=(Type=Bone,Name=\"dfrm_thigh_l\"),Weight=0.024),(Item=(Type=Bone,Name=\"d ...(truncated)"
    Weight=1
   )
   > &RigUnit_ParentConstraint(
    AdvancedSettings=(RotationOrderForFilter=XZY)
    Child=(Type=Bone,Name="dyn_mstr_tassel_03")
    Parents="((Item=(Type=Bone,Name=\"dyn_side_skirt_02_l\"),Weight=0.0269),(Item=(Type=Bone,Name=\"dyn_side_skirt_03_l\"),Weight=0.036),(Item=(Type=Bone,Name=\"dyn_front_skirt_01_l\"),Weight=0.0596),(Item=(Type=Bone,Name=\"dyn_side_skirt_04_l\"),Weight=0.0679),(Item=(Type=Bon ...(truncated)"
    Weight=1
   )
   > &RigUnit_ParentConstraint(
    AdvancedSettings=(RotationOrderForFilter=XZY)
    Child=(Type=Bone,Name="dyn_mstr_tassel_04")
    Parents="((Item=(Type=Bone,Name=\"dyn_back_side_skirt_04_l\"),Weight=0.0206),(Item=(Type=Bone,Name=\"dyn_front_skirt_01_l\"),Weight=0.025),(Item=(Type=Bone,Name=\"dyn_mstr_inner_front_skirt_01_l\"),Weight=0.0267),(Item=(Type=Bone,Name=\"dyn_side_skirt_03_l\"),Weight=0.0549 ...(truncated)"
    Weight=1
   )
   > &RigUnit_ParentConstraint(
    AdvancedSettings=(RotationOrderForFilter=XZY)
    Child=(Type=Bone,Name="dyn_mstr_tassel_05")
    Parents=((Item=(Type=Bone,Name="dyn_back_side_skirt_04_l"),Weight=0.0383),(Item=(Type=Bone,Name="dyn_side_skirt_03_l"),Weight=0.0741),(Item=(Type=Bone,Name="dyn_front_skirt_03_l"),Weight=0.3266),(Item=(Type=Bone,Name="dyn_side_skirt_04_l"),Weight=0.561))
    Weight=1
   )
   &RigUnit_ParentConstraint(
    AdvancedSettings=(RotationOrderForFilter=XZY)
    Child=(Type=Bone,Name="dyn_mstr_tassel_06")
    Parents="((Item=(Type=Bone,Name=\"dyn_back_side_skirt_03_l\"),Weight=0.0144),(Item=(Type=Bone,Name=\"dyn_back_skirt_03_l\"),Weight=0.0161),(Item=(Type=Bone,Name=\"dyn_side_skirt_03_l\"),Weight=0.081),(Item=(Type=Bone,Name=\"dyn_back_side_skirt_04_l\"),Weight=0.319),(Item=( ...(truncated)"
    Weight=1
    ExecutePin=@SequenceAggregate_2.B
   )
   > &RigUnit_ParentConstraint(
    AdvancedSettings=(RotationOrderForFilter=XZY)
    Child=(Type=Bone,Name="dyn_mstr_tassel_07")
    Parents="((Item=(Type=Bone,Name=\"dyn_back_skirt_02_l\"),Weight=0.015),(Item=(Type=Bone,Name=\"dyn_side_skirt_03_l\"),Weight=0.089),(Item=(Type=Bone,Name=\"dyn_back_side_skirt_03_l\"),Weight=0.1491),(Item=(Type=Bone,Name=\"dyn_back_skirt_03_l\"),Weight=0.161),(Item=(Type=B ...(truncated)"
    Weight=1
   )
   > &RigUnit_ParentConstraint(
    AdvancedSettings=(RotationOrderForFilter=XZY)
    Child=(Type=Bone,Name="dyn_mstr_tassel_08")
    Parents="((Item=(Type=Bone,Name=\"dyn_back_skirt_02_r\"),Weight=0.0774),(Item=(Type=Bone,Name=\"dyn_back_side_skirt_03_l\"),Weight=0.1212),(Item=(Type=Bone,Name=\"dyn_back_side_skirt_04_l\"),Weight=0.1979),(Item=(Type=Bone,Name=\"dyn_back_skirt_02_l\"),Weight=0.2487),(Item ...(truncated)"
    Weight=1
   )
   > &RigUnit_ParentConstraint(
    AdvancedSettings=(RotationOrderForFilter=XZY)
    Child=(Type=Bone,Name="dyn_mstr_tassel_09")
    Parents=((Item=(Type=Bone,Name="dyn_back_skirt_02_r"),Weight=0.323),(Item=(Type=Bone,Name="dyn_back_skirt_02_l"),Weight=0.3374),(Item=(Type=Bone,Name="dyn_back_skirt_03_l"),Weight=0.3396))
    Weight=1
   )
   > &RigUnit_ParentConstraint(
    AdvancedSettings=(RotationOrderForFilter=XZY)
    Child=(Type=Bone,Name="dyn_mstr_tassel_10")
    Parents=((Item=(Type=Bone,Name="dfrm_thigh_r"),Weight=0.0292),(Item=(Type=Bone,Name="dyn_back_skirt_02_l"),Weight=0.1393),(Item=(Type=Bone,Name="dyn_back_skirt_03_l"),Weight=0.1914),(Item=(Type=Bone,Name="dyn_back_skirt_02_r"),Weight=0.6402))
    Weight=1
   )
   > &RigUnit_ParentConstraint(
    AdvancedSettings=(RotationOrderForFilter=XZY)
    Child=(Type=Bone,Name="dyn_mstr_tassel_11")
    Parents="((Item=(Type=Bone,Name=\"dyn_back_skirt_01_r\"),Weight=0.0111),(Item=(Type=Bone,Name=\"dyn_back_skirt_02_l\"),Weight=0.0553),(Item=(Type=Bone,Name=\"dyn_back_skirt_03_l\"),Weight=0.0641),(Item=(Type=Bone,Name=\"dfrm_thigh_r\"),Weight=0.2505),(Item=(Type=Bone,Name= ...(truncated)"
    Weight=1
   )
   > &RigUnit_ParentConstraint(
    AdvancedSettings=(RotationOrderForFilter=XZY)
    Child=(Type=Bone,Name="dyn_mstr_tassel_12")
    Parents=((Item=(Type=Bone,Name="dyn_back_skirt_02_r"),Weight=0.0764),(Item=(Type=Bone,Name="dfrm_thigh_r"),Weight=0.9236))
    Weight=1
   )
   RigVMDispatchNode(BoneName=root, Default=0, Name=deformRig_l_thigh_up_alpha) {template="DISPATCH_RigDispatch_GetAnimAttribute(in Name,in BoneName,in Default,out Value,out Found)"}
   DISPATCH_RigDispatch_GetAnimAttribute = RigVMDispatchNode(BoneName=root, Default=0, Name=body_ctrlrig_left_mvmnt_alpha) {template="DISPATCH_RigDispatch_GetAnimAttribute(in Name,in BoneName,in Default,out Value,out Found)"}
   AlphaInterp = &RigVMFunction_AlphaInterp(InterpSpeedDecreasing=10, InterpSpeedIncreasing=10, Scale=1, bClampResult=False, Value=@DISPATCH_RigDispatch_GetAnimAttribute.Value)
   &RigUnit_ModifyTransforms(
    ExecutePin=()
    ItemToModify=((Item=(Type=Bone,Name="dyn_prnt_chain_sim"),Transform=(Rotation=(X=0,Y=0,Z=0,W=1),Translation=(X=0,Y=0,Z=0),Scale3D=(X=1,Y=1,Z=1))))
    Weight=@AlphaInterp.Result
    ItemToModify.0.Transform.Rotation=&RigVMFunction_MathQuaternionFromEuler(Euler=(X=0,Y=0,Z=-20)).Result
   )
   RigVMRerouteNode(Value=@SequenceAggregate.G)
  ]}
 RigVMFunctionLibrary {nodes [
   Body Additions = RigVMCollapseNode(ExecuteContext=())
   Chain IK = RigVMCollapseNode(ExecuteContext=())
  ]}
 RigVMModel {nodes [
   RigUnit_BeginExecution = &RigUnit_BeginExecution()
   FN_DeformRig_FNC = RigVMFunctionReferenceNode(
    DeformRig_ Curve="(deformRig_l_upperarm_up_alpha_72_#1=False,deformRig_r_upperarm_up_alpha_73_#2=False,deformRig_l_upperarm_fr_alpha_74_#3=False,deformRig_r_upperarm_fr_alpha_75_CB51B8 ...(truncated)"
    DeformRig_Bool="(upperarm_1_#4=True,elbow_3_#5=True,wrist_7_#6=True,groin_8_#7=True,glute_17_#8=True,knee_9_40C408324CA261 ...(truncated)"
    DeformRig_Debug=(upperarm_debug_18_#4=False,wrist_debug_20_#6=False,lat_debug_26_#9=False,groinGlute_debug_27_#7=False)
    DeformRig_Value="(upperarm_up_18_#10=(Rotation=(X=0.0,Y=-0.0,Z=0.0,W=1.0),Translation=(X=0.0,Y=0.0,Z=8.0),Scale3D=(X=0.70,Y=1.0,Z=1.0)),upperarm_fwd_fr_19_#11=(Rotation=(X=0.0,Y=-0.0,Z=0.0,W=1.0),Translation=(X=0.0, ...(truncated)"
    ExecuteContext=@RigUnit_BeginExecution.ExecutePin
   )
   > Body Additions = RigVMFunctionReferenceNode()
   > Chain IK = RigVMFunctionReferenceNode()
  ]}
 RigVMModel Rig {nodes [
   n1 = &RigUnit_PrepareForExecution()
   SequenceAggregate = RigVMAggregateNode(ExecuteContext=@n1.ExecutePin)
   n2 = &RigUnit_ItemArray(Items="((Type=Control,Name=\"dyn_back_chain_01_ctrl\"),(Type=Control,Name=\"dyn_back_chain_04_ctrl\"),(Type=Control,Name=\"dyn_back_chain_07_ctrl\"),(Type=Control,Name=\"dyn_back_chain_010_ctrl\"),(Type=Control,Name=\"dyn_back_chain_013_ctrl\"),(Type=Control,Name=\"dyn_back ...(truncated)")
   RigVMVariableNode(Value=@n2.Items, ExecuteContext=@SequenceAggregate.A) {variable=back_chain_ctrls}
   RigUnit_ItemArray = &RigUnit_ItemArray(Items="((Type=Control,Name=\"dyn_front_chain_01_ctrl\"),(Type=Control,Name=\"dyn_front_chain_04_ctrl\"),(Type=Control,Name=\"dyn_front_chain_07_ctrl\"),(Type=Control,Name=\"dyn_front_chain_010_ctrl\"),(Type=Control,Name=\"dyn_front_chain_013_ctrl\"),(Type=Control,Name=\"dyn ...(truncated)")
   > RigVMVariableNode(Value=@RigUnit_ItemArray.Items) {variable=front_chain_ctrls}
   > DISPATCH_RigVMDispatch_ArrayUnion = RigVMDispatchNode(Array=@RigUnit_ItemArray.Items, Other=@n2.Items) {template="DISPATCH_RigVMDispatch_ArrayUnion(io Array,in Other)"}
   > DISPATCH_RigVMDispatch_ArrayAppend = RigVMDispatchNode(Other=((Type=Control,Name="dyn_main_chain_ctrl")), Array=@DISPATCH_RigVMDispatch_ArrayUnion.Array) {template="DISPATCH_RigVMDispatch_ArrayAppend(io Array,in Other)"}
   > RigVMVariableNode(Value=@DISPATCH_RigVMDispatch_ArrayAppend.Array) {variable=chain_ctrls}
   &RigUnit_ModifyTransforms(
    ItemToModify="((Item=(Type=Control,Name=\"dyn_front_chain_04_ctrl\"),Transform=(Rotation=(X=0,Y=0,Z=0,W=1),Translation=(X=0,Y=-0.5,Z=0),Scale3D=(X=1,Y=1,Z=1))),(Item=(Type=Control,Name=\"dyn_front_chain_07_ctrl\"),Transform=(Rotation=(X=0,Y=0,Z=0,W=1),Translation=(X=0,Y=-0. ...(truncated)"
    Weight=1
    ExecutePin=@SequenceAggregate.B
   )
   > &RigUnit_ModifyTransforms(
    ItemToModify="((Item=(Type=Control,Name=\"dyn_back_chain_07_ctrl\"),Transform=(Rotation=(X=0,Y=0,Z=0,W=1),Translation=(X=0,Y=-0.5,Z=0),Scale3D=(X=1,Y=1,Z=1))),(Item=(Type=Control,Name=\"dyn_back_chain_010_ctrl\"),Transform=(Rotation=(X=0,Y=0,Z=0,W=1),Translation=(X=0,Y=-0.5 ...(truncated)"
    Weight=1
   )
   > RigVMVariableNode(Value=/Script/ControlRig.RigUnit_HierarchyGetPoseItemArray(ElementType=All, ItemsToGet=RigVMVariableNode() {variable=chain_ctrls}.Value) {is=struct RigVMUnitNode}.Pose) {variable=chain_ctrls_bindpose_pose}
   RigVMFunction_MathQuaternionFromEuler = &RigVMFunction_MathQuaternionFromEuler(Euler=(X=0,Y=-45,Z=15))
   &RigUnit_ModifyTransforms(
    ItemToModify="((Item=(Type=Control,Name=\"dyn_front_chain_07_ctrl\"),Transform=(Rotation=(X=0,Y=0,Z=0,W=1),Translation=(X=2,Y=0,Z=-4),Scale3D=(X=1,Y=1,Z=1))),(Item=(Type=Control,Name=\"dyn_main_chain_ctrl\"),Transform=(Rotation=(X=-0.05,Y=0.3794,Z=0.1206,W=0.916),Translatio ...(truncated)"
    Weight=1
    ItemToModify.1.Transform.Rotation=@RigVMFunction_MathQuaternionFromEuler.Result
    ExecutePin=@SequenceAggregate.C
   )
   > &RigUnit_ModifyTransforms(
    ItemToModify="((Item=(Type=Control,Name=\"dyn_back_chain_010_ctrl\"),Transform=(Rotation=(X=0,Y=0,Z=0,W=1),Translation=(X=0,Y=0,Z=3),Scale3D=(X=1,Y=1,Z=1))),(Item=(Type=Control,Name=\"dyn_back_chain_016_ctrl\"),Transform=(Rotation=(X=0,Y=0,Z=0,W=1),Translation=(X=0,Y=-2.25, ...(truncated)"
    Weight=1
    ItemToModify.0.Transform.Rotation=&RigVMFunction_MathQuaternionFromEuler(Euler=(X=0,Y=-10,Z=0)).Result
   )
   > RigVMVariableNode(Value=&RigUnit_HierarchyGetPoseItemArray(ElementType=All, ItemsToGet=RigVMVariableNode() {variable=chain_ctrls}.Value).Pose) {variable=chain_ctrls_thigh_l_up_pose}
  ]}
 SequenceAggregate_1_ContainedGraph {nodes [
   n1 = RigVMFunctionEntryNode()
   SequenceAggregate = &RigVMFunction_Sequence(ExecuteContext=@n1.ExecuteContext)
   SequenceAggregate_1 = &RigVMFunction_Sequence(ExecuteContext=@SequenceAggregate.B)
   RigVMFunctionReturnNode(A=@SequenceAggregate.A, B=@SequenceAggregate_1.A, C=@SequenceAggregate_1.B)
  ]}
 SequenceAggregate_2_1_ContainedGraph {nodes [
   n1 = RigVMFunctionEntryNode()
   SequenceAggregate_2 = &RigVMFunction_Sequence(ExecuteContext=@n1.ExecuteContext)
   SequenceAggregate_2_1 = &RigVMFunction_Sequence(ExecuteContext=@SequenceAggregate_2.B)
   SequenceAggregate_2_2 = &RigVMFunction_Sequence(ExecuteContext=@SequenceAggregate_2_1.B)
   SequenceAggregate_2_3 = &RigVMFunction_Sequence(ExecuteContext=@SequenceAggregate_2_2.B)
   RigVMFunctionReturnNode(A=@SequenceAggregate_2.A, B=@SequenceAggregate_2_1.A, C=@SequenceAggregate_2_2.A, D=@SequenceAggregate_2_3.A, E=@SequenceAggregate_2_3.B)
  ]}
 SequenceAggregate_3_1_1_ContainedGraph {nodes [
   n1 = RigVMFunctionEntryNode()
   SequenceAggregate_3_1 = &RigVMFunction_Sequence(ExecuteContext=@n1.ExecuteContext)
   SequenceAggregate_3_1_1 = &RigVMFunction_Sequence(ExecuteContext=@SequenceAggregate_3_1.B)
   RigVMFunctionReturnNode(A=@SequenceAggregate_3_1.A, B=@SequenceAggregate_3_1_1.A, C=@SequenceAggregate_3_1_1.B)
  ]}
 SequenceAggregate_5_1_ContainedGraph {nodes [
   n1 = RigVMFunctionEntryNode()
   SequenceAggregate_5 = &RigVMFunction_Sequence(ExecuteContext=@n1.ExecuteContext)
   SequenceAggregate_5_1 = &RigVMFunction_Sequence(ExecuteContext=@SequenceAggregate_5.B)
   SequenceAggregate_5_2 = &RigVMFunction_Sequence(ExecuteContext=@SequenceAggregate_5_1.B)
   RigVMFunctionReturnNode(A=@SequenceAggregate_5.A, B=@SequenceAggregate_5_1.A, C=@SequenceAggregate_5_2.A, D=@SequenceAggregate_5_2.B)
  ]}
]
hierarchy [
 root {loc=<0 0 0>, rot=<0 0 0 1>}
 pelvis {parent=root, loc=<0 -0.5301 98.0571>, rot=<0.0127 0.707 -0.0127 -0.707>}
 spine_01 {parent=pelvis, loc=<6.1208 -0.9328 0>, rot=<0 0 0.034 -0.9994>}
 spine_02 {parent=spine_01, loc=<6.5996 -0.2095 0>, rot=<0 0 0.0238 -0.9997>}
 spine_03 {parent=spine_02, loc=<6.4914 -0.1237 0>, rot=<0 0 0.0149 -0.9999>}
 spine_04 {parent=spine_03, loc=<6.9598 0.0794 0>, rot=<0 0 -0.0132 -0.9999>}
 spine_05 {parent=spine_04, loc=<6.9987 0 0>, rot=<0 0 -0.0015 -1>}
 attach_cape {parent=spine_05, loc=<9.3216 -8.4501 0>, rot=<0.0616 -0.7044 -0.0616 -0.7044>}
 neck_01 {parent=spine_05, loc=<11.3138 0.1529 0>, rot=<0 0 -0.2003 -0.9797>}
 neck_02 {parent=neck_01, loc=<4.5627 0 0>, rot=<0 0 0.0504 -0.9987>}
 head {parent=neck_02, loc=<5.3914 0 0>, rot=<0 0 0.115 -0.9934>}
 attach_backpack {parent=spine_05, loc=<0 -8.35 0>, rot=<-0.0093 -0.707 0.0093 -0.707>}
 clavicle_l {parent=spine_05, loc=<5.4831 -0.1053 -3.2946>, rot=<0.0465 -0.7289 -0.009 -0.6829>}
 upperarm_l {parent=clavicle_l, loc=<9.8658 -1.2136 -0.4105>, rot=<0.0164 -0.3245 -0.0269 -0.9454>}
 lowerarm_l {parent=upperarm_l, loc=<26.465 -0.8345 0.259>, rot=<0.0036 0.0161 -0.2298 -0.9731>}
 hand_l {parent=lowerarm_l, loc=<22.7355 -0.3203 -0.6105>, rot=<-0.6784 -0.0223 0.0735 -0.7307>}
 thumb_01_l {parent=hand_l, loc=<1.3118 -0.3191 -2.459>, rot=<0.6546 -0.4042 0.0664 -0.6354>}
 thumb_02_l {parent=thumb_01_l, loc=<5.9336 -0.0448 0.0401>, rot=<-0.0037 0.0213 0.298 -0.9543>}
 thumb_03_l {parent=thumb_02_l, loc=<2.2223 0.3682 -0.0907>, rot=<0.0118 -0.013 0.0066 -0.9998>}
 index_metacarpal_l {parent=hand_l, loc=<1.6535 0.0183 -2.96>, rot=<0.0158 -0.0124 -0.0173 -0.9996>}
 index_01_l {parent=index_metacarpal_l, loc=<8.3008 0.0357 -0.0524>, rot=<-0.0044 0.0485 0.0924 -0.9945>}
 index_02_l {parent=index_01_l, loc=<3.3587 -0.3004 0.0452>, rot=<-0.002 0.0027 0.3225 -0.9466>}
 index_03_l {parent=index_02_l, loc=<2.5578 -0.123 -0.0252>, rot=<-0.002 0.0067 0.2558 -0.9667>}
 middle_metacarpal_l {parent=hand_l, loc=<2.1856 -0.1725 -0.5984>, rot=<0.006 0.0168 -0.0329 -0.9993>}
 middle_01_l {parent=middle_metacarpal_l, loc=<7.3275 -0.2447 0.0129>, rot=<-0.0208 0.0042 0.0858 -0.9961>}
 middle_02_l {parent=middle_01_l, loc=<4.0123 0.1401 0.1175>, rot=<-0.0218 0.0258 0.3512 -0.9357>}
 middle_03_l {parent=middle_02_l, loc=<2.6077 0.0534 0.1373>, rot=<0.0082 0.0188 0.1708 -0.9851>}
 ring_metacarpal_l {parent=hand_l, loc=<2.2074 -0.3625 1.3638>, rot=<-0.0284 0.0722 -0.0458 -0.9959>}
 ring_01_l {parent=ring_metacarpal_l, loc=<7.0295 -0.6197 -0.1446>, rot=<-0.0489 -0.0252 0.1892 -0.9804>}
 ring_02_l {parent=ring_01_l, loc=<3.6174 0.4647 -0.1027>, rot=<0 0 0.2806 -0.9598>}
 ring_03_l {parent=ring_02_l, loc=<2.4152 0.114 0.0199>, rot=<0 0 0.1189 -0.9929>}
 pinky_metacarpal_l {parent=hand_l, loc=<1.7616 -0.9753 3.3814>, rot=<-0.0397 0.0814 -0.069 -0.9935>}
 pinky_01_l {parent=pinky_metacarpal_l, loc=<7.1709 -0.4831 -0.1094>, rot=<-0.0654 -0.0035 0.2242 -0.9723>}
 pinky_02_l {parent=pinky_01_l, loc=<2.8274 0.0893 -0.1043>, rot=<0.0037 0.0011 0.2854 -0.9584>}
 pinky_03_l {parent=pinky_02_l, loc=<2.2046 0.0307 -0.0078>, rot=<-0.0002 0.0098 0.1445 -0.9895>}
 deform_wrist_dn_base_l {parent=hand_l, loc=<0 -1.5326 0>, rot=<0 0 0 1>}
 deform_wrist_dn_l {parent=deform_wrist_dn_base_l, loc=<0 0 0>, rot=<0 0 0 1>}
 deform_wrist_up_base_l {parent=hand_l, loc=<0 2.3009 0>, rot=<0 0 0 1>}
 deform_wrist_up_l {parent=deform_wrist_up_base_l, loc=<0 0 0>, rot=<0 0 0 1>}
 lowerarm_twist_02_l {parent=lowerarm_l, loc=<10.6707 0 0>, rot=<0 0 0 1>}
 lowerarm_twist_01_l {parent=lowerarm_l, loc=<18.316 0 0>, rot=<0 0 0 1>}
 deform_elbow_in_base_l {parent=lowerarm_l, loc=<0.1834 3.6062 -0.2572>, rot=<0 0 0 1>}
 deform_elbow_in_l {parent=deform_elbow_in_base_l, loc=<0 0 0>, rot=<0 0 0 1>}
 deform_elbow_out_base_l {parent=lowerarm_l, loc=<0 -4 0>, rot=<0 0 0 1>}
 deform_elbow_out_l {parent=deform_elbow_out_base_l, loc=<0 0 0>, rot=<0 0 0 1>}
 upperarm_twist_01_l {parent=upperarm_l, loc=<7.2135 -0.1134 0>, rot=<0 0 0 1>}
 deform_upperarm_bk_base_l {parent=upperarm_twist_01_l, loc=<-6.4053 -4.3746 0.314>, rot=<0 0 0 1>}
 deform_upperarm_bk_l {parent=deform_upperarm_bk_base_l, loc=<0 0 0>, rot=<0 0 0 1>}
 deform_upperarm_fr_base_l {parent=upperarm_twist_01_l, loc=<-5.6312 5.092 0.3354>, rot=<0 0 0 1>}
 deform_upperarm_fr_l {parent=deform_upperarm_fr_base_l, loc=<0 0 0>, rot=<0 0 0 1>}
 deform_upperarm_up_base_l {parent=upperarm_twist_01_l, loc=<-8.891 0.4825 3.9517>, rot=<0 0 0 1>}
 deform_upperarm_up_l {parent=deform_upperarm_up_base_l, loc=<0 0 0>, rot=<0 0 0 1>}
 upperarm_twist_02_l {parent=upperarm_l, loc=<19.7998 -0.1134 0>, rot=<0 0 0 1>}
 clavicle_r {parent=spine_05, loc=<5.4831 -0.1053 3.2946>, rot=<0.7289 0.0465 -0.6829 0.009>}
 upperarm_r {parent=clavicle_r, loc=<-9.8658 1.2136 0.4105>, rot=<0.0164 -0.3245 -0.0269 -0.9454>}
 lowerarm_r {parent=upperarm_r, loc=<-26.465 0.8345 -0.259>, rot=<0.0036 0.0161 -0.2298 -0.9731>}
 hand_r {parent=lowerarm_r, loc=<-22.7355 0.3203 0.6106>, rot=<-0.6784 -0.0223 0.0735 -0.7307>}
 thumb_01_r {parent=hand_r, loc=<-1.3118 0.3191 2.459>, rot=<0.6546 -0.4042 0.0664 -0.6354>}
 thumb_02_r {parent=thumb_01_r, loc=<-5.9336 0.0448 -0.0401>, rot=<-0.0037 0.0213 0.298 -0.9543>}
 thumb_03_r {parent=thumb_02_r, loc=<-2.2223 -0.3682 0.0907>, rot=<0.0118 -0.013 0.0066 -0.9998>}
 index_metacarpal_r {parent=hand_r, loc=<-1.6535 -0.0183 2.96>, rot=<0.0158 -0.0124 -0.0173 -0.9996>}
 index_01_r {parent=index_metacarpal_r, loc=<-8.3008 -0.0357 0.0524>, rot=<-0.0044 0.0485 0.0924 -0.9945>}
 index_02_r {parent=index_01_r, loc=<-3.3587 0.3004 -0.0452>, rot=<-0.002 0.0027 0.3225 -0.9466>}
 index_03_r {parent=index_02_r, loc=<-2.5578 0.123 0.0252>, rot=<-0.002 0.0067 0.2558 -0.9667>}
 middle_metacarpal_r {parent=hand_r, loc=<-2.1856 0.1725 0.5984>, rot=<0.006 0.0168 -0.0329 -0.9993>}
 middle_01_r {parent=middle_metacarpal_r, loc=<-7.3275 0.2447 -0.0129>, rot=<-0.0208 0.0042 0.0858 -0.9961>}
 middle_02_r {parent=middle_01_r, loc=<-4.0123 -0.1401 -0.1175>, rot=<-0.0218 0.0258 0.3512 -0.9357>}
 middle_03_r {parent=middle_02_r, loc=<-2.6077 -0.0534 -0.1373>, rot=<0.0082 0.0188 0.1708 -0.9851>}
 ring_metacarpal_r {parent=hand_r, loc=<-2.2074 0.3625 -1.3638>, rot=<-0.0284 0.0722 -0.0458 -0.9959>}
 ring_01_r {parent=ring_metacarpal_r, loc=<-7.0295 0.6197 0.1446>, rot=<-0.0489 -0.0252 0.1892 -0.9804>}
 ring_02_r {parent=ring_01_r, loc=<-3.6174 -0.4647 0.1027>, rot=<0 0 0.2806 -0.9598>}
 ring_03_r {parent=ring_02_r, loc=<-2.4152 -0.114 -0.0199>, rot=<0 0 0.1189 -0.9929>}
 pinky_metacarpal_r {parent=hand_r, loc=<-1.7616 0.9753 -3.3814>, rot=<-0.0397 0.0814 -0.069 -0.9935>}
 pinky_01_r {parent=pinky_metacarpal_r, loc=<-7.1709 0.4831 0.1094>, rot=<-0.0654 -0.0035 0.2242 -0.9723>}
 pinky_02_r {parent=pinky_01_r, loc=<-2.8274 -0.0893 0.1043>, rot=<0.0037 0.0011 0.2854 -0.9584>}
 pinky_03_r {parent=pinky_02_r, loc=<-2.2046 -0.0307 0.0078>, rot=<-0.0002 0.0098 0.1445 -0.9895>}
 deform_wrist_dn_base_r {parent=hand_r, loc=<-0.0001 1.5328 0>, rot=<0 0 0 1>}
 deform_wrist_dn_r {parent=deform_wrist_dn_base_r, loc=<0 0 0>, rot=<0 0 0 1>}
 deform_wrist_up_base_r {parent=hand_r, loc=<-0.0002 -2.3007 -0.0001>, rot=<0 0 0 1>}
 deform_wrist_up_r {parent=deform_wrist_up_base_r, loc=<0 0 0>, rot=<0 0 0 1>}
 lowerarm_twist_02_r {parent=lowerarm_r, loc=<-10.6707 0 0>, rot=<0 0 0 -1>}
 lowerarm_twist_01_r {parent=lowerarm_r, loc=<-18.316 0 0>, rot=<0 0 0 -1>}
 dfrm_mstr_sleeve_01_r {parent=lowerarm_twist_01_r, loc=<-1.4396 -1.7091 -5.856>, rot=<0.0287 -0.7342 -0.6771 -0.0402>}
 dfrm_sleeve_01_r {parent=dfrm_mstr_sleeve_01_r, loc=<0 0 0>, rot=<0 0 0 1>}
 dfrm_mstr_sleeve_02_r {parent=lowerarm_twist_01_r, loc=<-1.1923 -5.758 -2.5786>, rot=<0.0651 -0.9601 -0.2719 -0.0014>}
 dfrm_sleeve_02_r {parent=dfrm_mstr_sleeve_02_r, loc=<0 0 0>, rot=<0 0 0 1>}
 dfrm_mstr_sleeve_03_r {parent=lowerarm_twist_01_r, loc=<-0.2563 -5.7983 2.7793>, rot=<0.0951 -0.983 0.14 0.0711>}
 dfrm_sleeve_03_r {parent=dfrm_mstr_sleeve_03_r, loc=<0 0 0>, rot=<0 0 0 1>}
 dfrm_mstr_sleeve_04_r {parent=lowerarm_twist_01_r, loc=<0.191 -2.6677 7.8538>, rot=<0.0704 -0.9485 0.2773 0.1357>}
 dfrm_sleeve_04_r {parent=dfrm_mstr_sleeve_04_r, loc=<0 0 0>, rot=<0 0 0 1>}
 dfrm_mstr_sleeve_05_r {parent=lowerarm_twist_01_r, loc=<0.542 1.8769 9.7773>, rot=<0.0812 -0.6825 0.7078 0.163>}
 dfrm_sleeve_05_r {parent=dfrm_mstr_sleeve_05_r, loc=<0 0 0>, rot=<0 0 0 1>}
 dfrm_mstr_sleeve_06_r {parent=lowerarm_twist_01_r, loc=<0.7301 6.4311 6.3428>, rot=<0.0729 -0.1915 0.9716 0.1182>}
 dfrm_sleeve_06_r {parent=dfrm_mstr_sleeve_06_r, loc=<0 0 0>, rot=<0 0 0 1>}
 dfrm_mstr_sleeve_07_r {parent=lowerarm_twist_01_r, loc=<0.3708 7.063 0.3164>, rot=<0.0536 0.136 0.9857 0.0841>}
 dfrm_sleeve_07_r {parent=dfrm_mstr_sleeve_07_r, loc=<0 0 0>, rot=<0 0 0 1>}
 dfrm_mstr_sleeve_08_r {parent=lowerarm_twist_01_r, loc=<-0.771 3.6643 -4.3004>, rot=<-0.01 0.3935 0.9181 0.0466>}
 dfrm_sleeve_08_r {parent=dfrm_mstr_sleeve_08_r, loc=<0 0 0>, rot=<0 0 0 1>}
 deform_elbow_in_base_r {parent=lowerarm_r, loc=<-0.1834 -3.6062 0.2572>, rot=<0 0 0 -1>}
 deform_elbow_in_r {parent=deform_elbow_in_base_r, loc=<0 0 0>, rot=<0 0 0 1>}
 deform_elbow_out_base_r {parent=lowerarm_r, loc=<0 4 -0.0001>, rot=<0 0 0 -1>}
 deform_elbow_out_r {parent=deform_elbow_out_base_r, loc=<0 0 0>, rot=<0 0 0 1>}
 upperarm_twist_01_r {parent=upperarm_r, loc=<-7.2135 0.1134 0>, rot=<0 0 0 -1>}
 deform_upperarm_bk_base_r {parent=upperarm_twist_01_r, loc=<6.4053 4.3746 -0.314>, rot=<0 0 0 1>}
 deform_upperarm_bk_r {parent=deform_upperarm_bk_base_r, loc=<0 0 0>, rot=<0 0 0 1>}
 deform_upperarm_fr_base_r {parent=upperarm_twist_01_r, loc=<5.6312 -5.092 -0.3354>, rot=<0 0 0 1>}
 deform_upperarm_fr_r {parent=deform_upperarm_fr_base_r, loc=<0 0 0>, rot=<0 0 0 1>}
 deform_upperarm_up_base_r {parent=upperarm_twist_01_r, loc=<8.891 -0.4825 -3.9517>, rot=<0 0 0 1>}
 deform_upperarm_up_r {parent=deform_upperarm_up_base_r, loc=<0 0 0>, rot=<0 0 0 1>}
 upperarm_twist_02_r {parent=upperarm_r, loc=<-19.7998 0.1134 0>, rot=<0 0 0 -1>}
 dyn_base_shawl_B_r {parent=spine_05, loc=<1.9628 -6.9787 17.7406>, rot=<0.1389 -0.9903 0.0027 0.0014>}
 dyn_mstr_shawl_B_r {parent=dyn_base_shawl_B_r, loc=<0 0 0>, rot=<0 0 0 1>}
 dyn_shawl_B_01_r {parent=dyn_mstr_shawl_B_r, loc=<0 0 0>, rot=<0 0 0 1>}
 dyn_shawl_B_02_r {parent=dyn_shawl_B_01_r, loc=<5.7752 0 0>, rot=<0.0041 -0.0004 -0.0892 -0.996>}
 dyn_shawl_B_03_r {parent=dyn_shawl_B_02_r, loc=<5.7695 0 0>, rot=<0 0 -0.0297 -0.9996>}
 dyn_base_shawl_A_r {parent=spine_05, loc=<3.6452 -7.5271 13.572>, rot=<0.1232 -0.9924 0.0023 0.0014>}
 dyn_mstr_shawl_A_r {parent=dyn_base_shawl_A_r, loc=<0 0 0>, rot=<0 0 0 1>}
 dyn_shawl_A_01_r {parent=dyn_mstr_shawl_A_r, loc=<0 0 0>, rot=<0 0 0 1>}
 dyn_shawl_A_02_r {parent=dyn_shawl_A_01_r, loc=<6.5646 0 0>, rot=<-0.1085 -0.0136 -0.0594 -0.9922>}
 dyn_shawl_A_03_r {parent=dyn_shawl_A_02_r, loc=<6.5646 0 0>, rot=<0 0 -0.0241 -0.9997>}
 dyn_shawl_A_04_r {parent=dyn_shawl_A_03_r, loc=<6.5646 0 0>, rot=<-0.0391 0 0 -0.9992>}
 dyn_mstr_lapel_lace_l {parent=spine_05, loc=<7.9153 6.2679 -3.0655>, rot=<-0.0611 0.1805 0.955 0.2271>}
 dyn_lapel_lace_01_l {parent=dyn_mstr_lapel_lace_l, loc=<0 0 0>, rot=<0 0 0 1>}
 dyn_lapel_lace_02_l {parent=dyn_lapel_lace_01_l, loc=<5.8252 -0.0168 -0.0068>, rot=<0 0 0.0313 -0.9995>}
 dyn_mstr_lapel_lace_r {parent=spine_05, loc=<7.9152 6.2679 3.0655>, rot=<0.0619 -0.1806 0.9547 0.2284>}
 dyn_lapel_lace_01_r {parent=dyn_mstr_lapel_lace_r, loc=<0 0 0>, rot=<0 0 0 1>}
 dyn_lapel_lace_02_r {parent=dyn_lapel_lace_01_r, loc=<5.8252 0 0>, rot=<0 -0.0006 0.0299 -0.9996>}
 deform_pec_aim_base_l {parent=spine_04, loc=<3.2456 9.3309 -1.2237>, rot=<0.0603 -0.4033 0.3318 -0.8507>}
 deform_pec_aim_l {parent=deform_pec_aim_base_l, loc=<0 0 0>, rot=<0 0 0 1>}
 deform_pec_base_l {parent=deform_pec_aim_l, loc=<9.3665 0 0>, rot=<0 0 0 1>}
 deform_pec_l {parent=deform_pec_base_l, loc=<0 0 0>, rot=<0 0 0 1>}
 deform_pec_aim_base_r {parent=spine_04, loc=<3.2452 9.3309 1.2237>, rot=<-0.4033 -0.0603 0.8507 0.3318>}
 deform_pec_aim_r {parent=deform_pec_aim_base_r, loc=<0 0 0>, rot=<0 0 0 1>}
 deform_pec_base_r {parent=deform_pec_aim_r, loc=<-9.3667 -0.0003 -0.0004>, rot=<0 0 0 1>}
 deform_pec_r {parent=deform_pec_base_r, loc=<0 0 0>, rot=<0 0 0 1>}
 thigh_l {parent=pelvis, loc=<-1.0855 -0.2109 -8.7053>, rot=<0.0009 0.0355 0.0155 -0.9993>}
 calf_l {parent=thigh_l, loc=<-42.5284 -0.0019 -0.7297>, rot=<0.0006 -0.0161 -0.0391 -0.9991>}
 foot_l {parent=calf_l, loc=<-46.7401 -2.5014 -0.592>, rot=<0.0427 -0.0238 0.0353 -0.9982>}
 ball_l {parent=foot_l, loc=<-3.5658 10.8066 -0.1544>, rot=<0.0001 -0.0001 -0.7063 -0.7079>}
 calf_twist_01_l {parent=calf_l, loc=<-35.8483 -1.9197 -0.5794>, rot=<0.0237 0.0095 -0.0035 -0.9997>}
 dyn_mstr_lace_l {parent=calf_twist_01_l, loc=<-3.6316 4.8119 -0.4303>, rot=<-0.0393 -0.003 -0.6744 -0.7373>}
 dyn_lace_knot_l {parent=dyn_mstr_lace_l, loc=<0 0 0>, rot=<0 0 0 1>}
 dyn_lace_outer_l {parent=dyn_lace_knot_l, loc=<0 0 0>, rot=<-0.0847 -0.6972 -0.0859 -0.7067>}
 dyn_lace_inner_l {parent=dyn_lace_knot_l, loc=<0 0 0>, rot=<-0.0418 0.5609 0.0615 -0.8246>}
 dfrm_calf_l {parent=calf_l, loc=<0 0 0>, rot=<-0.0063 0.0002 0.9996 -0.0267>}
 deform_knee_in_base_l {parent=dfrm_calf_l, loc=<-0.5366 5.059 0.2355>, rot=<0 0 0 -1>}
 deform_knee_in_l {parent=deform_knee_in_base_l, loc=<0 0 0>, rot=<0 0 0 1>}
 deform_knee_out_base_l {parent=dfrm_calf_l, loc=<-0.2137 -3.9943 0>, rot=<-0.0063 0.0002 0.9996 0.0267>}
 deform_knee_out_l {parent=deform_knee_out_base_l, loc=<0 0 0>, rot=<0 0 0 1>}
 dfrm_thigh_l {parent=thigh_l, loc=<0 0 0>, rot=<-0.0086 0 1 0>}
 deform_groin_base_l {parent=dfrm_thigh_l, loc=<-0.476 -5.6776 -0.2568>, rot=<-0.0086 0 1 0>}
 deform_groin_l {parent=deform_groin_base_l, loc=<0 0 0>, rot=<0 0 0 1>}
 dfrm_aim_up_calf_l {parent=dfrm_thigh_l, loc=<42.681 0.0062 -3.9973>, rot=<0.0004 0.0183 -0.0657 -0.9977>}
 dyn_mstr_side_skirt_l {parent=dfrm_thigh_l, loc=<16.6137 4.2933 -9.0221>, rot=<-0.8002 -0.0305 0.0237 -0.5985>}
 dyn_side_skirt_01_l {parent=dyn_mstr_side_skirt_l, loc=<0 0 0>, rot=<0 0 0 1>}
 dyn_side_skirt_02_l {parent=dyn_side_skirt_01_l, loc=<14.304 0 0>, rot=<0 0 0.0078 -1>}
 dyn_side_skirt_03_l {parent=dyn_side_skirt_02_l, loc=<14.3121 0 0>, rot=<0.0353 0.0014 -0.0383 -0.9986>}
 dyn_side_skirt_04_l {parent=dyn_side_skirt_03_l, loc=<14.3109 0 0>, rot=<-0.0353 -0.0003 0.0077 -0.9993>}
 dyn_mstr_front_skirt_l {parent=dfrm_thigh_l, loc=<22.2527 -5.3385 -8.0779>, rot=<-0.4709 -0.0202 -0.0006 -0.8819>}
 dyn_front_skirt_01_l {parent=dyn_mstr_front_skirt_l, loc=<0 0 0>, rot=<0 0 0 1>}
 dyn_front_skirt_02_l {parent=dyn_front_skirt_01_l, loc=<11.4592 0 0>, rot=<0.0862 0.0014 -0.0167 -0.9961>}
 dyn_front_skirt_03_l {parent=dyn_front_skirt_02_l, loc=<11.4624 0 0>, rot=<0 -0.0034 0.0194 -0.9998>}
 dyn_mstr_tassel_01 {parent=dfrm_thigh_l, loc=<31.635 -10.4391 2.8545>, rot=<0.0004 0.0511 0.0001 -0.9987>}
 dyn_tassel_01 {parent=dyn_mstr_tassel_01, loc=<0 0 0>, rot=<0 0 0 1>}
 dyn_mstr_back_side_skirt_l {parent=dfrm_thigh_l, loc=<9.7811 10.8556 -3.99>, rot=<0.9336 0.0071 0.0041 0.3582>}
 dyn_back_side_skirt_01_l {parent=dyn_mstr_back_side_skirt_l, loc=<0 0 0>, rot=<0 0 0 1>}
 dyn_back_side_skirt_02_l {parent=dyn_back_side_skirt_01_l, loc=<12.8808 0 0>, rot=<-0.0664 0.0024 0.0355 -0.9972>}
 dyn_back_side_skirt_03_l {parent=dyn_back_side_skirt_02_l, loc=<12.8861 0 0>, rot=<-0.0973 -0.0066 -0.0281 -0.9948>}
 dyn_back_side_skirt_04_l {parent=dyn_back_side_skirt_03_l, loc=<12.8768 0 0>, rot=<-0.0501 0.0159 0.006 -0.9986>}
 dyn_mstr_inner_front_skirt_l {parent=dfrm_thigh_l, loc=<20.8127 -9.905 -0.316>, rot=<0.0003 0.0039 -0.0287 -0.9996>}
 dyn_mstr_inner_front_skirt_01_l {parent=dyn_mstr_inner_front_skirt_l, loc=<0 0 0>, rot=<0 0 0 1>}
 thigh_r {parent=pelvis, loc=<-1.0855 -0.2109 8.7053>, rot=<0.0355 -0.0009 0.9993 0.0155>}
 calf_r {parent=thigh_r, loc=<42.5284 0.0019 0.7297>, rot=<0.0006 -0.0161 -0.0391 -0.9991>}
 foot_r {parent=calf_r, loc=<46.7401 2.5014 0.592>, rot=<0.0427 -0.0238 0.0353 -0.9982>}
 ball_r {parent=foot_r, loc=<3.5658 -10.8066 0.1544>, rot=<0.0001 -0.0001 -0.7063 -0.7079>}
 calf_twist_02_r {parent=calf_r, loc=<19.1546 1.7943 0.2638>, rot=<0.0237 0.0095 -0.0035 -0.9997>}
 calf_twist_01_r {parent=calf_r, loc=<35.8483 1.9197 0.5794>, rot=<0.0237 0.0095 -0.0035 -0.9997>}
 dyn_mstr_race_r {parent=calf_twist_01_r, loc=<3.6316 -4.8119 0.4303>, rot=<0.003 -0.0393 0.7373 -0.6744>}
 dyn_lace_knot_r {parent=dyn_mstr_race_r, loc=<0 0 0>, rot=<0 0 0 1>}
 dyn_lace_outer_r {parent=dyn_lace_knot_r, loc=<0 0 0>, rot=<0.0847 0.6972 -0.0859 -0.7067>}
 dyn_lace_inner_r {parent=dyn_lace_knot_r, loc=<0 0 0>, rot=<0.0418 -0.5609 0.0615 -0.8246>}
 dfrm_calf_r {parent=calf_r, loc=<0 0 0>, rot=<0.0002 0.0063 -0.0267 -0.9996>}
 deform_knee_in_base_r {parent=dfrm_calf_r, loc=<-0.5365 5.059 -0.2355>, rot=<-0.0002 -0.0063 0.0267 -0.9996>}
 deform_knee_in_r {parent=deform_knee_in_base_r, loc=<0 0 0>, rot=<0 0 0 1>}
 deform_knee_out_base_r {parent=dfrm_calf_r, loc=<-0.2137 -3.9943 0>, rot=<-0.0002 -0.0063 0.0267 -0.9996>}
 deform_knee_out_r {parent=deform_knee_out_base_r, loc=<0 0 0>, rot=<0 0 0 1>}
 dfrm_thigh_r {parent=thigh_r, loc=<0 0 0>, rot=<0 0.0086 0 -1>}
 deform_groin_base_r {parent=dfrm_thigh_r, loc=<-0.476 -5.6776 0.2568>, rot=<0 -0.0086 0 -1>}
 deform_groin_r {parent=deform_groin_base_r, loc=<0 0 0>, rot=<0 0 0 1>}
 dfrm_aim_up_calf_r {parent=dfrm_thigh_r, loc=<42.6526 0.005 3.2229>, rot=<-0.0004 -0.0183 -0.0657 -0.9977>}
 deform_glute_base_l {parent=pelvis, loc=<-7.3466 -9.0213 -9.1245>, rot=<0.0172 0.031 -0.4477 -0.8935>}
 deform_glute_l {parent=deform_glute_base_l, loc=<0 0 0>, rot=<0 0 0 1>}
 deform_glute_base_r {parent=pelvis, loc=<-7.3464 -9.0213 9.1245>, rot=<0.031 -0.0172 0.8935 -0.4477>}
 deform_glute_r {parent=deform_glute_base_r, loc=<0 0 0>, rot=<0 0 0 1>}
 dfrm_aim_up_thigh_l {parent=pelvis, loc=<-0.7336 -0.2246 -12.6898>, rot=<-0.044 0.001 0.9989 0.0155>}
 dfrm_aim_up_thigh_r {parent=pelvis, loc=<-0.7797 -0.2232 12.6938>, rot=<0.0355 -0.0009 0.9993 0.0155>}
 dyn_mstr_clock_chain {parent=pelvis, loc=<1.6656 -5.1756 13.8558>, rot=<0.204 -0.7724 0.5826 0.1498>}
 dyn_clock_chain_01 {parent=dyn_mstr_clock_chain, loc=<0 0 0>, rot=<-0.0006 -0.0195 0.0048 -0.9998>}
 dyn_clock_chain_02 {parent=dyn_clock_chain_01, loc=<4.2546 0 0>, rot=<0.0157 -0.0049 -0.1363 -0.9905>}
 dyn_clock_chain_03 {parent=dyn_clock_chain_02, loc=<4.3884 0 0>, rot=<-0.0105 0.0159 0.1669 -0.9858>}
 dfrm_clock_second {parent=dyn_clock_chain_03, loc=<11.107 0.9852 -0.2928>, rot=<-0.1498 0.7012 -0.1828 -0.6727>}
 dfrm_clock_minute {parent=dyn_clock_chain_03, loc=<11.107 0.9852 -0.2928>, rot=<-0.1498 0.7012 -0.1828 -0.6727>}
 dfrm_clock_hour {parent=dyn_clock_chain_03, loc=<11.107 0.9852 -0.2928>, rot=<-0.1498 0.7012 -0.1828 -0.6727>}
 dyn_mstr_tassel_02 {parent=pelvis, loc=<-40.2164 10.531 -13.3712>, rot=<0.0071 -0.0001 0.9999 0.0156>}
 dyn_tassel_02 {parent=dyn_mstr_tassel_02, loc=<0 0 0>, rot=<0 0 0 1>}
 dyn_mstr_tassel_03 {parent=pelvis, loc=<-48.1263 8.7735 -18.4582>, rot=<0.0327 0.2477 0.9671 0.0481>}
 dyn_tassel_03 {parent=dyn_mstr_tassel_03, loc=<0 0 0>, rot=<0 0 0 1>}
 dyn_mstr_tassel_04 {parent=pelvis, loc=<-55.1089 6.6929 -22.253>, rot=<-0.0072 0.428 0.9027 0.0441>}
 dyn_tassel_04 {parent=dyn_mstr_tassel_04, loc=<0 0 0>, rot=<0 0 0 1>}
 dyn_mstr_tassel_05 {parent=pelvis, loc=<-63.5757 3.8252 -25.8522>, rot=<0.0186 0.4558 0.8898 0.0146>}
 dyn_tassel_05 {parent=dyn_mstr_tassel_05, loc=<0 0 0>, rot=<0 0 0 1>}
 dyn_mstr_tassel_06 {parent=pelvis, loc=<-64.354 -10.4654 -21.2957>, rot=<0.0047 -0.9855 -0.1658 -0.0347>}
 dyn_tassel_06 {parent=dyn_mstr_tassel_06, loc=<0 0 0>, rot=<0 0 0 1>}
 dyn_mstr_tassel_07 {parent=pelvis, loc=<-57.0686 -12.5928 -15.7691>, rot=<-0.0085 -0.9805 -0.195 -0.0219>}
 dyn_tassel_07 {parent=dyn_mstr_tassel_07, loc=<0 0 0>, rot=<0 0 0 1>}
 dyn_mstr_tassel_08 {parent=pelvis, loc=<-50.0689 -13.4466 -9.7045>, rot=<-0.0075 -0.9997 0.0007 -0.023>}
 dyn_tassel_08 {parent=dyn_mstr_tassel_08, loc=<0 0 0>, rot=<0 0 0 1>}
 dyn_mstr_tassel_09 {parent=pelvis, loc=<-43.4339 -13.4847 -4.1707>, rot=<-0.0196 -0.9997 -0.0002 0.0136>}
 dyn_tassel_09 {parent=dyn_mstr_tassel_09, loc=<0 0 0>, rot=<0 0 0 1>}
 dyn_mstr_tassel_10 {parent=pelvis, loc=<-36.7893 -13.5306 1.78>, rot=<-0.0228 -0.9997 -0.0001 0.0079>}
 dyn_tassel_10 {parent=dyn_mstr_tassel_10, loc=<0 0 0>, rot=<0 0 0 1>}
 dyn_mstr_tassel_11 {parent=pelvis, loc=<-30.4091 -13.1388 7.3028>, rot=<0.0074 -1 -0.0002 0.005>}
 dyn_tassel_11 {parent=dyn_mstr_tassel_11, loc=<0 0 0>, rot=<0 0 0 1>}
 dyn_mstr_tassel_12 {parent=pelvis, loc=<-22.7048 -11.8697 14.1697>, rot=<-0.0388 -0.9877 0.1511 -0.0054>}
 dyn_tassel_12 {parent=dyn_mstr_tassel_12, loc=<0 0 0>, rot=<0 0 0 1>}
 dyn_mstr_back_skirt_r {parent=pelvis, loc=<-8.0355 -13.8917 5.6775>, rot=<-0.0209 -0.9972 0.0724 -0.0015>}
 dyn_back_skirt_01_r {parent=dyn_mstr_back_skirt_r, loc=<0 0 0>, rot=<0 0 0 1>}
 dyn_back_skirt_02_r {parent=dyn_back_skirt_01_r, loc=<12.1984 0.0382 -0.0012>, rot=<-0.0001 0.0004 0.014 -0.9999>}
 dyn_mstr_back_skirt_l {parent=pelvis, loc=<-7.9537 -14.1947 -5.5959>, rot=<-0.0273 -0.9949 -0.0951 -0.0212>}
 dyn_back_skirt_01_l {parent=dyn_mstr_back_skirt_l, loc=<0 0 0>, rot=<0 0 0 1>}
 dyn_back_skirt_02_l {parent=dyn_back_skirt_01_l, loc=<12.8159 0.076 0.0364>, rot=<-0.0946 -0.0112 0.0305 -0.995>}
 dyn_back_skirt_03_l {parent=dyn_back_skirt_02_l, loc=<12.8055 -0.0271 0.0052>, rot=<0.0479 -0.0015 -0.0106 -0.9988>}
 dyn_mstr_front_chain {parent=pelvis, loc=<-5.5234 9.4791 -6.0965>, rot=<-0.1437 0.0537 0.9777 0.1432>}
 dyn_front_chain {parent=dyn_mstr_front_chain, loc=<0 0 0>, rot=<0 0 0 1>}
 dyn_front_chain_01 {parent=dyn_front_chain, loc=<2.516 0 0>, rot=<0.0164 -0.2153 -0.0702 -0.9739>}
 dyn_front_chain_02 {parent=dyn_front_chain_01, loc=<2.3312 0 0>, rot=<0.0009 -0.0261 -0.0333 -0.9991>}
 dyn_front_chain_03 {parent=dyn_front_chain_02, loc=<2.044 0 0>, rot=<-0.001 -0.0243 -0.0366 -0.999>}
 dyn_front_chain_04 {parent=dyn_front_chain_03, loc=<1.9214 0 0>, rot=<0.0009 -0.0229 -0.0366 -0.9991>}
 dyn_front_chain_05 {parent=dyn_front_chain_04, loc=<1.8101 0 0>, rot=<0.0004 -0.0244 -0.0484 -0.9985>}
 dyn_front_chain_06 {parent=dyn_front_chain_05, loc=<1.7185 0 0>, rot=<0.0005 -0.0234 -0.0447 -0.9987>}
 dyn_front_chain_07 {parent=dyn_front_chain_06, loc=<1.6398 0 0>, rot=<-0.0664 -0.048 -0.0571 -0.995>}
 dyn_front_chain_08 {parent=dyn_front_chain_07, loc=<1.9685 0 0>, rot=<0.0801 -0.0544 -0.0687 -0.9929>}
 dyn_front_chain_09 {parent=dyn_front_chain_08, loc=<2.0369 0 0>, rot=<-0.0267 -0.0327 -0.0601 -0.9973>}
 dyn_front_chain_010 {parent=dyn_front_chain_09, loc=<2.0103 0 0>, rot=<-0.0079 -0.0218 -0.0641 -0.9977>}
 dyn_front_chain_011 {parent=dyn_front_chain_010, loc=<1.9075 0 0>, rot=<0.0034 -0.0096 -0.0838 -0.9964>}
 dyn_front_chain_012 {parent=dyn_front_chain_011, loc=<1.7592 0 0>, rot=<0.0064 0.0057 -0.11 -0.9939>}
 dyn_front_chain_013 {parent=dyn_front_chain_012, loc=<1.6385 0 0>, rot=<0.0046 0.0009 -0.104 -0.9946>}
 dyn_front_chain_014 {parent=dyn_front_chain_013, loc=<1.2458 0 0>, rot=<0 0 0 1>}
 dyn_front_chain_01_ctrl_null {parent=dyn_front_chain, type=Null, loc=<2.516 0 0>, rot=<-0.0164 0.2153 0.0702 0.9739>}
 dyn_front_chain_01_ctrl {controlType=EULER_TRANSFORM, parent=dyn_front_chain_01_ctrl_null, type=Control, loc=<0 0 0>, rot=<0 0 0 1>}
 dyn_mstr_back_chain {parent=pelvis, loc=<4.33 -11.6928 0.1005>, rot=<0.2154 -0.9595 -0.0503 -0.1744>}
 dyn_back_chain {parent=dyn_mstr_back_chain, loc=<0 0 0>, rot=<0 0 0 1>}
 dyn_back_chain_01 {parent=dyn_back_chain, loc=<2.7569 0 0>, rot=<0.0492 0.1306 -0.0511 -0.9889>}
 dyn_back_chain_02 {parent=dyn_back_chain_01, loc=<2.2558 0 0>, rot=<0.0004 0.0151 -0.0208 -0.9997>}
 dyn_back_chain_03 {parent=dyn_back_chain_02, loc=<2.0586 0 0>, rot=<0.001 0.0157 -0.0198 -0.9997>}
 dyn_back_chain_04 {parent=dyn_back_chain_03, loc=<1.9841 0 0>, rot=<0.0038 0.017 -0.0212 -0.9996>}
 dyn_back_chain_05 {parent=dyn_back_chain_04, loc=<1.915 0 0>, rot=<0.0004 0.0183 -0.0227 -0.9996>}
 dyn_back_chain_06 {parent=dyn_back_chain_05, loc=<1.852 0 0>, rot=<-0.0017 0.0195 -0.0242 -0.9995>}
 dyn_back_chain_07 {parent=dyn_back_chain_06, loc=<1.7957 0 0>, rot=<-0.0019 0.0206 -0.0257 -0.9995>}
 dyn_back_chain_08 {parent=dyn_back_chain_07, loc=<1.7468 0 0>, rot=<-0.0006 0.0219 -0.0274 -0.9994>}
 dyn_back_chain_09 {parent=dyn_back_chain_08, loc=<1.7285 0 0>, rot=<0.0028 0.0339 -0.0335 -0.9989>}
 dyn_back_chain_010 {parent=dyn_back_chain_09, loc=<1.9869 0 0>, rot=<0 0.0351 -0.0343 -0.9988>}
 dyn_back_chain_011 {parent=dyn_back_chain_010, loc=<2.0352 0 0>, rot=<-0.0018 0.0272 -0.032 -0.9991>}
 dyn_back_chain_012 {parent=dyn_back_chain_011, loc=<2.039 0 0>, rot=<0.0081 0.0218 -0.0321 -0.9992>}
 dyn_back_chain_013 {parent=dyn_back_chain_012, loc=<1.993 0 0>, rot=<-0.0009 0.0179 -0.0351 -0.9992>}
 dyn_back_chain_014 {parent=dyn_back_chain_013, loc=<1.8951 0 0>, rot=<-0.0006 0.0147 -0.0421 -0.999>}
 dyn_back_chain_015 {parent=dyn_back_chain_014, loc=<1.7461 0 0>, rot=<-0.0002 0.0133 -0.0577 -0.9982>}
 dyn_back_chain_016 {parent=dyn_back_chain_015, loc=<1.6417 0 0>, rot=<-0.0035 0.0379 -0.0847 -0.9957>}
 dyn_back_chain_017 {parent=dyn_back_chain_016, loc=<1.996 0 0>, rot=<-0.0041 0.0415 -0.0898 -0.9951>}
 dyn_back_chain_018 {parent=dyn_back_chain_017, loc=<1.9706 0 0>, rot=<-0.0023 0.0282 -0.0862 -0.9959>}
 dyn_back_chain_019 {parent=dyn_back_chain_018, loc=<1.9144 0 0>, rot=<-0.0013 0.0157 -0.0826 -0.9965>}
 dyn_back_chain_020 {parent=dyn_back_chain_019, loc=<1.2927 0 0>, rot=<0 0 0 1>}
 dyn_back_chain_01_ctrl_null {parent=dyn_back_chain, type=Null, loc=<2.7569 0 0>, rot=<0.0492 0.1306 -0.0511 -0.9889>}
 dyn_back_chain_01_ctrl {controlType=EULER_TRANSFORM, parent=dyn_back_chain_01_ctrl_null, type=Control, loc=<0 0 0>, rot=<0 0 0 1>}
 dyn_mstr_chain_sim {parent=pelvis, loc=<-1.0023 -0.0498 -15.5755>, rot=<-0.1316 0.6755 0.72 0.0893>}
 dyn_prnt_chain_sim {parent=dyn_mstr_chain_sim, loc=<0 0 0>, rot=<0 0 0 1>}
 dyn_mstr_chain_sim_01 {parent=dyn_prnt_chain_sim, loc=<0 0 0>, rot=<0 0 0 1>}
 dyn_mstr_chain_sim_02 {parent=dyn_mstr_chain_sim_01, loc=<9.324 0 0>, rot=<0 0 0 1>}
 dyn_mstr_chain_sim_2 {parent=pelvis, loc=<-1.0023 -0.0498 -15.5755>, rot=<-0.1316 0.6755 0.72 0.0893>}
 dyn_prnt_chain_sim_2 {parent=dyn_mstr_chain_sim_2, loc=<0 0 0>, rot=<0 0 0 1>}
 dyn_mstr_chain_sim_01_2 {parent=dyn_prnt_chain_sim_2, loc=<0 0 0>, rot=<0 0 0 1>}
 dyn_mstr_chain_sim_02_2 {parent=dyn_mstr_chain_sim_01_2, loc=<9.324 0 0>, rot=<0 0 0 1>}
 dyn_main_chain_ctrl_null {parent=pelvis, type=Null, loc=<-18.7061 -0.9685 -21.3603>, rot=<-0.5558 -0.6004 -0.3999 -0.4132>}
 dyn_main_chain_ctrl {controlType=EULER_TRANSFORM, parent=dyn_main_chain_ctrl_null, type=Control, loc=<0 0 0>, rot=<0 0 0 1>}
 dyn_front_chain_013_ctrl_null {parent=dyn_main_chain_ctrl, type=Null, loc=<1.2372 0.1457 0.0128>, rot=<0.0586 -0.9982 -0.0108 -0.0045>}
 dyn_front_chain_013_ctrl {controlType=EULER_TRANSFORM, parent=dyn_front_chain_013_ctrl_null, type=Control, loc=<0 0 0>, rot=<0 0 0 1>}
 dyn_front_chain_014_ctrl_null {parent=dyn_main_chain_ctrl, type=Null, loc=<0 0 0>, rot=<0.0586 -0.9982 -0.0108 -0.0045>}
 dyn_front_chain_014_ctrl {controlType=EULER_TRANSFORM, parent=dyn_front_chain_014_ctrl_null, type=Control, loc=<0 0 0>, rot=<0 0 0 1>}
 dyn_back_chain_019_ctrl_null {parent=dyn_main_chain_ctrl, type=Null, loc=<-1.2927 0 0>, rot=<0 0 0 1>}
 dyn_back_chain_019_ctrl {controlType=EULER_TRANSFORM, parent=dyn_back_chain_019_ctrl_null, type=Control, loc=<0 0 0>, rot=<0 0 0 1>}
 dyn_back_chain_020_ctrl_null {parent=dyn_main_chain_ctrl, type=Null, loc=<0 0 0>, rot=<0 0 0 1>}
 dyn_back_chain_020_ctrl {controlType=EULER_TRANSFORM, parent=dyn_back_chain_020_ctrl_null, type=Control, loc=<0 0 0>, rot=<0 0 0 1>}
 dyn_front_chain_04_ctrl_null {parent=pelvis, type=Null, loc=<-12.3992 10.0212 -11.1524>, rot=<-0.4062 0.1173 0.9049 -0.0483>}
 dyn_front_chain_04_ctrl {controlType=EULER_TRANSFORM, parent=dyn_front_chain_04_ctrl_null, type=Control, loc=<0 0 0>, rot=<0 0 0 1>}
 dyn_front_chain_07_ctrl_null {parent=pelvis, type=Null, loc=<-15.6319 8.5894 -14.8963>, rot=<-0.4767 0.227 0.8329 -0.1659>}
 dyn_front_chain_07_ctrl {controlType=EULER_TRANSFORM, parent=dyn_front_chain_07_ctrl_null, type=Control, loc=<0 0 0>, rot=<0 0 0 1>}
 dyn_front_chain_010_ctrl_null {parent=pelvis, type=Null, loc=<-17.9212 4.9308 -19.0172>, rot=<-0.5148 0.2555 0.7337 -0.3625>}
 dyn_front_chain_010_ctrl {controlType=EULER_TRANSFORM, parent=dyn_front_chain_010_ctrl_null, type=Control, loc=<0 0 0>, rot=<0 0 0 1>}
 dyn_back_chain_04_ctrl_null {parent=pelvis, type=Null, loc=<-2.9304 -13.966 -4.6143>, rot=<0.1028 -0.9277 -0.164 -0.3194>}
 dyn_back_chain_04_ctrl {controlType=EULER_TRANSFORM, parent=dyn_back_chain_04_ctrl_null, type=Control, loc=<0 0 0>, rot=<0 0 0 1>}
 dyn_back_chain_07_ctrl_null {parent=pelvis, type=Null, loc=<-7.1891 -14.1346 -8.1785>, rot=<0.0244 -0.913 -0.1895 -0.3604>}
 dyn_back_chain_07_ctrl {controlType=EULER_TRANSFORM, parent=dyn_back_chain_07_ctrl_null, type=Control, loc=<0 0 0>, rot=<0 0 0 1>}
 dyn_back_chain_010_ctrl_null {parent=pelvis, type=Null, loc=<-11.0712 -13.2161 -11.8908>, rot=<-0.079 -0.8744 -0.226 -0.422>}
 dyn_back_chain_010_ctrl {controlType=EULER_TRANSFORM, parent=dyn_back_chain_010_ctrl_null, type=Control, loc=<0 0 0>, rot=<0 0 0 1>}
 dyn_back_chain_013_ctrl_null {parent=pelvis, type=Null, loc=<-14.6648 -10.7901 -16.1159>, rot=<-0.1774 -0.8308 -0.2663 -0.4555>}
 dyn_back_chain_013_ctrl {controlType=EULER_TRANSFORM, parent=dyn_back_chain_013_ctrl_null, type=Control, loc=<0 0 0>, rot=<0 0 0 1>}
 dyn_back_chain_016_ctrl_null {parent=pelvis, type=Null, loc=<-17.2552 -7.5583 -19.3647>, rot=<-0.3463 -0.754 -0.3288 -0.4511>}
 dyn_back_chain_016_ctrl {controlType=EULER_TRANSFORM, parent=dyn_back_chain_016_ctrl_null, type=Control, loc=<0 0 0>, rot=<0 0 0 1>}
]
variables [
 Calf_l_Bent {type=double, default=0}
 Calf_r_Bent {type=double, default=0}
 Thigh_l_Backward {type=double, default=0}
 Thigh_r_Backward {type=double, default=0}
 back_chain_controls {type="TArray<FRigElementKey>", default=(())}
 back_chain_ctrls {type="TArray<FRigElementKey>"}
 backward_movement_alpha {type=double, default=0}
 chain_ctrls {type="TArray<FRigElementKey>", default=(())}
 chain_ctrls_bindpose_pose {type=FRigPose, default=(Elements=,HierarchyTopologyVersion=-1,PoseHash=-1)}
 chain_ctrls_thigh_l_up_pose {type=FRigPose, default=(Elements=,HierarchyTopologyVersion=-1,PoseHash=-1)}
 dfrm_rig_thigh_up_alpha {type=double, default=0}
 forward_backward_movement_alpha {type=double, default=0}
 forward_movement_alpha {type=double, default=0}
 front_chain_controls {type="TArray<FRigElementKey>", default=(())}
 front_chain_ctrls {type="TArray<FRigElementKey>"}
 game_time_in_seconds {type=double, default=0}
 hand_r_multFactor {type=double, default=0}
 is_crouching {type=bool, default=False}
 is_dbno {type=bool, default=False}
 is_skydive_floating {type=bool, default=False}
 is_skydiving {type=bool, default=False}
 left_movement_alpha {type=double, default=0}
 left_right_movement_alpha {type=double, default=0}
 right_movement_alpha {type=double, default=0}
 spine_twist_right_delta_ref_alpha {type=double, default=0}
 spr_thigh_l_back_alpha {type=double, default=0}
 spr_thigh_r_back_alpha {type=double, default=0}
]
