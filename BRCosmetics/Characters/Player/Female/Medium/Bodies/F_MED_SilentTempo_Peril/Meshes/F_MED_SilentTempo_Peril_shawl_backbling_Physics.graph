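# Physics asset for the right-side shawl backbling of the SilentTempo Peril female medium body, providing capsule bodies and joints for secondary motion.
# Two thin-capsule chains, A with four links and B with three, hang from master bones through tightly limited swing and twist constraints.
# A cross-constraint with free rotation joins the second links of both chains.
# Larger capsules on pelvis, upper spine, right upper arm and its twist bone represent the body for collision.
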PhysicsAsset /BRCosmetics/Characters/Player/Female/Medium/Bodies/F_MED_SilentTempo_Peril/Meshes/F_MED_SilentTempo_Peril_shawl_backbling_Physics
bodies {
 dyn_mstr_shawl_A_r [{shapeType=Capsule, length=0, radius=0.5, shapeName=dyn_mstr_shawl_A_r_capsule}]
 dyn_mstr_shawl_B_r [{shapeType=Capsule, length=0, radius=0.5, shapeName=dyn_mstr_shawl_B_r_capsule}]
 dyn_shawl_A_01_r [{shapeType=Capsule, length=7.3866, radius=0.5, shapeName=dyn_shawl_A_01_r_capsule, center=<6.3 0 0>, rotation=<rot 0 0.0004 -10>}]
 dyn_shawl_A_02_r [{shapeType=Capsule, length=7.3866, radius=0.5, shapeName=dyn_shawl_A_01_r_capsule, center=<6.3 0 0>, rotation=<rot -0.0001 0.0009 0>}]
 dyn_shawl_A_03_r [{shapeType=Capsule, length=7.3866, radius=0.5, shapeName=dyn_shawl_A_01_r_capsule, center=<6.3 0 0>, rotation=<rot 0 0.0004 -10>}]
 dyn_shawl_A_04_r [
  {shapeType=Capsule, length=1.6281, radius=0.5, shapeName=dyn_shawl_A_04_r_capsule, center=<13.2 -0.8 1.7>}
  {shapeType=Capsule, length=5.154, radius=0.5, shapeName=dyn_shawl_A_04_r_capsule, center=<5.6298 0 0.4993>, rotation=<rot -0.0001 0.0002 -5>}
 ]
 dyn_shawl_B_01_r [{shapeType=Capsule, length=12.9162, radius=0.5, shapeName=dyn_shawl_B_01_r_capsule, center=<5.7 0 0>, rotation=<rot 0 0.001 -10>}]
 dyn_shawl_B_02_r [{shapeType=Capsule, length=7.0274, radius=0.5, shapeName=dyn_shawl_B_02_r_capsule, center=<5.7 0 0>, rotation=<rot 0 0.0006 -15>}]
 dyn_shawl_B_03_r [{shapeType=Capsule, length=6.1478, radius=0.5, shapeName=dyn_shawl_B_03_r_capsule, center=<5.7 0 0>, rotation=<rot 0 0.0001 -10>}]
 pelvis [{shapeType=Capsule, length=10.5178, radius=15, shapeName=pelvis_capsule, center=<1.8295 2.4303 0>}]
 spine_04 [{shapeType=Capsule, length=6.7968, radius=11.9906, shapeName=spine_04_capsule, center=<0 4.3792 1.9751>}]
 upperarm_r [{shapeType=Capsule, length=14.8678, radius=5.0009, shapeName=upperarm_r_capsule, center=<-10.8741 0 0>, rotation=<rot -90 179.9802 -179.9802>}]
 upperarm_twist_01_r [{shapeType=Capsule, length=14.8678, radius=5.0009, shapeName=upperarm_r_capsule, center=<-7.6163 0 0>, rotation=<rot -90 179.9995 -179.9996>}]
}
constraints [
 {bone1Name=dyn_shawl_A_04_r, bone2Name=dyn_shawl_A_03_r, swing1LimitDegrees=10, swing2LimitDegrees=15, twistLimitDegrees=2.5}
 {bone1Name=dyn_shawl_A_03_r, bone2Name=dyn_shawl_A_02_r, swing1LimitDegrees=10, swing2LimitDegrees=10, twistLimitDegrees=2.5}
 {bone1Name=dyn_shawl_A_02_r, bone2Name=dyn_shawl_A_01_r, swing1LimitDegrees=10, swing2LimitDegrees=7.5, twistLimitDegrees=2.5}
 {bone1Name=dyn_shawl_A_01_r, bone2Name=dyn_mstr_shawl_A_r, swing1LimitDegrees=10, swing2LimitDegrees=15, twistLimitDegrees=2.5}
 {bone1Name=dyn_shawl_B_03_r, bone2Name=dyn_shawl_B_02_r, swing1LimitDegrees=10, swing2LimitDegrees=10, twistLimitDegrees=2.5}
 {bone1Name=dyn_shawl_B_02_r, bone2Name=dyn_shawl_B_01_r, swing1LimitDegrees=10, swing2LimitDegrees=7.5, twistLimitDegrees=2.5}
 {bone1Name=dyn_shawl_B_01_r, bone2Name=dyn_mstr_shawl_B_r, swing1LimitDegrees=10, swing2LimitDegrees=5, twistLimitDegrees=2.5}
 {
  bone1Name=dyn_shawl_B_02_r
  bone2Name=dyn_shawl_A_02_r
  swing1LimitDegrees=45
  swing1Motion=Free
  swing2LimitDegrees=45
  swing2Motion=Free
  twistLimitDegrees=45
  twistMotion=Free
 }
]
