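# Physics asset for the skirt and clock chain on the SilentTempo Peril medium female player body.
# Capsules on pelvis, thighs and calves, plus a FloorCOL box, sit alongside bodies for the skirt and clock-chain bones.
# Limited-angle constraints link each strand bone to bone, with tight twist and a twist-locked back-side link to the left thigh.
# Skirt strands are mostly left-side, many using large offset capsules; only one back strand is right-side.
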PhysicsAsset /BRCosmetics/Characters/Player/Female/Medium/Bodies/F_MED_SilentTempo_Peril/Meshes/F_MED_SilentTempo_Peril_skirt_Physics
bodies {
 FloorCOL [{shapeType=Box, extentX=150, extentY=150, extentZ=10, shapeName=FloorCOL_box, center=<0 0 -0.0418>, rotation=<rot 0 -90 0>}]
 dfrm_calf_l [
  {
   shapeType=Capsule
   length=40.7286
   radius=9.5896
   shapeName=dfrm_calf_l_capsule
   center=<27.544 2.8812 0>
   rotation=<rot 90 179.9999 -179.9999>
  }
  {shapeType=Capsule, length=4.3319, radius=9.4123, shapeName=dfrm_calf_l_capsule, center=<0.9113 3.0061 0>, rotation=<rot 0 -0.0003 -90>}
 ]
 dfrm_calf_r [{
   shapeType=Capsule
   length=41.9594
   radius=9.5896
   shapeName=dfrm_calf_l_capsule
   center=<27.544 -2.8812 0>
   rotation=<rot 90 179.9999 0.0001>
  }]
 dfrm_thigh_l [{
   shapeType=Capsule
   length=20.3514
   radius=11.9162
   shapeName=dfrm_thigh_l_capsule
   center=<22.5408 2.0645 3.2318>
   rotation=<rot 88.7613 -0.0015 2.9551>
  }]
 dfrm_thigh_r [{
   shapeType=Capsule
   length=19.6412
   radius=12.6264
   shapeName=dfrm_thigh_l_capsule
   center=<22.6115 -0.8203 -2.9351>
   rotation=<rot 88.7613 -0.0015 -177.0449>
  }]
 dyn_back_side_skirt_02_l [{shapeType=Capsule, length=0, radius=0.5, shapeName=dyn_back_side_skirt_02_l_capsule, center=<12.8768 0 0>}]
 dyn_back_side_skirt_03_l [{shapeType=Capsule, length=6.8063, radius=7.3105, shapeName=dyn_back_skirt_02_r_capsule, center=<13.468 -5 0>}]
 dyn_back_side_skirt_04_l [{shapeType=Capsule, length=33.4361, radius=16.227, shapeName=dyn_back_skirt_02_r_capsule, center=<13.468 -10 0>}]
 dyn_back_skirt_01_l [{shapeType=Capsule, length=0.3914, radius=0.5, shapeName=dyn_back_skirt_02_r_capsule}]
 dyn_back_skirt_01_r [{shapeType=Capsule, length=0.3914, radius=0.5, shapeName=dyn_back_skirt_01_r_capsule}]
 dyn_back_skirt_02_l [{shapeType=Capsule, length=18.896, radius=11.3531, shapeName=dyn_back_skirt_02_r_capsule, center=<13.468 -10 0>}]
 dyn_back_skirt_02_r [{shapeType=Capsule, length=13.2254, radius=0.8914, shapeName=dyn_back_skirt_02_r_capsule, center=<13.468 0 0>}]
 dyn_back_skirt_03_l [{shapeType=Capsule, length=23.4334, radius=11.9699, shapeName=dyn_back_skirt_02_r_capsule, center=<13.468 -9 0>}]
 dyn_clock_chain_01 [{
   shapeType=Capsule
   length=4.1493
   radius=0.9705
   shapeName=dyn_clock_chain_01_capsule
   center=<2.1575 0 0>
   rotation=<rot 90 144.7357 144.7357>
  }]
 dyn_clock_chain_02 [{
   shapeType=Capsule
   length=4.1493
   radius=0.9705
   shapeName=dyn_clock_chain_01_capsule
   center=<2.1575 0 0>
   rotation=<rot 90 144.7357 144.7357>
  }]
 dyn_clock_chain_03 [
  {shapeType=Capsule, length=19.493, radius=1.3637, shapeName=dyn_clock_chain_03_capsule, center=<11.12 4.3359 0>}
  {shapeType=Capsule, length=9.3215, radius=0.9722, shapeName=dyn_clock_chain_03_capsule, center=<1.9333 0 0>}
  {shapeType=Capsule, length=8.9282, radius=1.3655, shapeName=dyn_clock_chain_03_capsule, center=<17.4685 7.6211 0>}
 ]
 dyn_front_skirt_01_l [{shapeType=Capsule, length=0.9761, radius=0.5, shapeName=dyn_back_skirt_02_r_capsule, center=<11.461 0 0>}]
 dyn_front_skirt_02_l [{shapeType=Capsule, length=6.9052, radius=7.2115, shapeName=dyn_back_skirt_02_r_capsule, center=<13.468 -5 0>}]
 dyn_front_skirt_03_l [{shapeType=Capsule, length=12.6407, radius=1.4761, shapeName=dyn_back_skirt_02_r_capsule, center=<13.468 0 0>}]
 dyn_mstr_clock_chain [{shapeType=Capsule, length=0, radius=0.5, shapeName=dyn_mstr_clock_chain_capsule}]
 dyn_side_skirt_02_l [{shapeType=Capsule, length=0.976, radius=0.5001, shapeName=dyn_back_skirt_02_r_capsule, center=<14.3452 0 0>}]
 dyn_side_skirt_03_l [{shapeType=Capsule, length=6.967, radius=7.1498, shapeName=dyn_back_skirt_02_r_capsule, center=<13.468 -5 0>}]
 dyn_side_skirt_04_l [{shapeType=Capsule, length=6.967, radius=7.1498, shapeName=dyn_back_skirt_02_r_capsule, center=<13.468 -5 0>}]
 pelvis [{shapeType=Capsule, length=12.837, radius=10.3442, shapeName=pelvis_capsule, center=<-8.3681 -3.5647 0>}]
}
constraints [
 {bone1Name=dyn_side_skirt_04_l, bone2Name=dyn_side_skirt_03_l, swing1LimitDegrees=25, swing2LimitDegrees=15, twistLimitDegrees=2}
 {bone1Name=dyn_back_side_skirt_04_l, bone2Name=dyn_back_side_skirt_03_l, swing1LimitDegrees=25, swing2LimitDegrees=10, twistLimitDegrees=2}
 {bone1Name=dyn_clock_chain_03, bone2Name=dyn_clock_chain_02, swing1LimitDegrees=20, swing2LimitDegrees=30, twistLimitDegrees=2}
 {bone1Name=dyn_clock_chain_02, bone2Name=dyn_clock_chain_01, swing1LimitDegrees=20, swing2LimitDegrees=30, twistLimitDegrees=2}
 {bone1Name=dyn_clock_chain_01, bone2Name=dyn_mstr_clock_chain, swing1LimitDegrees=20, swing2LimitDegrees=30, twistLimitDegrees=2}
 {bone1Name=dyn_side_skirt_03_l, bone2Name=dyn_side_skirt_02_l, swing1LimitDegrees=25, swing2LimitDegrees=15, twistLimitDegrees=1}
 {bone1Name=dyn_front_skirt_03_l, bone2Name=dyn_front_skirt_02_l, swing1LimitDegrees=25, swing2LimitDegrees=15, twistLimitDegrees=2}
 {bone1Name=dyn_front_skirt_02_l, bone2Name=dyn_front_skirt_01_l, swing1LimitDegrees=25, swing2LimitDegrees=15, twistLimitDegrees=2}
 {bone1Name=dyn_back_side_skirt_03_l, bone2Name=dyn_back_side_skirt_02_l, swing1LimitDegrees=25, swing2LimitDegrees=10, twistLimitDegrees=2}
 {bone1Name=dyn_back_skirt_02_r, bone2Name=dyn_back_skirt_01_r, swing1LimitDegrees=40, swing2LimitDegrees=20, twistLimitDegrees=2}
 {bone1Name=dyn_back_skirt_03_l, bone2Name=dyn_back_skirt_02_l, swing1LimitDegrees=25, swing2LimitDegrees=15, twistLimitDegrees=2}
 {bone1Name=dyn_back_skirt_02_l, bone2Name=dyn_back_skirt_01_l, swing1LimitDegrees=40, swing2LimitDegrees=20, twistLimitDegrees=2}
 {
  bone1Name=dyn_back_side_skirt_02_l
  bone2Name=dfrm_thigh_l
  swing1LimitDegrees=20
  swing2LimitDegrees=15
  twistLimitDegrees=45
  twistMotion=Locked
 }
]
