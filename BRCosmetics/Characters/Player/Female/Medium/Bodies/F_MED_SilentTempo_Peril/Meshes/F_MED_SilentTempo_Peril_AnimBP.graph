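# Post-process style AnimBP for the SilentTempo Peril female body cosmetic, adding secondary motion to chains, skirt, shawl, laces, tassels and sleeve.
# Copied parent pose feeds RigidBody sims, a Control Rig and AnimDynamics, with variants for front-end and backpack.
# Weighted constraints pin tassel and lapel roots before final sims; an emote-weighted slot layers clock-hand animation.
# The event graph updates gravity, movement and spine-twist values, plus material WPO parameters from limb motion.
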AnimBlueprint /BRCosmetics/Characters/Player/Female/Medium/Bodies/F_MED_SilentTempo_Peril/Meshes/F_MED_SilentTempo_Peril_AnimBP
target_skeleton=/Game/Characters/Player/Female/Medium/Bodies/F_MED_SilentTempo_Peril/Meshes/F_MED_SilentTempo_Peril_Skeleton
graphs [
 AnimGraph {nodes [
   n1 = Anim Dynamics - Bone: dyn_tassel_01(ComponentPose=Local To Component(LocalPose=Use cached pose 'PostRivets'()), Alpha=1, ExternalForce="0,0,0") {
    angularDampingOverride=0.999
    angularSpringConstant=50
    bAngularSpring=true
    bOverrideAngularDamping=true
    simulationSpace=RootRelative
    alphaBoolBlend {blendOption=Linear}
    boundBone {boneName=dyn_tassel_01}
    componentAppliedLinearAccClamp=<2500 2500 2500>
    componentLinearAccScale=<0.1 0.1 0.1>
    componentLinearVelScale=<0.9 0.9 0.9>
    physicsBodyDefinitions [{
      boundBone {boneName=dyn_tassel_01}
      boxExtents=<12 7 7>
      constraintSetup {angularLimitsMax=<0 2.5 0>, angularLimitsMin=<0 -2.5 -10>, angularTarget=<1 0 0>}
      localJointOffset=<8 0 0>
     }]
    planarLimits [{drivingBone {boneName=FloorCOL}, planeTransform {location=<0 0 2.5>, rotation=<rot 0 0 0>, scale=<2 2 1>}}]
   }
   n2 = Anim Dynamics - Bone: dyn_tassel_02(ComponentPose=@n1, Alpha=1, ExternalForce="0,0,0") {
    angularDampingOverride=0.999
    angularSpringConstant=50
    bAngularSpring=true
    bOverrideAngularDamping=true
    simulationSpace=RootRelative
    alphaBoolBlend {blendOption=Linear}
    boundBone {boneName=dyn_tassel_02}
    componentAppliedLinearAccClamp=<2500 2500 2500>
    componentLinearAccScale=<0.1 0.1 0.1>
    componentLinearVelScale=<0.9 0.9 0.9>
    physicsBodyDefinitions [{
      boundBone {boneName=dyn_tassel_02}
      boxExtents=<12 7 7>
      constraintSetup {angularLimitsMax=<0 2.5 10>, angularLimitsMin=<0 -2.5 -10>, angularTarget=<1 0 0>}
      localJointOffset=<8 0 0>
     }]
    planarLimits [{drivingBone {boneName=FloorCOL}, planeTransform {location=<0 0 2.5>, rotation=<rot 0 0 0>, scale=<2 2 1>}}]
   }
   n3 = Anim Dynamics - Bone: dyn_tassel_03(ComponentPose=@n2, Alpha=1, ExternalForce="0,0,0") {
    angularDampingOverride=0.999
    angularSpringConstant=50
    bAngularSpring=true
    bOverrideAngularDamping=true
    simulationSpace=RootRelative
    alphaBoolBlend {blendOption=Linear}
    boundBone {boneName=dyn_tassel_03}
    componentAppliedLinearAccClamp=<2500 2500 2500>
    componentLinearAccScale=<0.1 0.1 0.1>
    componentLinearVelScale=<0.9 0.9 0.9>
    physicsBodyDefinitions [{
      boundBone {boneName=dyn_tassel_03}
      boxExtents=<12 7 7>
      constraintSetup {angularLimitsMax=<0 2.5 10>, angularLimitsMin=<0 -2.5 -10>, angularTarget=<1 0 0>}
      localJointOffset=<8 0 0>
     }]
    planarLimits [{drivingBone {boneName=FloorCOL}, planeTransform {location=<0 0 2.5>, rotation=<rot 0 0 0>, scale=<2 2 1>}}]
   }
   n4 = Anim Dynamics - Bone: dyn_tassel_04(ComponentPose=@n3, Alpha=1, ExternalForce="0,0,0") {
    angularDampingOverride=0.999
    angularSpringConstant=50
    bAngularSpring=true
    bOverrideAngularDamping=true
    simulationSpace=RootRelative
    alphaBoolBlend {blendOption=Linear}
    boundBone {boneName=dyn_tassel_04}
    componentAppliedLinearAccClamp=<2500 2500 2500>
    componentLinearAccScale=<0.1 0.1 0.1>
    componentLinearVelScale=<0.9 0.9 0.9>
    physicsBodyDefinitions [{
      boundBone {boneName=dyn_tassel_04}
      boxExtents=<12 7 7>
      constraintSetup {angularLimitsMax=<0 2.5 10>, angularLimitsMin=<0 -2.5 -10>, angularTarget=<1 0 0>}
      localJointOffset=<8 0 0>
     }]
    planarLimits [{drivingBone {boneName=FloorCOL}, planeTransform {location=<0 0 2.5>, rotation=<rot 0 0 0>, scale=<2 2 1>}}]
   }
   n5 = Anim Dynamics - Bone: dyn_tassel_05(ComponentPose=@n4, Alpha=1, ExternalForce="0,0,0") {
    angularDampingOverride=0.999
    angularSpringConstant=50
    bAngularSpring=true
    bOverrideAngularDamping=true
    simulationSpace=RootRelative
    alphaBoolBlend {blendOption=Linear}
    boundBone {boneName=dyn_tassel_05}
    componentAppliedLinearAccClamp=<2500 2500 2500>
    componentLinearAccScale=<0.1 0.1 0.1>
    componentLinearVelScale=<0.9 0.9 0.9>
    physicsBodyDefinitions [{
      boundBone {boneName=dyn_tassel_05}
      boxExtents=<12 7 7>
      constraintSetup {angularLimitsMax=<0 2.5 10>, angularLimitsMin=<0 -2.5 -10>, angularTarget=<1 0 0>}
      localJointOffset=<8 0 0>
     }]
    planarLimits [{drivingBone {boneName=FloorCOL}, planeTransform {location=<0 0 2.5>, rotation=<rot 0 0 0>, scale=<2 2 1>}}]
   }
   n6 = Anim Dynamics - Bone: dyn_tassel_06(ComponentPose=@n5, Alpha=1, ExternalForce="0,0,0") {
    angularDampingOverride=0.999
    angularSpringConstant=50
    bAngularSpring=true
    bOverrideAngularDamping=true
    simulationSpace=RootRelative
    alphaBoolBlend {blendOption=Linear}
    boundBone {boneName=dyn_tassel_06}
    componentAppliedLinearAccClamp=<2500 2500 2500>
    componentLinearAccScale=<0.1 0.1 0.1>
    componentLinearVelScale=<0.9 0.9 0.9>
    physicsBodyDefinitions [{
      boundBone {boneName=dyn_tassel_06}
      boxExtents=<12 7 7>
      constraintSetup {angularLimitsMax=<0 2.5 10>, angularLimitsMin=<0 -2.5 -10>, angularTarget=<1 0 0>}
      localJointOffset=<8 0 0>
     }]
    planarLimits [{drivingBone {boneName=FloorCOL}, planeTransform {location=<0 0 2.5>, rotation=<rot 0 0 0>, scale=<2 2 1>}}]
   }
   n7 = Anim Dynamics - Bone: dyn_tassel_07(ComponentPose=@n6, Alpha=1, ExternalForce="0,0,0") {
    angularDampingOverride=0.999
    angularSpringConstant=50
    bAngularSpring=true
    bOverrideAngularDamping=true
    simulationSpace=RootRelative
    alphaBoolBlend {blendOption=Linear}
    boundBone {boneName=dyn_tassel_07}
    componentAppliedLinearAccClamp=<2500 2500 2500>
    componentLinearAccScale=<0.1 0.1 0.1>
    componentLinearVelScale=<0.9 0.9 0.9>
    physicsBodyDefinitions [{
      boundBone {boneName=dyn_tassel_07}
      boxExtents=<12 7 7>
      constraintSetup {angularLimitsMax=<0 2.5 10>, angularLimitsMin=<0 -2.5 -10>, angularTarget=<1 0 0>}
      localJointOffset=<8 0 0>
     }]
    planarLimits [{drivingBone {boneName=FloorCOL}, planeTransform {location=<0 0 2.5>, rotation=<rot 0 0 0>, scale=<2 2 1>}}]
   }
   n8 = Anim Dynamics - Bone: dyn_tassel_08(ComponentPose=@n7, Alpha=1, ExternalForce="0,0,0") {
    angularDampingOverride=0.999
    angularSpringConstant=50
    bAngularSpring=true
    bOverrideAngularDamping=true
    simulationSpace=RootRelative
    alphaBoolBlend {blendOption=Linear}
    boundBone {boneName=dyn_tassel_08}
    componentAppliedLinearAccClamp=<2500 2500 2500>
    componentLinearAccScale=<0.1 0.1 0.1>
    componentLinearVelScale=<0.9 0.9 0.9>
    physicsBodyDefinitions [{
      boundBone {boneName=dyn_tassel_08}
      boxExtents=<12 7 7>
      constraintSetup {angularLimitsMax=<0 2.5 10>, angularLimitsMin=<0 -2.5 -10>, angularTarget=<1 0 0>}
      localJointOffset=<8 0 0>
     }]
    planarLimits [{drivingBone {boneName=FloorCOL}, planeTransform {location=<0 0 2.5>, rotation=<rot 0 0 0>, scale=<2 2 1>}}]
   }
   n9 = Anim Dynamics - Bone: dyn_tassel_09(ComponentPose=@n8, Alpha=1, ExternalForce="0,0,0") {
    angularDampingOverride=0.999
    angularSpringConstant=50
    bAngularSpring=true
    bOverrideAngularDamping=true
    simulationSpace=RootRelative
    alphaBoolBlend {blendOption=Linear}
    boundBone {boneName=dyn_tassel_09}
    componentAppliedLinearAccClamp=<2500 2500 2500>
    componentLinearAccScale=<0.1 0.1 0.1>
    componentLinearVelScale=<0.9 0.9 0.9>
    physicsBodyDefinitions [{
      boundBone {boneName=dyn_tassel_09}
      boxExtents=<12 7 7>
      constraintSetup {angularLimitsMax=<0 2.5 10>, angularLimitsMin=<0 -2.5 -10>, angularTarget=<1 0 0>}
      localJointOffset=<8 0 0>
     }]
    planarLimits [{drivingBone {boneName=FloorCOL}, planeTransform {location=<0 0 2.5>, rotation=<rot 0 0 0>, scale=<2 2 1>}}]
   }
   n10 = Anim Dynamics - Bone: dyn_tassel_10(ComponentPose=@n9, Alpha=1, ExternalForce="0,0,0") {
    angularDampingOverride=0.999
    angularSpringConstant=50
    bAngularSpring=true
    bOverrideAngularDamping=true
    simulationSpace=RootRelative
    alphaBoolBlend {blendOption=Linear}
    boundBone {boneName=dyn_tassel_10}
    componentAppliedLinearAccClamp=<2500 2500 2500>
    componentLinearAccScale=<0.1 0.1 0.1>
    componentLinearVelScale=<0.9 0.9 0.9>
    physicsBodyDefinitions [{
      boundBone {boneName=dyn_tassel_10}
      boxExtents=<12 7 7>
      constraintSetup {angularLimitsMax=<0 2.5 10>, angularLimitsMin=<0 -2.5 -10>, angularTarget=<1 0 0>}
      localJointOffset=<8 0 0>
     }]
    planarLimits [{drivingBone {boneName=FloorCOL}, planeTransform {location=<0 0 2.5>, rotation=<rot 0 0 0>, scale=<2 2 1>}}]
   }
   n11 = Anim Dynamics - Bone: dyn_tassel_11(ComponentPose=@n10, Alpha=1, ExternalForce="0,0,0") {
    angularDampingOverride=0.999
    angularSpringConstant=50
    bAngularSpring=true
    bOverrideAngularDamping=true
    simulationSpace=RootRelative
    alphaBoolBlend {blendOption=Linear}
    boundBone {boneName=dyn_tassel_11}
    componentAppliedLinearAccClamp=<2500 2500 2500>
    componentLinearAccScale=<0.1 0.1 0.1>
    componentLinearVelScale=<0.9 0.9 0.9>
    physicsBodyDefinitions [{
      boundBone {boneName=dyn_tassel_11}
      boxExtents=<12 7 7>
      constraintSetup {angularLimitsMax=<0 2.5 10>, angularLimitsMin=<0 -2.5 -10>, angularTarget=<1 0 0>}
      localJointOffset=<8 0 0>
     }]
    planarLimits [{drivingBone {boneName=FloorCOL}, planeTransform {location=<0 0 2.5>, rotation=<rot 0 0 0>, scale=<2 2 1>}}]
   }
   n12 = Anim Dynamics - Bone: dyn_tassel_12(ComponentPose=@n11, Alpha=1, ExternalForce="0,0,0") {
    angularDampingOverride=0.999
    angularSpringConstant=50
    bAngularSpring=true
    bOverrideAngularDamping=true
    simulationSpace=RootRelative
    alphaBoolBlend {blendOption=Linear}
    boundBone {boneName=dyn_tassel_12}
    componentAppliedLinearAccClamp=<2500 2500 2500>
    componentLinearAccScale=<0.1 0.1 0.1>
    componentLinearVelScale=<0.9 0.9 0.9>
    physicsBodyDefinitions [{
      boundBone {boneName=dyn_tassel_12}
      boxExtents=<12 7 7>
      constraintSetup {angularLimitsMax=<0 2.5 2.5>, angularLimitsMin=<0 -2.5 -12.5>, angularTarget=<1 0 0>}
      localJointOffset=<8 0 0>
     }]
    planarLimits [{drivingBone {boneName=FloorCOL}, planeTransform {location=<0 0 2.5>, rotation=<rot 0 0 0>, scale=<2 2 1>}}]
   }
   n13 = RigidBody!(ComponentPose=@n12, Alpha, OverrideWorldGravity:Vector="0,0,-980") -> (Pose) {
    bClampLinearTranslationLimitToRefPose=true
    bOverrideWorldGravity=true
    overridePhysicsAsset=/Game/Characters/Player/Female/Medium/Bodies/F_MED_SilentTempo_Peril/Meshes/F_MED_SilentTempo_Peril_laces_Physics
    simulationSpace=BaseBoneSpace
    alphaBoolBlend {blendOption=Linear}
    baseBoneRef {boneName=spine_05}
    componentAppliedLinearAccClamp=<2500 2500 2500>
    componentLinearAccScale=<0.8 0.8 0.8>
    componentLinearVelScale=<0.6 0.6 0.6>
    overrideWorldGravity {z=-980}
    simSpaceSettings {maxAngularAcceleration=1000, maxAngularVelocity=3, maxLinearAcceleration=250, maxLinearVelocity=500, velocityScaleZ=0, worldAlpha=1}
   }
   n14 = Anim Dynamics - Bone: dfrm_sleeve_07_r(ComponentPose=@n13, Alpha=1, ExternalForce="0,0,0") {
    angularDampingOverride=1
    angularSpringConstant=2500
    bAngularSpring=true
    bOverrideAngularDamping=true
    bUsePlanarLimit=false
    simulationSpace=RootRelative
    alphaBoolBlend {blendOption=Linear}
    boundBone {boneName=dfrm_sleeve_07_r}
    componentAppliedLinearAccClamp=<2500 2500 2500>
    componentLinearAccScale=<0.1 0.1 0.1>
    componentLinearVelScale=<0.9 0.9 0.9>
    physicsBodyDefinitions [{
      boundBone {boneName=dfrm_sleeve_07_r}
      boxExtents=<17 3.1104 8.5>
      constraintSetup {coneAngle=45, angularLimitsMax=<2 10 5>, angularLimitsMin=<-2 10 -20>, angularTarget=<1 0 0>}
      localJointOffset=<0.1884 0 0>
     }]
    sphericalLimits [{limitRadius=0.272, drivingBone {boneName=None}, sphereLocalOffset=<0 0 0>}]
   }
   n15 = Anim Dynamics - Bone: dfrm_sleeve_08_r(ComponentPose=@n14, Alpha=1, ExternalForce="0,0,0") {
    angularDampingOverride=1
    angularSpringConstant=2500
    bAngularSpring=true
    bOverrideAngularDamping=true
    bUsePlanarLimit=false
    simulationSpace=RootRelative
    alphaBoolBlend {blendOption=Linear}
    boundBone {boneName=dfrm_sleeve_08_r}
    componentAppliedLinearAccClamp=<2500 2500 2500>
    componentLinearAccScale=<0.1 0.1 0.1>
    componentLinearVelScale=<0.9 0.9 0.9>
    physicsBodyDefinitions [{
      boundBone {boneName=dfrm_sleeve_08_r}
      boxExtents=<17 3.1104 8.5>
      constraintSetup {coneAngle=45, angularLimitsMax=<2 10 5>, angularLimitsMin=<-2 10 -20>, angularTarget=<1 0 0>}
      localJointOffset=<0.1884 0 0>
     }]
    sphericalLimits [{limitRadius=0.272, drivingBone {boneName=None}, sphereLocalOffset=<0 0 0>}]
   }
   n16 = Anim Dynamics - Bone: dfrm_sleeve_01_r(ComponentPose=@n15, Alpha=1, ExternalForce="0,0,0") {
    angularDampingOverride=1
    angularSpringConstant=2500
    bAngularSpring=true
    bOverrideAngularDamping=true
    bUsePlanarLimit=false
    simulationSpace=RootRelative
    alphaBoolBlend {blendOption=Linear}
    boundBone {boneName=dfrm_sleeve_01_r}
    componentAppliedLinearAccClamp=<2500 2500 2500>
    componentLinearAccScale=<0.1 0.1 0.1>
    componentLinearVelScale=<0.9 0.9 0.9>
    physicsBodyDefinitions [{
      boundBone {boneName=dfrm_sleeve_01_r}
      boxExtents=<17 3.1104 8.5>
      constraintSetup {coneAngle=45, angularLimitsMax=<2 10 5>, angularLimitsMin=<-2 10 -20>, angularTarget=<1 0 0>}
      localJointOffset=<0.1884 0 0>
     }]
    sphericalLimits [{limitRadius=0.272, drivingBone {boneName=None}, sphereLocalOffset=<0 0 0>}]
   }
   n17 = Anim Dynamics - Bone: dfrm_sleeve_06_r(ComponentPose=@n16, Alpha=1, ExternalForce="0,0,0") {
    angularDampingOverride=1
    angularSpringConstant=2500
    bAngularSpring=true
    bOverrideAngularDamping=true
    bUsePlanarLimit=false
    simulationSpace=RootRelative
    alphaBoolBlend {blendOption=Linear}
    boundBone {boneName=dfrm_sleeve_06_r}
    componentAppliedLinearAccClamp=<2500 2500 2500>
    componentLinearAccScale=<0.1 0.1 0.1>
    componentLinearVelScale=<0.9 0.9 0.9>
    physicsBodyDefinitions [{
      boundBone {boneName=dfrm_sleeve_06_r}
      boxExtents=<17 3.1104 8.5>
      constraintSetup {coneAngle=45, angularLimitsMax=<2 10 5>, angularLimitsMin=<-2 10 -20>, angularTarget=<1 0 0>}
      localJointOffset=<0.1884 0 0>
     }]
    sphericalLimits [{limitRadius=0.272, drivingBone {boneName=None}, sphereLocalOffset=<0 0 0>}]
   }
   n18 = Anim Dynamics - Bone: dfrm_sleeve_05_r(ComponentPose=@n17, Alpha=1, ExternalForce="0,0,0") {
    angularDampingOverride=1
    angularSpringConstant=2500
    bAngularSpring=true
    bOverrideAngularDamping=true
    bUsePlanarLimit=false
    simulationSpace=RootRelative
    alphaBoolBlend {blendOption=Linear}
    boundBone {boneName=dfrm_sleeve_05_r}
    componentAppliedLinearAccClamp=<2500 2500 2500>
    componentLinearAccScale=<0.1 0.1 0.1>
    componentLinearVelScale=<0.9 0.9 0.9>
    physicsBodyDefinitions [{
      boundBone {boneName=dfrm_sleeve_05_r}
      boxExtents=<17 3.1104 8.5>
      constraintSetup {coneAngle=45, angularLimitsMax=<2 10 5>, angularLimitsMin=<-2 10 -20>, angularTarget=<1 0 0>}
      localJointOffset=<0.1884 0 0>
     }]
    sphericalLimits [{limitRadius=0.272, drivingBone {boneName=None}, sphereLocalOffset=<0 0 0>}]
   }
   n19 = Anim Dynamics - Bone: dfrm_sleeve_04_r(ComponentPose=@n18, Alpha=1, ExternalForce="0,0,0") {
    angularDampingOverride=1
    angularSpringConstant=2500
    bAngularSpring=true
    bOverrideAngularDamping=true
    bUsePlanarLimit=false
    simulationSpace=RootRelative
    alphaBoolBlend {blendOption=Linear}
    boundBone {boneName=dfrm_sleeve_04_r}
    componentAppliedLinearAccClamp=<2500 2500 2500>
    componentLinearAccScale=<0.1 0.1 0.1>
    componentLinearVelScale=<0.9 0.9 0.9>
    physicsBodyDefinitions [{
      boundBone {boneName=dfrm_sleeve_04_r}
      boxExtents=<17 3.1104 8.5>
      constraintSetup {coneAngle=45, angularLimitsMax=<2 10 5>, angularLimitsMin=<-2 10 -20>, angularTarget=<1 0 0>}
      localJointOffset=<0.1884 0 0>
     }]
    sphericalLimits [{limitRadius=0.272, drivingBone {boneName=None}, sphereLocalOffset=<0 0 0>}]
   }
   n20 = Anim Dynamics - Bone: dfrm_sleeve_02_r(ComponentPose=@n19, Alpha=1, ExternalForce="0,0,0") {
    angularDampingOverride=1
    angularSpringConstant=2500
    bAngularSpring=true
    bOverrideAngularDamping=true
    bUsePlanarLimit=false
    simulationSpace=RootRelative
    alphaBoolBlend {blendOption=Linear}
    boundBone {boneName=dfrm_sleeve_02_r}
    componentAppliedLinearAccClamp=<2500 2500 2500>
    componentLinearAccScale=<0.1 0.1 0.1>
    componentLinearVelScale=<0.9 0.9 0.9>
    physicsBodyDefinitions [{
      boundBone {boneName=dfrm_sleeve_02_r}
      boxExtents=<17 3.1104 8.5>
      constraintSetup {coneAngle=45, angularLimitsMax=<2 10 5>, angularLimitsMin=<-2 10 -20>, angularTarget=<1 0 0>}
      localJointOffset=<0.1884 0 0>
     }]
    sphericalLimits [{limitRadius=0.272, drivingBone {boneName=None}, sphereLocalOffset=<0 0 0>}]
   }
   n21 = Anim Dynamics - Bone: dfrm_sleeve_03_r(ComponentPose=@n20, Alpha=1, ExternalForce="0,0,0") {
    angularDampingOverride=1
    angularSpringConstant=2500
    bAngularSpring=true
    bOverrideAngularDamping=true
    bUsePlanarLimit=false
    simulationSpace=RootRelative
    alphaBoolBlend {blendOption=Linear}
    boundBone {boneName=dfrm_sleeve_03_r}
    componentAppliedLinearAccClamp=<2500 2500 2500>
    componentLinearAccScale=<0.1 0.1 0.1>
    componentLinearVelScale=<0.9 0.9 0.9>
    physicsBodyDefinitions [{
      boundBone {boneName=dfrm_sleeve_03_r}
      boxExtents=<17 3.1104 8.5>
      constraintSetup {coneAngle=45, angularLimitsMax=<2 10 5>, angularLimitsMin=<-2 10 -20>, angularTarget=<1 0 0>}
      localJointOffset=<0.1884 0 0>
     }]
    sphericalLimits [{limitRadius=0.272, drivingBone {boneName=None}, sphereLocalOffset=<0 0 0>}]
   }
   n22 = Layered blend per bone(
    BasePose=Component To Local(ComponentPose=@n21)
    BlendPoses_0=Slot 'DefaultSlot'()
    BlendWeights_0=Get Built In Emote() -> (Built In Emote:double)
   ) {layerSetup [{branchFilters [{boneName=dfrm_clock_hour}, {boneName=dfrm_clock_minute}, {boneName=dfrm_clock_second}]}]}
   Output Pose(Result=@n22) {layerGroup=None}
   n23 = RigidBody(ComponentPose=Local To Component(LocalPose=Copy Pose From Mesh() {bUseAttachedParent=true}), OverrideWorldGravity="0,0,-980") {
    bClampLinearTranslationLimitToRefPose=true
    bOverrideWorldGravity=true
    overridePhysicsAsset=&/F_MED_SilentTempo_Peril_Chain_Physics
    simulationSpace=BaseBoneSpace
    alphaBoolBlend {blendOption=Linear}
    baseBoneRef {boneName=pelvis}
    componentAppliedLinearAccClamp=<2500 2500 2500>
    componentLinearAccScale=<0.4 0.3 0.1>
    componentLinearVelScale=<0.4 0.3 0.1>
    overrideWorldGravity {z=-980}
    simSpaceSettings {maxAngularAcceleration=1000, maxAngularVelocity=3, maxLinearAcceleration=250, maxLinearVelocity=500, velocityScaleZ=0, worldAlpha=1}
   }
   n24 = RigidBody(ComponentPose=Local To Component(LocalPose=Copy Pose From Mesh() {bUseAttachedParent=true}), Alpha=0.5, OverrideWorldGravity="0,0,-980") {
    bClampLinearTranslationLimitToRefPose=true
    bOverrideWorldGravity=true
    overridePhysicsAsset=&F_MED_SilentTempo_Peril_Chain_Physics
    simulationSpace=BaseBoneSpace
    alphaBoolBlend {blendOption=Linear}
    baseBoneRef {boneName=pelvis}
    componentAppliedLinearAccClamp=<2500 2500 2500>
    componentLinearAccScale=<0.4 0.3 0.1>
    componentLinearVelScale=<0.4 0.3 0.1>
    overrideWorldGravity {z=-980}
    simSpaceSettings {maxAngularAcceleration=1000, maxAngularVelocity=3, maxLinearAcceleration=250, maxLinearVelocity=500, velocityScaleZ=0, worldAlpha=1}
   }
   n25 = Control Rig!(
    Source=Blend Poses by bool(BlendPose_0=Component To Local(ComponentPose=@n24), BlendPose_1=Component To Local(ComponentPose=@n23), bActiveValue=Get bIsInFrontEnd()) {blendType=Linear}
    Alpha
    is_skydiving:bool=Get bIsSkyDiving()
    is_skydive_floating:bool=Get bIsSkydiveFloating()
    is_crouching:bool=Get bIsCrouching()
    is_dbno:bool=Get bIsDBNO()
    forward_backward_movement_alpha:double=Get forwards_backward_movement_alpha() -> (forwards_backward_movement_alpha:double)
    left_right_movement_alpha:double=0
    spine_twist_right_delta_ref_alpha:double=Get spine_twist_right_delta_ref_alpha() -> (spine_twist_right_delta_ref_alpha:double)
    game_time_in_seconds:double=Get game_time_in_seconds() -> (game_time_in_seconds:double)
   ) -> (Pose) {
    alphaBoolBlend {blendOption=Linear}
    controlRigAssetReference {blueprintRigClass=&/F_MED_SilentTempo_Peril_Merged_CtrlRig.F_MED_SilentTempo_Peril_Merged_CtrlRig_C}
   }
   n26 = RigidBody(ComponentPose=Local To Component(LocalPose=@n25), OverrideWorldGravity=Get gravity_override() -> (gravity_override:Vector)) {
    bClampLinearTranslationLimitToRefPose=true
    bOverrideWorldGravity=true
    overridePhysicsAsset=&/F_MED_SilentTempo_Peril_skirt_Physics
    simulationSpace=BaseBoneSpace
    alphaBoolBlend {blendOption=Linear}
    baseBoneRef {boneName=pelvis}
    componentAppliedLinearAccClamp=<2500 2500 2500>
    componentLinearAccScale=<0.4 0.3 0.1>
    componentLinearVelScale=<0.4 0.3 0.1>
    overrideWorldGravity {z=-980}
    simSpaceSettings {maxAngularAcceleration=1000, maxAngularVelocity=3, maxLinearAcceleration=250, maxLinearVelocity=500, velocityScaleZ=0, worldAlpha=1}
   }
   n27 = Anim Dynamics - Bone: dyn_lace_inner_l(ComponentPose=@n26, Alpha=1, ExternalForce="0,0,0") {
    angularDampingOverride=1
    angularSpringConstant=175
    bAngularSpring=true
    bOverrideAngularDamping=true
    simulationSpace=RootRelative
    alphaBoolBlend {blendOption=Linear}
    boundBone {boneName=dyn_lace_inner_l}
    componentAppliedLinearAccClamp=<2500 2500 2500>
    componentLinearAccScale=<0.1 0.1 0.1>
    componentLinearVelScale=<0.9 0.9 0.9>
    physicsBodyDefinitions [{
      boundBone {boneName=dyn_lace_inner_l}
      boxExtents=<6 4 4>
      constraintSetup {angularLimitsMax=<0 5 25>, angularLimitsMin=<0 -5 -25>, angularTarget=<1 0 0>}
      localJointOffset=<1.5 0 0>
     }]
   }
   n28 = Anim Dynamics - Bone: dyn_lace_outer_l(ComponentPose=@n27, Alpha=1, ExternalForce="0,0,0") {
    angularDampingOverride=1
    angularSpringConstant=175
    bAngularSpring=true
    bOverrideAngularDamping=true
    simulationSpace=RootRelative
    alphaBoolBlend {blendOption=Linear}
    boundBone {boneName=dyn_lace_outer_l}
    componentAppliedLinearAccClamp=<2500 2500 2500>
    componentLinearAccScale=<0.1 0.1 0.1>
    componentLinearVelScale=<0.9 0.9 0.9>
    physicsBodyDefinitions [{
      boundBone {boneName=dyn_lace_outer_l}
      boxExtents=<6 4 4>
      constraintSetup {angularLimitsMax=<0 5 25>, angularLimitsMin=<0 -5 -25>, angularTarget=<1 0 0>}
      localJointOffset=<1.5 0 0>
     }]
   }
   n29 = Anim Dynamics - Bone: dyn_lace_inner_r(ComponentPose=@n28, Alpha=1, ExternalForce="0,0,0") {
    angularDampingOverride=1
    angularSpringConstant=175
    bAngularSpring=true
    bOverrideAngularDamping=true
    simulationSpace=RootRelative
    alphaBoolBlend {blendOption=Linear}
    boundBone {boneName=dyn_lace_inner_r}
    componentAppliedLinearAccClamp=<2500 2500 2500>
    componentLinearAccScale=<0.1 0.1 0.1>
    componentLinearVelScale=<0.9 0.9 0.9>
    physicsBodyDefinitions [{
      boundBone {boneName=dyn_lace_inner_r}
      boxExtents=<6 4 4>
      constraintSetup {angularLimitsMax=<0 5 25>, angularLimitsMin=<0 -5 -25>, angularTarget=<1 0 0>}
      localJointOffset=<1.5 0 0>
     }]
   }
   n30 = Anim Dynamics - Bone: dyn_lace_outer_r(ComponentPose=@n29, Alpha=1, ExternalForce="0,0,0") {
    angularDampingOverride=1
    angularSpringConstant=175
    bAngularSpring=true
    bOverrideAngularDamping=true
    simulationSpace=RootRelative
    alphaBoolBlend {blendOption=Linear}
    boundBone {boneName=dyn_lace_outer_r}
    componentAppliedLinearAccClamp=<2500 2500 2500>
    componentLinearAccScale=<0.1 0.1 0.1>
    componentLinearVelScale=<0.9 0.9 0.9>
    physicsBodyDefinitions [{
      boundBone {boneName=dyn_lace_outer_r}
      boxExtents=<6 4 4>
      constraintSetup {angularLimitsMax=<0 5 25>, angularLimitsMin=<0 -5 -25>, angularTarget=<1 0 0>}
      localJointOffset=<1.5 0 0>
     }]
   }
   Save cached pose 'pre_sim_pose_cache'(Pose=Component To Local(ComponentPose=@n30)) {cacheName=pre_sim_pose_cache}
   n31 = RigidBody(
    ComponentPose=Local To Component(LocalPose=Use cached pose 'pre_sim_pose_cache'())
    Alpha=Get is_wearing_backling_REV_alpha() -> (is_wearing_backling_REV_alpha:double)
    OverrideWorldGravity=Get gravity_override()
   ) {
    bClampLinearTranslationLimitToRefPose=true
    bOverrideWorldGravity=true
    overridePhysicsAsset=&/F_MED_SilentTempo_Peril_shawl_Physics
    simulationSpace=BaseBoneSpace
    alphaBoolBlend {blendOption=Linear}
    baseBoneRef {boneName=clavicle_r}
    componentAppliedLinearAccClamp=<2500 2500 2500>
    componentLinearAccScale=<0.4 0.3 0.1>
    componentLinearVelScale=<0.4 0.3 0.1>
    overrideWorldGravity {z=-980}
    simSpaceSettings {maxAngularAcceleration=1000, maxAngularVelocity=3, maxLinearAcceleration=250, maxLinearVelocity=500, velocityScaleZ=0, worldAlpha=1}
   }
   n32 = RigidBody!(
    ComponentPose=Local To Component(LocalPose=Use cached pose 'pre_sim_pose_cache'())
    Alpha=Get is_wearing_backbling_alpha() -> (is_wearing_backbling_alpha:double)
    OverrideWorldGravity="0,0,-980"
    ExternalForce:Vector
   ) -> (Pose) {
    bClampLinearTranslationLimitToRefPose=true
    bOverrideWorldGravity=true
    overridePhysicsAsset=&/F_MED_SilentTempo_Peril_shawl_backbling_Physics
    simulationSpace=BaseBoneSpace
    alphaBoolBlend {blendOption=Linear}
    baseBoneRef {boneName=clavicle_r}
    componentAppliedLinearAccClamp=<6000 6000 6000>
    componentLinearAccScale=<0.1 0.1 0.1>
    componentLinearVelScale=<0.1 0.1 0.1>
    overrideWorldGravity {z=-980}
    simSpaceSettings {
     maxAngularAcceleration=100
     maxAngularVelocity=3
     maxLinearAcceleration=400
     maxLinearVelocity=250
     velocityScaleZ=0
     worldAlpha=1
     externalLinearDragV=<0.1 0.1>
    }
   }
   Save cached pose 'PostShawlDynamics'(Pose=Blend Poses by bool(
     BlendPose_0=Component To Local(ComponentPose=@n32)
     BlendPose_1=Component To Local(ComponentPose=@n31)
     bActiveValue=Get backpack() -> (backpack:bool)
    ) {blendType=Linear}) {cacheName=PostShawlDynamics}
   n33 = "Constraint - dyn_mstr_tassel_01 (PPPPPP)"!(
    ComponentPose=Local To Component(LocalPose=Use cached pose 'PostShawlDynamics'())
    Alpha=1
    ConstraintWeights_0=0.731
    ConstraintWeights_1:float=0.0749
    ConstraintWeights_2:float=0.0733
    ConstraintWeights_3:float=0.0666
    ConstraintWeights_4:float=0.0402
    ConstraintWeights_5:float=0.0139
   ) -> (Pose) {
    alphaBoolBlend {blendOption=Linear}
    boneToModify {boneName=dyn_mstr_tassel_01}
    constraintSetup [
     {targetBone {boneName=dyn_mstr_inner_front_skirt_01_l}}
     {targetBone {boneName=dfrm_thigh_l}}
     {targetBone {boneName=dyn_front_skirt_02_l}}
     {targetBone {boneName=dyn_front_skirt_03_l}}
     {targetBone {boneName=dyn_front_skirt_01_l}}
     {targetBone {boneName=dfrm_thigh_r}}
    ]
    constraintWeights [1,1,1,1,1,1]
   }
   n34 = "Constraint - dyn_mstr_tassel_02 (PPPPPP)"!(
    ComponentPose=@n33
    Alpha=1
    ConstraintWeights_0=0.477
    ConstraintWeights_1:float=0.2264
    ConstraintWeights_2:float=0.1922
    ConstraintWeights_3:float=0.0756
    ConstraintWeights_4:float=0.0182
    ConstraintWeights_5:float=0.0106
   ) -> (Pose) {
    alphaBoolBlend {blendOption=Linear}
    boneToModify {boneName=dyn_mstr_tassel_02}
    constraintSetup [
     {targetBone {boneName=dyn_mstr_inner_front_skirt_01_l}}
     {targetBone {boneName=dyn_front_skirt_03_l}}
     {targetBone {boneName=dyn_front_skirt_02_l}}
     {targetBone {boneName=dyn_front_skirt_01_l}}
     {targetBone {boneName=dyn_side_skirt_02_l}}
     {targetBone {boneName=dyn_side_skirt_03_l}}
    ]
    constraintWeights [1,1,1,1,1,1]
   }
   n35 = "Constraint - dyn_mstr_tassel_03 (PPPPPPP)"!(
    ComponentPose=@n34
    Alpha=1
    ConstraintWeights_0=0.4262
    ConstraintWeights_1:float=0.226
    ConstraintWeights_2:float=0.1575
    ConstraintWeights_3:float=0.0679
    ConstraintWeights_4:float=0.0596
    ConstraintWeights_5:float=0.036
    ConstraintWeights_6:float=0.0269
   ) -> (Pose) {
    alphaBoolBlend {blendOption=Linear}
    boneToModify {boneName=dyn_mstr_tassel_03}
    constraintSetup [
     {targetBone {boneName=dyn_front_skirt_03_l}}
     {targetBone {boneName=dyn_front_skirt_02_l}}
     {targetBone {boneName=dyn_mstr_inner_front_skirt_01_l}}
     {targetBone {boneName=dyn_side_skirt_04_l}}
     {targetBone {boneName=dyn_front_skirt_01_l}}
     {targetBone {boneName=dyn_side_skirt_03_l}}
     {targetBone {boneName=dyn_side_skirt_02_l}}
    ]
    constraintWeights [1,1,1,1,1,1,1]
   }
   n36 = "Constraint - dyn_mstr_tassel_04 (PPPPPP)"!(
    ComponentPose=@n35
    Alpha=1
    ConstraintWeights_0=0.5422
    ConstraintWeights_1:float=0.194
    ConstraintWeights_2:float=0.1606
    ConstraintWeights_3:float=0.0564
    ConstraintWeights_4:float=0.0257
    ConstraintWeights_5:float=0.0212
   ) -> (Pose) {
    alphaBoolBlend {blendOption=Linear}
    boneToModify {boneName=dyn_mstr_tassel_04}
    constraintSetup [
     {targetBone {boneName=dyn_front_skirt_03_l}}
     {targetBone {boneName=dyn_side_skirt_04_l}}
     {targetBone {boneName=dyn_front_skirt_02_l}}
     {targetBone {boneName=dyn_side_skirt_03_l}}
     {targetBone {boneName=dyn_front_skirt_01_l}}
     {targetBone {boneName=dyn_back_side_skirt_04_l}}
    ]
    constraintWeights [1,1,1,1,1,1]
   }
   n37 = "Constraint - dyn_mstr_tassel_05 (PPPP)"!(
    ComponentPose=@n36
    Alpha=1
    ConstraintWeights_0=0.5611
    ConstraintWeights_1:float=0.3266
    ConstraintWeights_2:float=0.0741
    ConstraintWeights_3:float=0.0383
   ) -> (Pose) {
    alphaBoolBlend {blendOption=Linear}
    boneToModify {boneName=dyn_mstr_tassel_05}
    constraintSetup [
     {targetBone {boneName=dyn_side_skirt_04_l}}
     {targetBone {boneName=dyn_front_skirt_03_l}}
     {targetBone {boneName=dyn_side_skirt_03_l}}
     {targetBone {boneName=dyn_back_side_skirt_04_l}}
    ]
    constraintWeights [1,1,1,1]
   }
   n38 = "Constraint - dyn_mstr_tassel_06 (PPPPP)"!(
    ComponentPose=@n37
    Alpha=1
    ConstraintWeights_0=0.5696
    ConstraintWeights_1:float=0.319
    ConstraintWeights_2:float=0.081
    ConstraintWeights_3:float=0.0161
    ConstraintWeights_4:float=0.0144
   ) -> (Pose) {
    alphaBoolBlend {blendOption=Linear}
    boneToModify {boneName=dyn_mstr_tassel_06}
    constraintSetup [
     {targetBone {boneName=dyn_side_skirt_04_l}}
     {targetBone {boneName=dyn_back_side_skirt_04_l}}
     {targetBone {boneName=dyn_side_skirt_03_l}}
     {targetBone {boneName=dyn_back_skirt_03_l}}
     {targetBone {boneName=dyn_back_side_skirt_03_l}}
    ]
    constraintWeights [1,1,1,1,1]
   }
   n39 = "Constraint - dyn_mstr_tassel_07 (PPPPPP)"!(
    ComponentPose=@n38
    Alpha=1
    ConstraintWeights_0=0.3685
    ConstraintWeights_1:float=0.2174
    ConstraintWeights_2:float=0.161
    ConstraintWeights_3:float=0.1491
    ConstraintWeights_4:float=0.089
    ConstraintWeights_5:float=0.015
   ) -> (Pose) {
    alphaBoolBlend {blendOption=Linear}
    boneToModify {boneName=dyn_mstr_tassel_07}
    constraintSetup [
     {targetBone {boneName=dyn_back_side_skirt_04_l}}
     {targetBone {boneName=dyn_side_skirt_04_l}}
     {targetBone {boneName=dyn_back_skirt_03_l}}
     {targetBone {boneName=dyn_back_side_skirt_03_l}}
     {targetBone {boneName=dyn_side_skirt_03_l}}
     {targetBone {boneName=dyn_back_skirt_02_l}}
    ]
    constraintWeights [1,1,1,1,1,1]
   }
   n40 = "Constraint - dyn_mstr_tassel_08 (PPPPP)"!(
    ComponentPose=@n39
    Alpha=1
    ConstraintWeights_0=0.3547
    ConstraintWeights_1:float=0.2487
    ConstraintWeights_2:float=0.1979
    ConstraintWeights_3:float=0.1212
    ConstraintWeights_4:float=0.0774
   ) -> (Pose) {
    alphaBoolBlend {blendOption=Linear}
    boneToModify {boneName=dyn_mstr_tassel_08}
    constraintSetup [
     {targetBone {boneName=dyn_back_skirt_03_l}}
     {targetBone {boneName=dyn_back_skirt_02_l}}
     {targetBone {boneName=dyn_back_side_skirt_04_l}}
     {targetBone {boneName=dyn_back_side_skirt_03_l}}
     {targetBone {boneName=dyn_back_skirt_02_r}}
    ]
    constraintWeights [1,1,1,1,1]
   }
   n41 = "Constraint - dyn_mstr_tassel_09 (PPP)"!(ComponentPose=@n40, Alpha=1, ConstraintWeights_0=0.3396, ConstraintWeights_1:float=0.3374, ConstraintWeights_2:float=0.323) -> (Pose) {
    alphaBoolBlend {blendOption=Linear}
    boneToModify {boneName=dyn_mstr_tassel_09}
    constraintSetup [{targetBone {boneName=dyn_back_skirt_03_l}}, {targetBone {boneName=dyn_back_skirt_02_l}}, {targetBone {boneName=dyn_back_skirt_02_r}}]
    constraintWeights [1,1,1]
   }
   n42 = "Constraint - dyn_mstr_tassel_10 (PPPP)"!(
    ComponentPose=@n41
    Alpha=1
    ConstraintWeights_0=0.6402
    ConstraintWeights_1:float=0.1914
    ConstraintWeights_2:float=0.1393
    ConstraintWeights_3:float=0.0292
   ) -> (Pose) {
    alphaBoolBlend {blendOption=Linear}
    boneToModify {boneName=dyn_mstr_tassel_10}
    constraintSetup [
     {targetBone {boneName=dyn_back_skirt_02_r}}
     {targetBone {boneName=dyn_back_skirt_03_l}}
     {targetBone {boneName=dyn_back_skirt_02_l}}
     {targetBone {boneName=dfrm_thigh_r}}
    ]
    constraintWeights [1,1,1,1]
   }
   n43 = "Constraint - dyn_mstr_tassel_11 (PPPPP)"!(
    ComponentPose=@n42
    Alpha=1
    ConstraintWeights_0=0.6191
    ConstraintWeights_1:float=0.2505
    ConstraintWeights_2:float=0.0641
    ConstraintWeights_3:float=0.0553
    ConstraintWeights_4:float=0.0111
   ) -> (Pose) {
    alphaBoolBlend {blendOption=Linear}
    boneToModify {boneName=dyn_mstr_tassel_11}
    constraintSetup [
     {targetBone {boneName=dyn_back_skirt_02_r}}
     {targetBone {boneName=dfrm_thigh_r}}
     {targetBone {boneName=dyn_back_skirt_03_l}}
     {targetBone {boneName=dyn_back_skirt_02_l}}
     {targetBone {boneName=dyn_back_skirt_01_r}}
    ]
    constraintWeights [1,1,1,1,1]
   }
   n44 = "Constraint - dyn_mstr_tassel_12 (PP)"!(ComponentPose=@n43, Alpha=1, ConstraintWeights_0=0.9236, ConstraintWeights_1:float=0.0764) -> (Pose) {
    alphaBoolBlend {blendOption=Linear}
    boneToModify {boneName=dyn_mstr_tassel_12}
    constraintSetup [{targetBone {boneName=dfrm_thigh_r}}, {targetBone {boneName=dyn_back_skirt_02_r}}]
    constraintWeights [1,1]
   }
   n45 = "Constraint - dyn_mstr_lapel_lace_l (PPPPP)"!(
    ComponentPose=@n44
    Alpha=1
    ConstraintWeights_0=0.7172
    ConstraintWeights_1:float=0.1372
    ConstraintWeights_2:float=0.0688
    ConstraintWeights_3:float=0.0633
    ConstraintWeights_4:float=0.0135
   ) -> (Pose) {
    alphaBoolBlend {blendOption=Linear}
    boneToModify {boneName=dyn_mstr_lapel_lace_l}
    constraintSetup [
     {targetBone {boneName=spine_05}}
     {targetBone {boneName=spine_04}}
     {targetBone {boneName=neck_01}}
     {targetBone {boneName=clavicle_l}}
     {targetBone {boneName=spine_03}}
    ]
    constraintWeights [1,1,1,1,1]
   }
   n46 = "Constraint - dyn_mstr_lapel_lace_r (PPPPP)"!(
    ComponentPose=@n45
    Alpha=1
    ConstraintWeights_0=0.7171
    ConstraintWeights_1:float=0.1372
    ConstraintWeights_2:float=0.069
    ConstraintWeights_3:float=0.0631
    ConstraintWeights_4:float=0.0135
   ) -> (Pose) {
    alphaBoolBlend {blendOption=Linear}
    boneToModify {boneName=dyn_mstr_lapel_lace_r}
    constraintSetup [
     {targetBone {boneName=spine_05}}
     {targetBone {boneName=spine_04}}
     {targetBone {boneName=neck_01}}
     {targetBone {boneName=clavicle_r}}
     {targetBone {boneName=spine_03}}
    ]
    constraintWeights [1,1,1,1,1]
   }
   Save cached pose 'PostRivets'(Pose=Component To Local(ComponentPose=@n46)) {cacheName=PostRivets}
   Get gravity_override()
   Get left_right_movement_alpha() -> (left_right_movement_alpha:double)
  ]}
 EventGraph {nodes [
   n1 = Event BlueprintUpdateAnimation()
   > n2 = Sequence!(execute) -> (then_0, then_1, then_2:Exec, then_3:Exec, then_4:Exec, then_5:Exec, then_6:Exec)
   n3 = Gravity Override State Function DEPRECATED(execute=@n2.then_1, CharacterPartAnimInstance=Self-Reference(), GameTimeInSeconds=GetGameTimeInSeconds())
   > Set gravity_override(gravity_override=@n3)
   Set spine_twist_right_delta_ref_alpha(
    execute=@n2.then_5
    spine_twist_right_delta_ref_alpha:double=GetAlphaFromReferenceTransformDelta(
     SkeletalMeshComponent=Get skeletal_mesh() -> (skeletal_mesh:Skeletal Mesh Component Object Reference)
     BoneName=spine_04
     BaseName=pelvis
     DeltaComponentType=QuaternionTwist
     InputRange=(X=-7,Y=-31)
     TwistAxis="1,0,0"
    )
   ) -> (Output_Get:double)
   n4 = PlayerMovement(execute=@n2.then_0, AnimInstance=Self-Reference(), XDotProductCoefficient=1, YDotProductCoefficient=1, ZDotProductCoefficient=1)
   > Set left_right_movement_alpha(left_right_movement_alpha:double=@n4) -> (Output_Get:double)
   > Set forwards_backward_movement_alpha(forwards_backward_movement_alpha:double=@n4.YMovement) -> (Output_Get:double)
   Branch(execute=@n2.then_3, Condition=Get bIsSkyDiving())
   > Set WPO Globals Function(
    Material=Get material_instance_dynamic() -> (material_instance_dynamic:Material Instance Dynamic Object Reference)
    Overlay Material=Get Overlay_MID() -> (Overlay_MID:Material Instance Dynamic Object Reference)
    Skeletal Mesh=Get skeletal_mesh()
    "Local Noise Transform  Location"="Get Local Noise Transform  Location"() -> ("Local Noise Transform  Location":Vector)
    Local Noise Offset=Get Local Noise Offset()
   )
   > Set WPO Regional Function!(
    execute
    Material:Material Instance Dynamic Object Reference=Get material_instance_dynamic()
    Overlay Material:Material Instance Dynamic Object Reference=Get Overlay_MID()
    Skeletal Mesh:Skeletal Mesh Component Object Reference=Get skeletal_mesh()
    Velocity Transform:Name=foot_l
    Velocity Interp Speed:double=Get Velocity Interp Speed()
    Collision Transform:Name=calf_l
    Collision Transform Axis:EAxis Enum
    Invert Collision Transform Axis:bool=true
    Delta Time X:double=@n1.DeltaTimeX
    Previous Velocity Transform Location:"Vector (by ref)"="Get RG1 Previous Velocity Transform Location "() -> ("RG1 Previous Velocity Transform Location ":Vector)
    Previous Velocity Transform Velocity:"Vector (by ref)"=Get RG1 Previous Velocity Transform Velocity() -> (RG1 Previous Velocity Transform Velocity:Vector)
    Region:int=1
   ) -> (then)
   > Set WPO Regional Function(
    Material=Get material_instance_dynamic()
    Overlay Material=Get Overlay_MID()
    Skeletal Mesh=Get skeletal_mesh()
    Velocity Transform=foot_r
    Velocity Interp Speed=Get Velocity Interp Speed()
    Collision Transform=calf_r
    Invert Collision Transform Axis=true
    Delta Time X=@n1.DeltaTimeX
    Previous Velocity Transform Location=Get RG2 Previous Velocity Transform Location() -> (RG2 Previous Velocity Transform Location:Vector)
    Previous Velocity Transform Velocity=Get RG2 Previous Velocity Transform Velocity() -> (RG2 Previous Velocity Transform Velocity:Vector)
    Region=2
   )
   > Set WPO Regional Function(
    Material=Get material_instance_dynamic()
    Overlay Material=Get Overlay_MID()
    Skeletal Mesh=Get skeletal_mesh()
    Velocity Transform=hand_r
    Velocity Interp Speed=Get Velocity Interp Speed()
    Collision Transform=lowerarm_r
    Invert Collision Transform Axis=true
    Delta Time X=@n1.DeltaTimeX
    Previous Velocity Transform Location=Get RG3 Previous Velocity Transform Location() -> (RG3 Previous Velocity Transform Location:Vector)
    Previous Velocity Transform Velocity=Get RG3 Previous Velocity Transform Velocity() -> (RG3 Previous Velocity Transform Velocity:Vector)
    Region=3
   )
   Set game_time_in_seconds(execute=@n2.then_2, game_time_in_seconds:double=GetGameTimeInSeconds()) -> (Output_Get:double)
   n5 = Branch(execute=@n2.then_6, Condition=Get bIsInFrontEnd())
   > Set Built In Emote(Built In Emote:double=GetCurveValue(self=GetOwnerFrontEndAnimInstance(), CurveName=emote_BuiltIn)) -> (Output_Get:double)
   Set Built In Emote(
    execute=@n5.else
    Built In Emote=GetCurveValue(
     self=GetOwnerFortPlayerAnimInstance!(self:Custom Character Part Anim Instance Object Reference) -> (ReturnValue:Fort Player Anim Instance Object Reference)
     CurveName=emote_BuiltIn
    )
   )
   n6 = Branch(execute=@n2.then_4, Condition=Get bIsSkydiveFloating())
   > n7 = Set WPO PHO Wind Intensity(WPO PHO Wind Intensity:double=1) -> (Output_Get:double)
   SetScalarParameterValue(
    execute=[@n7.then, @n8.then]
    self=Get material_instance_dynamic()
    ParameterName=None
    Value=Get WPO PHO Wind Intensity() -> (WPO PHO Wind Intensity:double)
   )
   n8 = Set WPO PHO Wind Intensity(execute=@n6.else, WPO PHO Wind Intensity=0)
   Event OnCharacterCustomizationLoaded()
   > n9 = Sequence!(execute) -> (then_0, then_1)
   > n10 = Set skeletal_mesh(skeletal_mesh:Skeletal Mesh Component Object Reference=GetPartSkeletalMeshComponent()) -> (Output_Get:Skeletal Mesh Component Object Reference)
   > n11 = Get Dynamic Material Instance(SkeletalMeshComponent=@n10)
   > Set material_instance_dynamic(material_instance_dynamic:Material Instance Dynamic Object Reference=@n11) -> (Output_Get:Material Instance Dynamic Object Reference)
   > n12 = GetOrCreateDynamicOverlayMaterial(MeshComponent=@n10, MaterialName=None)
   > Set Overlay_MID(Overlay_MID:Material Instance Dynamic Object Reference=@n12) -> (Output_Get:Material Instance Dynamic Object Reference)
   n13 = Is Valid(exec=@n9.then_1, InputObject=Get pawn_owner() -> (pawn_owner:Player Pawn Athena Object Reference))
   n14 = CheckForBackbling()
   n15 = Branch(
    execute=[@n13.Is Valid, @n17.Completed]
    Condition=OR Boolean!(A=@n14, B=@n14.bIsBackblingACape, C:bool=@n14.bIsBackblingLong) -> (ReturnValue)
   )
   > Set is_wearing_backbling_alpha(is_wearing_backbling_alpha:double=1) -> (Output_Get:double)
   > Set is_wearing_backling_REV_alpha(is_wearing_backling_REV_alpha:double=0) -> (Output_Get:double)
   > Set backpack(backpack:bool=true) -> (Output_Get:bool)
   Set is_wearing_backbling_alpha(execute=@n15.else, is_wearing_backbling_alpha=0)
   > Set is_wearing_backling_REV_alpha(is_wearing_backling_REV_alpha=1)
   > Set backpack(backpack=false)
   Event BlueprintInitializeAnimation()
   > n16 = Cast To PlayerPawn_Athena(Object=TryGetPawnOwner())
   > Set pawn_owner(pawn_owner:Player Pawn Athena Object Reference=@n16) -> (Output_Get:Player Pawn Athena Object Reference)
   n17 = Do Once()
   Reroute Node(InputPin:Wildcard) -> (OutputPin:Wildcard)
   "float > float"(A="Absolute (Float)"(A=Get!() -> (PawnVelocity_X:double, PawnVelocity_Y:double, PawnVelocity_Z:double).PawnVelocity_Y), B=550)
   Get bIsSprinting()
  ]}
]
variables [
 backpack {type=boolean}
 Built In Emote {type=number}
 FE String Alpha {type=number}
 forwards_backward_movement_alpha {type=number}
 game_time_in_seconds {type=number}
 gravity_override {type=Vector}
 is_wearing_backbling_alpha {type=number}
 is_wearing_backling_REV_alpha {type=number}
 left_right_movement_alpha {type=number}
 Local Noise Offset {type=Vector}
 "Local Noise Transform  Location" {type=Vector}
 material_instance_dynamic {type=MaterialInstanceDynamic}
 Overlay_MID {type=MaterialInstanceDynamic}
 pawn_owner {type=PlayerPawn_Athena_C}
 "RG1 Previous Velocity Transform Location " {type=Vector}
 RG1 Previous Velocity Transform Velocity {type=Vector}
 RG2 Previous Velocity Transform Location {type=Vector}
 RG2 Previous Velocity Transform Velocity {type=Vector}
 RG3 Previous Velocity Transform Location {type=Vector}
 RG3 Previous Velocity Transform Velocity {type=Vector}
 skeletal_mesh {type=SkeletalMeshComponent}
 spine_twist_right_delta_ref_alpha {type=number}
 Velocity Interp Speed {type=number, default=10}
 WPO PHO Wind Intensity {type=number}
]
