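# Post-copy animation Blueprint for the SilentTempo Peril face accessory part, adding secondary motion to hair, bangs and earrings.
# The copied parent pose runs through a Control Rig driven by head-angle deltas, movement states and backbling state, then is cached.
# On customization load, a backbling check selects one of two RigidBody simulations with different physics assets, base bones and gravity.
# A chain of spring-based AnimDynamics nodes then swings individual bangs, top hair and earrings.
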AnimBlueprint /BRCosmetics/Characters/Player/Female/Medium/Bodies/F_MED_SilentTempo_Peril/Meshes/Parts/F_MED_SilentTempo_Peril_FaceAcc_AnimBP
target_skeleton=/Game/Characters/Player/Female/Medium/Bodies/F_MED_SilentTempo_Peril/Meshes/Parts/F_MED_SilentTempo_Peril_FaceAcc_Skeleton
graphs [
 AnimGraph {nodes [
   n1 = RigidBody!(
    ComponentPose=Local To Component(LocalPose=Use cached pose 'pre_sim_cache'())
    Alpha=Get is_wearing_backling_REV_alpha() -> (is_wearing_backling_REV_alpha:double)
    OverrideWorldGravity:Vector=Get gravity_override() -> (gravity_override:Vector)
    ExternalForce
   ) -> (Pose) {
    bClampLinearTranslationLimitToRefPose=true
    bOverrideWorldGravity=true
    overridePhysicsAsset=/Game/Characters/Player/Female/Medium/Bodies/F_MED_SilentTempo_Peril/Meshes/Parts/F_MED_SilentTempo_Peril_FaceAcc_Physics
    simulationSpace=BaseBoneSpace
    alphaBoolBlend {blendOption=Linear}
    baseBoneRef {boneName=spine_05}
    componentAppliedLinearAccClamp=<1850 1850 1850>
    componentLinearAccScale=<0.4 0.15 0.3>
    componentLinearVelScale=<0.4 0.15 0.3>
    overrideWorldGravity {z=-980}
    simSpaceSettings {maxAngularAcceleration=3000, maxAngularVelocity=3, maxLinearAcceleration=250, maxLinearVelocity=500, velocityScaleZ=0, worldAlpha=1}
   }
   n2 = Component To Local(ComponentPose=@n1)
   n3 = RigidBody(
    ComponentPose=Local To Component(LocalPose=Use cached pose 'pre_sim_cache'())
    Alpha=Get is_wearing_backbling_alpha() -> (is_wearing_backbling_alpha:double)
    OverrideWorldGravity="0,0,-980"
   ) {
    bClampLinearTranslationLimitToRefPose=true
    bOverrideWorldGravity=true
    overridePhysicsAsset=/Game/Characters/Player/Female/Medium/Bodies/F_MED_SilentTempo_Peril/Meshes/Parts/F_MED_SilentTempo_Peril_FaceAcc_Backbling_Physics
    simulationSpace=BaseBoneSpace
    alphaBoolBlend {blendOption=Linear}
    baseBoneRef {boneName=neck_02}
    componentAppliedLinearAccClamp=<6000 6000 6000>
    componentLinearAccScale=<0.1 0.1 0.1>
    componentLinearVelScale=<0.1 0.1 0.1>
    overrideWorldGravity {z=-980}
    simSpaceSettings {
     maxAngularAcceleration=100
     maxAngularVelocity=3
     maxLinearAcceleration=400
     maxLinearVelocity=250
     velocityScaleZ=0
     worldAlpha=1
     externalLinearDragV=<0.1 0.1>
    }
   }
   n4 = Component To Local(ComponentPose=@n3)
   n5 = Get backpack() -> (backpack:bool)
   n6 = Anim Dynamics - Bone: dyn_earring_l(
    ComponentPose=Local To Component(LocalPose=Blend Poses by bool(BlendPose_0=@n4, BlendPose_1=@n2, bActiveValue=@n5) {blendType=Linear})
    Alpha=1
    ExternalForce="0,0,0"
   ) {
    angularDampingOverride=1
    angularSpringConstant=50
    bAngularSpring=true
    bOverrideAngularDamping=true
    bUsePlanarLimit=false
    bUseSphericalLimits=true
    simulationSpace=RootRelative
    alphaBoolBlend {blendOption=Linear}
    boundBone {boneName=dyn_earring_l}
    componentAppliedLinearAccClamp=<2500 2500 2500>
    componentLinearAccScale=<0.1 0.1 0.1>
    componentLinearVelScale=<0.9 0.9 0.9>
    physicsBodyDefinitions [{
      boundBone {boneName=dyn_earring_l}
      boxExtents=<12 1.089 7>
      constraintSetup {angularLimitsMax=<2 20 15>, angularLimitsMin=<-2 -20 -15>, angularTarget=<1 0 0>}
      localJointOffset=<4 0 0>
     }]
    sphericalLimits [{limitRadius=33.2978, drivingBone {boneName=head}, sphereLocalOffset=<2.3132 0 27.4163>}]
   }
   n7 = Anim Dynamics - Bone: dyn_earring_r(ComponentPose=@n6, Alpha=1, ExternalForce="0,0,0") {
    angularDampingOverride=0.9999
    angularSpringConstant=50
    bAngularSpring=true
    bOverrideAngularDamping=true
    simulationSpace=RootRelative
    alphaBoolBlend {blendOption=Linear}
    boundBone {boneName=dyn_earring_r}
    componentAppliedLinearAccClamp=<2500 2500 2500>
    componentLinearAccScale=<0.1 0.1 0.1>
    componentLinearVelScale=<0.9 0.9 0.9>
    physicsBodyDefinitions [{
      boundBone {boneName=dyn_earring_r}
      boxExtents=<12 7 7>
      constraintSetup {angularLimitsMax=<4 20 5>, angularLimitsMin=<-4 -20 -15>, angularTarget=<1 0 0>}
      localJointOffset=<4 0 0>
     }]
   }
   n8 = Anim Dynamics - Bone: dyn_bangs_B_01_l(ComponentPose=@n7, Alpha=1, ExternalForce="0,0,0") {
    angularDampingOverride=0.9999
    angularSpringConstant=50
    bAngularSpring=true
    bOverrideAngularDamping=true
    simulationSpace=RootRelative
    alphaBoolBlend {blendOption=Linear}
    boundBone {boneName=dyn_bangs_B_01_l}
    componentAppliedLinearAccClamp=<2500 2500 2500>
    componentLinearAccScale=<0.1 0.1 0.1>
    componentLinearVelScale=<0.9 0.9 0.9>
    physicsBodyDefinitions [{
      boundBone {boneName=dyn_bangs_B_01_l}
      boxExtents=<8.5 7.5 7.5>
      constraintSetup {angularLimitsMax=<2 10 2.5>, angularLimitsMin=<-2 -2.5 -20>, angularTarget=<1 0 0>}
      localJointOffset=<4 0 0>
     }]
   }
   n9 = Anim Dynamics - Bone: dyn_hair_top(ComponentPose=@n8, Alpha=1, ExternalForce="0,0,0") {
    angularDampingOverride=0.99
    angularSpringConstant=500
    bAngularSpring=true
    bOverrideAngularDamping=true
    bUsePlanarLimit=false
    simulationSpace=RootRelative
    alphaBoolBlend {blendOption=Linear}
    boundBone {boneName=dyn_hair_top}
    componentAppliedLinearAccClamp=<2500 2500 2500>
    componentLinearAccScale=<0.1 0.1 0.1>
    componentLinearVelScale=<0.9 0.9 0.9>
    physicsBodyDefinitions [{
      boundBone {boneName=dyn_hair_top}
      boxExtents=<10 8 8>
      constraintSetup {angularLimitsMax=<2 25 15>, angularLimitsMin=<-2 -25 -15>, angularTarget=<1 0 0>}
      localJointOffset=<2 0 0>
     }]
   }
   n10 = Anim Dynamics - Bone: dyn_bangs_C_l(ComponentPose=@n9, Alpha=1, ExternalForce="0,0,0") {
    angularDampingOverride=1
    angularSpringConstant=65
    bAngularSpring=true
    bOverrideAngularDamping=true
    simulationSpace=RootRelative
    alphaBoolBlend {blendOption=Linear}
    boundBone {boneName=dyn_bangs_C_l}
    componentAppliedLinearAccClamp=<2500 2500 2500>
    componentLinearAccScale=<0.1 0.1 0.1>
    componentLinearVelScale=<0.9 0.9 0.9>
    physicsBodyDefinitions [{
      boundBone {boneName=dyn_bangs_C_l}
      boxExtents=<10 8 8>
      constraintSetup {angularLimitsMax=<2 15 2>, angularLimitsMin=<-2 -15 -20>, angularTarget=<1 0 0>}
      localJointOffset=<3 0 0>
     }]
   }
   n11 = Anim Dynamics - Bone: dyn_bangs_D_01_l(ComponentPose=@n10, Alpha=1, ExternalForce="0,0,0") {
    angularDampingOverride=1
    angularSpringConstant=65
    bAngularSpring=true
    bOverrideAngularDamping=true
    simulationSpace=RootRelative
    alphaBoolBlend {blendOption=Linear}
    boundBone {boneName=dyn_bangs_D_01_l}
    componentAppliedLinearAccClamp=<2500 2500 2500>
    componentLinearAccScale=<0.1 0.1 0.1>
    componentLinearVelScale=<0.9 0.9 0.9>
    physicsBodyDefinitions [{
      boundBone {boneName=dyn_bangs_D_01_l}
      boxExtents=<10 8 8>
      constraintSetup {angularLimitsMax=<2 15 10>, angularLimitsMin=<-2 -15 -20>, angularTarget=<1 0 0>}
      localJointOffset=<3 0 0>
     }]
   }
   n12 = Anim Dynamics - Bone: dyn_bangs_EA_01_l(ComponentPose=@n11, Alpha=1, ExternalForce="0,0,0") {
    angularDampingOverride=1
    angularSpringConstant=65
    bAngularSpring=true
    bOverrideAngularDamping=true
    simulationSpace=RootRelative
    alphaBoolBlend {blendOption=Linear}
    boundBone {boneName=dyn_bangs_EA_01_l}
    componentAppliedLinearAccClamp=<2500 2500 2500>
    componentLinearAccScale=<0.1 0.1 0.1>
    componentLinearVelScale=<0.9 0.9 0.9>
    physicsBodyDefinitions [{
      boundBone {boneName=dyn_bangs_EA_01_l}
      boxExtents=<10 8 8>
      constraintSetup {angularLimitsMax=<2 15 10>, angularLimitsMin=<-2 -15 -20>, angularTarget=<1 0 0>}
      localJointOffset=<3 0 0>
     }]
   }
   n13 = Anim Dynamics - Bone: dyn_bangs_A_01_r(ComponentPose=@n12, Alpha=1, ExternalForce="0,0,0") {
    angularDampingOverride=1
    angularSpringConstant=50
    bAngularSpring=true
    bOverrideAngularDamping=true
    simulationSpace=RootRelative
    alphaBoolBlend {blendOption=Linear}
    boundBone {boneName=dyn_bangs_A_01_r}
    componentAppliedLinearAccClamp=<2500 2500 2500>
    componentLinearAccScale=<0.1 0.1 0.1>
    componentLinearVelScale=<0.9 0.9 0.9>
    physicsBodyDefinitions [{
      boundBone {boneName=dyn_bangs_A_01_r}
      boxExtents=<10 9 9>
      constraintSetup {angularLimitsMax=<2 10 2.5>, angularLimitsMin=<-2 -2.5 -20>, angularTarget=<1 0 0>}
      localJointOffset=<3 0 0>
     }]
   }
   n14 = Anim Dynamics - Bone: dyn_bangs_C_01_r(ComponentPose=@n13, Alpha=1, ExternalForce="0,0,0") {
    angularDampingOverride=1
    angularSpringConstant=50
    bAngularSpring=true
    bOverrideAngularDamping=true
    simulationSpace=RootRelative
    alphaBoolBlend {blendOption=Linear}
    boundBone {boneName=dyn_bangs_C_01_r}
    componentAppliedLinearAccClamp=<2500 2500 2500>
    componentLinearAccScale=<0.1 0.1 0.1>
    componentLinearVelScale=<0.9 0.9 0.9>
    physicsBodyDefinitions [{
      boundBone {boneName=dyn_bangs_C_01_r}
      boxExtents=<10 9 9>
      constraintSetup {angularLimitsMax=<2 10 2.5>, angularLimitsMin=<-2 10 -20>, angularTarget=<1 0 0>}
      localJointOffset=<3 0 0>
     }]
   }
   n15 = Anim Dynamics - Bone: dyn_bangs_D_01_r(ComponentPose=@n14, Alpha=1, ExternalForce="0,0,0") {
    angularDampingOverride=1
    angularSpringConstant=250
    bAngularSpring=true
    bOverrideAngularDamping=true
    simulationSpace=RootRelative
    alphaBoolBlend {blendOption=Linear}
    boundBone {boneName=dyn_bangs_D_01_r}
    componentAppliedLinearAccClamp=<2500 2500 2500>
    componentLinearAccScale=<0.1 0.1 0.1>
    componentLinearVelScale=<0.9 0.9 0.9>
    physicsBodyDefinitions [{
      boundBone {boneName=dyn_bangs_D_01_r}
      boxExtents=<10 9 9>
      constraintSetup {angularLimitsMax=<2 60 5>, angularLimitsMin=<-2 -10 -20>, angularTarget=<1 0 0>}
      localJointOffset=<3 0 0>
     }]
   }
   n16 = Anim Dynamics - Bone: dyn_bangs_E_01_r(ComponentPose=@n15, Alpha=1, ExternalForce="0,0,0") {
    angularDampingOverride=1
    angularSpringConstant=50
    bAngularSpring=true
    bOverrideAngularDamping=true
    simulationSpace=RootRelative
    alphaBoolBlend {blendOption=Linear}
    boundBone {boneName=dyn_bangs_E_01_r}
    componentAppliedLinearAccClamp=<2500 2500 2500>
    componentLinearAccScale=<0.1 0.1 0.1>
    componentLinearVelScale=<0.9 0.9 0.9>
    physicsBodyDefinitions [{
      boundBone {boneName=dyn_bangs_E_01_r}
      boxExtents=<10 9 9>
      constraintSetup {angularLimitsMax=<2 10 2.5>, angularLimitsMin=<-2 10 -20>, angularTarget=<1 0 0>}
      localJointOffset=<3 0 0>
     }]
   }
   Output Pose(Result=Component To Local(ComponentPose=@n16)) {layerGroup=None}
   n17 = Control Rig!(
    Source=Copy Pose From Mesh() {bUseAttachedParent=true}
    Alpha
    is_wearing_backbling:double=Get is_wearing_backbling_alpha()
    head_rotate_down_delta_alpha:double=Get head_rotate_down_delta_alpha() -> (head_rotate_down_delta_alpha:double)
    head_tilt_right_delta_alpha:double=Get head_tilt_right_delta_alpha() -> (head_tilt_right_delta_alpha:double)
    is_tactical_sprinting:bool=Get bIsTacticalSprinting()
    is_jumping:bool=Get bIsJumping()
    is_crouching:bool=Get bIsCrouching()
   ) -> (Pose) {
    alphaBoolBlend {blendOption=Linear}
    controlRigAssetReference {blueprintRigClass=/Game/Characters/Player/Female/Medium/Bodies/F_MED_SilentTempo_Peril/Meshes/Parts/F_MED_SilentTempo_Peril_FaceAcc_CtrlRig.F_MED_SilentTempo_Peril_FaceAcc_CtrlRig_C}
   }
   Save cached pose 'pre_sim_cache'(Pose=@n17) {cacheName=pre_sim_cache}
  ]}
 EventGraph {nodes [
   Event BlueprintUpdateAnimation()
   > n1 = Sequence()
   > n2 = Gravity Override State Function DEPRECATED(CharacterPartAnimInstance=Self-Reference(), GameTimeInSeconds=GetGameTimeInSeconds())
   > Set gravity_override(gravity_override=@n2)
   n3 = Get body_mesh() -> (body_mesh:Skeletal Mesh Component Object Reference)
   Set head_rotate_down_delta_alpha(
    execute=@n1.then_1
    head_rotate_down_delta_alpha:double=GetAlphaFromReferenceTransformDelta(SkeletalMeshComponent=@n3, BoneName=head, BaseName=spine_05, DeltaComponentType=QuaternionTwist, InputRange=(X=10,Y=-20), TwistAxis="0,0,1")
   ) -> (Output_Get:double)
   > Set head_tilt_right_delta_alpha(head_tilt_right_delta_alpha:double=GetAlphaFromReferenceTransformDelta(
     SkeletalMeshComponent=@n3
     BoneName=neck_02
     BaseName=spine_03
     DeltaComponentType=QuaternionTwist
     InputRange=(X=10,Y=18)
     TwistAxis="0,1,0"
    )) -> (Output_Get:double)
   Event OnCharacterCustomizationLoaded()
   > n4 = Is Valid(InputObject=Get pawn_owner() -> (pawn_owner:Player Pawn Athena Object Reference))
   n5 = CheckForBackbling()
   n6 = Branch(execute=[@n4.Is Valid, @n8.Completed], Condition=OR Boolean!(A=@n5, B=@n5.bIsBackblingACape, C:bool=@n5.bIsBackblingLong) -> (ReturnValue))
   > Set is_wearing_backbling_alpha(is_wearing_backbling_alpha:double=1) -> (Output_Get:double)
   > Set is_wearing_backling_REV_alpha(is_wearing_backling_REV_alpha:double=0) -> (Output_Get:double)
   > Set backpack(backpack:bool=true) -> (Output_Get:bool)
   Set is_wearing_backbling_alpha(execute=@n6.else, is_wearing_backbling_alpha=0)
   > Set is_wearing_backling_REV_alpha(is_wearing_backling_REV_alpha=1)
   > Set backpack(backpack=false)
   Event BlueprintInitializeAnimation()
   > Set body_mesh(body_mesh:Skeletal Mesh Component Object Reference=GetOwnerSkeletalMeshComponent()) -> (Output_Get:Skeletal Mesh Component Object Reference)
   > n7 = Cast To PlayerPawn_Athena(Object=TryGetPawnOwner())
   > Set pawn_owner(pawn_owner:Player Pawn Athena Object Reference=@n7) -> (Output_Get:Player Pawn Athena Object Reference)
   n8 = Do Once()
  ]}
]
variables [
 backpack {type=boolean}
 body_mesh {type=SkeletalMeshComponent}
 gravity_override {type=Vector}
 head_rotate_down_delta_alpha {type=number}
 head_tilt_right_delta_alpha {type=number}
 is_wearing_backbling_alpha {type=number}
 is_wearing_backling_REV_alpha {type=number}
 pawn_owner {type=PlayerPawn_Athena_C}
 spine_rotate_right_alpha {type=number}
]
